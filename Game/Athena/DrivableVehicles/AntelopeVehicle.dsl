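; Antelope drivable vehicle Blueprint for Athena, handling cosmetic feedback, boost cues and attached weapon pickups.
; Construction creates the boost meter material and one shared wheel-blur material per axle; helpers push boost, damage and headlight values to materials, caching the latter three.
; Hits send magnitude-scaled shakes and haptics to driver and passenger, plus a directional shake on sudden, mostly horizontal speed loss.
; Other functions apply or remove boost effects and socket pickups with randomized rotation.
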(fn AddCooldownCueEvent
  (params (Player%20Pawn :type Fort%20Player%20Pawn%20Object%20Reference))
  (bind _returnvalue (Environment|ApplyGlobalEnvironmentGameplayEffecttoActor Player%20Pawn (Variables|Default|GetGE_Boost)))
  (Variables|Default|SetCooldownCueEvent _returnvalue))

(fn AttachWeaponToCart
  (params (Selection :type Integer))
  (bind (_returnvalue_roll _returnvalue_pitch _returnvalue_yaw) (Math|Rotator|BreakRotator (Transformation|GetActorRotation (Variables|Default|GetFortPickup))))
  (bind _skeletalmesh (Variables|Vehicle|GetSkeletalMesh))
  (bind _fortpickup (Variables|Default|GetFortPickup))
  (switch int Selection
    (:Default
      _)
    (:0
      _)
    (:1
      (Transformation|AttachActorToComponent _fortpickup _skeletalmesh "Weapon2" "SnapToTarget" "SnapToTarget" "SnapToTarget" true)
      (Transformation|SetActorRotation (Variables|Default|GetFortPickup) (Math|Rotator|MakeRotator (Math|Random|RandomFloatinRange (+ _returnvalue_roll -1e+02) (+ _returnvalue_roll -8e+01)) (Math|Random|RandomFloatinRange (+ _returnvalue_pitch -5.0) (+ _returnvalue_pitch 5.0)) (Math|Random|RandomFloatinRange (+ _returnvalue_yaw -1e+02) (+ _returnvalue_yaw -8e+01))) false :!label _label))
    (:2
      (Transformation|AttachActorToComponent _fortpickup _skeletalmesh "Weapon3" "SnapToTarget" "SnapToTarget" "SnapToTarget" true)
      (goto _label))
    (:3
      (Transformation|AttachActorToComponent _fortpickup _skeletalmesh "Weapon4" "SnapToTarget" "SnapToTarget" "SnapToTarget" true)
      (goto _label))
    (:4
      (Transformation|AttachActorToComponent _fortpickup _skeletalmesh "Weapon5" "SnapToTarget" "SnapToTarget" "SnapToTarget" true)
      (goto _label)))
  (Transformation|AttachActorToComponent _fortpickup _skeletalmesh "Weapon1" "SnapToTarget" "SnapToTarget" "SnapToTarget" true)
  (goto _label))

(fn OnRep_CorrectiveImpulse)

(fn PlayHapticsOnHit
  (params (HitMagnitude :type Float%20%28double-precision%29))
  (bind _passengerplayercontroller (Variables|Default|GetPassengerPlayerController))
  (bind _drivingplayercontroller (Variables|Default|GetDrivingPlayerController))
  (bind _validplayercontrollers (LocalVariables|GetValidPlayerControllers))
  (Utilities|FlowControl|Sequence
    (:then_0
      (Utilities|IsValid _drivingplayercontroller
        (:Is%20Valid
          (Utilities|Array|Add (LocalVariables|GetValidPlayerControllers) _drivingplayercontroller))))
    (:then_1
      (Utilities|IsValid _passengerplayercontroller
        (:Is%20Valid
          (Utilities|Array|Add (LocalVariables|GetValidPlayerControllers) _passengerplayercontroller))))
    (:then_2
      (if (> (Utilities|Array|Length (LocalVariables|GetValidPlayerControllers)) 0)
        (if (> (Variables|GetHitMagnitude) 1e+04)
          (Utilities|FlowControl|Sequence
            (:then_0
              (LocalVariables|SetCameraShakeMagnitude (Math|Float|MapRangeClamped (Variables|GetHitMagnitude) 1.5e+04 5e+04 0.1 1.0))
              (Utilities|IsValid (Variables|Default|GetDrivingPlayerController)
                (:Is%20Valid
                  (Game|Feedback|ClientStartCameraShake :self (LocalVariables|GetValidPlayerControllers) :Shake "/Game/Athena/DrivableVehicles/CameraShake_HitImpact_ShoppingCart.CameraShake_HitImpact_ShoppingCart_C" :Scale (LocalVariables|GetCameraShakeMagnitude) :PlaySpace "CameraLocal")
                  (if (> (Variables|GetHitMagnitude) 2e+04)
                    (Game|Feedback|ClientPlayForceFeedback _validplayercontrollers "/Game/Weapons/ForceFeedback/FF_Generic_Impact_Near.FF_Generic_Impact_Near" "None" false false false)
                    (else
                      (Game|Feedback|ClientPlayForceFeedback _validplayercontrollers "/Game/Weapons/ForceFeedback/FF_Generic_Build.FF_Generic_Build" "None" false false false))))))
            (:then_1
              (if (< (Math|Float|Absolute(Float) (Math|Vector|DotProduct (Math|Vector|Normalize (Vehicle|GetPreviousLinearVelocity)) "0, 0,1.000000")) 0.5)
                (bind _output_get (LocalVariables|SetHitDelta (- (Variables|Vehicle|GetCachedSpeed) (Math|Vector|VectorLength (Transformation|GetComponentVelocity (Variables|Vehicle|GetSkeletalMesh))))))
                (if (> _output_get (LocalVariables|GetCameraPushVelocityDelta))
                  (Game|Feedback|ClientStartCameraShake (LocalVariables|GetValidPlayerControllers) "/Game/Athena/DrivableVehicles/CameraShake_BigHit.CameraShake_BigHit_C" (Math|Float|MapRangeClamped (LocalVariables|GetHitDelta) 3e+02 9e+02 0.5 1.0) "UserDefined" (Math|Rotator|MakeRotfromX (Vehicle|GetPreviousLinearVelocity))))))))))))

(fn RemoveCooldownCueEvent
  (params (Player%20Pawn :type Fort%20Player%20Pawn%20Object%20Reference))
  (bind _returnvalue (Ability|GetAbilitySystemComponent Player%20Pawn))
  (Ability|SendGameplayEventtoActor Player%20Pawn (LocalVariables|GetGamePlayEvent) (Utilities|Struct|MakeGameplayEventData :ContextHandle "()" :InstigatorTags "(GameplayTags=)" :TargetTags "(GameplayTags=)" :EventMagnitude 0.0 :TargetData "()"))
  (if (not (Class|ItemUIFunctionLibrary|IsNullItem _returnvalue))
    (GameplayEffects|RemoveActiveGameplayEffect _returnvalue (Variables|Default|GetCooldownCueEvent) -1)))

(fn SendBoostMeterToMaterial
  (Rendering|Material|SetScalarParameterValue (Variables|Default|GetBoostMeterMID_0) "BoostAmount" (CallFunction|GetBoostAccumulated) :!class /Script/Engine.MaterialInstanceDynamic))

(fn SendBoostTimeToMaterial
  (params (ParameterValue :type Float%20%28double-precision%29))
  (bind _output_get (Variables|Default|SetSavedBoostTimeForMID ParameterValue))
  (Rendering|Material|SetScalarParameterValueonMaterials (Variables|Vehicle|GetSkeletalMesh) "BoostTime" _output_get))

(fn SendDamageStateToMaterial
  (params (ParameterValue :type Float%20%28double-precision%29))
  (bind _output_get (Variables|Default|SetSavedDamageForMID ParameterValue))
  (Rendering|Material|SetScalarParameterValueonMaterials (Variables|Vehicle|GetSkeletalMesh) "Damage" _output_get))

(fn SendHeadlightStateToMaterial
  (params (ParameterValue :type Float%20%28double-precision%29))
  (bind _output_get (Variables|Default|SetSavedHeadlightStateForMID ParameterValue))
  (Rendering|Material|SetScalarParameterValueonMaterials (Variables|Vehicle|GetSkeletalMesh) "HeadLights" _output_get))

(fn ConstructionScript
  (Variables|FortAthenaVehicle|SetOverlapComponent (Variables|Default|GetOverlapVolume))
  (bind _returnvalue (Rendering|Material|CreateDynamicMaterialInstance (Variables|Default|GetBoostMeter) 0 0 "None" :!class /Script/Engine.PrimitiveComponent))
  (Variables|FortAthenaAntelopeVehicle|SetBoostMeterMID _returnvalue)
  (bind _returnvalue_2 (Rendering|Material|CreateDynamicMaterialInstance (Variables|Default|GetSM_WheelBlur_FL) 0 "/Game/Athena/DrivableVehicles/FX/Materials/M_Quad_Tire_Glow_Inst.M_Quad_Tire_Glow_Inst" "None" :!class /Script/Engine.PrimitiveComponent))
  (bind _output_get (Variables|FortAthenaAntelopeVehicle|SetWheelBlursFront _returnvalue_2))
  (Rendering|Material|SetMaterial (Variables|Default|GetSM_WheelBlur_FR) 0 _output_get)
  (bind _returnvalue_3 (Rendering|Material|CreateDynamicMaterialInstance (Variables|Default|GetSM_WheelBlur_BL) 0 "/Game/Athena/DrivableVehicles/FX/Materials/M_Quad_Tire_Glow_Inst.M_Quad_Tire_Glow_Inst" "None" :!class /Script/Engine.PrimitiveComponent))
  (bind _output_get_2 (Variables|FortAthenaAntelopeVehicle|SetWheelBlursBack _returnvalue_3))
  (Rendering|Material|SetMaterial (Variables|Default|GetSM_WheelBlur_BR) 0 _output_get_2)
  (Rendering|SetTranslucentSortPriority (Variables|Default|GetSM_WheelBlur_FL) -1)
  (Rendering|SetTranslucentSortPriority (Variables|Default|GetSM_WheelBlur_FR) -1)
  (Rendering|SetTranslucentSortPriority (Variables|Default|GetSM_WheelBlur_BL) -1)
  (Rendering|SetTranslucentSortPriority (Variables|Default|GetSM_WheelBlur_BR) -1))
