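; Athena vehicle spawner Blueprint that spawns a drivable vehicle on the server when Athena vehicles are enabled.
; It prefers the game mode's class override over its own class, then skips spawning during the "NoVehicle" calendar event.
; The spawn transform can be yawed 90 degrees for Y-forward setups, and the spawner can destroy itself afterwards.
; On construction, it sets the vehicle's always-spawn flag and optionally tries applying mods.
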(event EventSpawnVehicle
  (bind _returnvalue (CallFunction|GetVehicleClass))
  (bind _asactor (Utilities|Casting|CastToActorClass _returnvalue))
  ;[1 1 1] Spawn shopping cart vehicle. Wait until OnMapInfoInitialized event, in case there are any overrides set by active events
  (|CallParentFunction)
  (Networking|SwitchHasAuthority
    (:Authority
      (if (Fortnite|AreAthenaVehiclesEnabled)
        (bind _asfort_game_mode_athena (Utilities|Casting|CastToFortGameModeAthena (Game|GetGameMode))
          (:then
            (if (Utilities|IsValidClass (Class|FortGameModeAthena|GetVehicleClassOverride _asfort_game_mode_athena _asactor))
              _)
            (bind _data (Utilities|FlowControl|MergeExecutionwithData (Class|FortGameModeAthena|GetVehicleClassOverride _asfort_game_mode_athena _asactor) _asactor)
              (:Then
                (bind _biseventactive (CalendarEvents|GetCalendarEventInformation "NoVehicle"))
                (if (not _biseventactive)
                  (bind _returnvalue_2 (CallFunction|SpawnVehiclewithConstruction _data (select (Variables|Default|GetYAxisForward) (Math|Transform|ComposeTransforms (struct :Location "0, 0, 0" :Rotation "0,90.000000, 0" :Scale "1.000000,1.000000,1.000000") (Transformation|GetActorTransform)) (Transformation|GetActorTransform))))
                  (Variables|Default|SetSpawnedVehicle _returnvalue_2)
                  (if (Variables|Default|GetbShouldDestroyAfterVehicleSpawn)
                    (Actor|DestroyActor))))))
          (:CastFailed
            _))
        (Actor|DestroyActor)
        (else
          _)))
    (:Remote
      _)))

(event EventOnConstructVehicle
  (params (Vehicle :type Fort%20Athena%20Vehicle%20Object%20Reference))
  ;[1 1 1] Spawn shopping cart vehicle. Wait until OnMapInfoInitialized event, in case there are any overrides set by active events
  (|CallParentFunction Vehicle)
  (Spawn|SetForcedtoAlwaysSpawn Vehicle (CallFunction|GetForceSpawnAlways))
  (if (Variables|FortAthenaVehicleSpawner|GetbCanSpawnWithMod)
    (Vehicle|TryConstructwithModServer Vehicle (Variables|Default|GetForceMods))))

(fn ConstructionScript)
