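# Golf cart base material variant for FX, two-sided and enabled for skeletal meshes, mesh particles and Niagara mesh particles.
# Base color is a tinted texture, alongside a normal map and a Metallic parameter.
# A UseCracks static switch blends a tiled crack mask into color and roughness.
# Offset camera distance, divided by a cull distance and scaled, drives a temporal AA dither threshold for distance-based dissolve.
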Material /Game/Athena/DrivableVehicles/FX/Materials/Golf_Cart_Base__FX1
graph {
 properties {
  bUsedWithMeshDeformer=true
  bUsedWithMeshParticles=true
  bUsedWithNiagaraMeshParticles=true
  bUsedWithSkeletalMesh=true
  pixelDepthOffsetMode=PDOM_Legacy
  twoSided=true
  usedShadingModels=Default Lit
 }
 nodes [
  n1 = Add(A=Distance(A=WorldPosition().XYZ, B=CameraPositionWS()), B=ScalarParameter() {defaultValue=-20, parameterName=s})
  n2 = ScalarParameter() {defaultValue=40, parameterName=cull distance}
  MaterialFunctionCall(Alpha Threshold=Multiply(A=Divide(A=@n1, B=@n2), B=ScalarParameter() {defaultValue=1, parameterName=ScaleDither}), Random=Constant()) {materialFunction=&DitherTemporalAA_Fortnite}
  ScalarParameter() {parameterName=Metallic}
  n3 = Multiply(
   A=TextureSampleParameter2D() {parameterName=BaseColor, texture=/Game/Athena/DrivableVehicles/Golf_Cart/Golf_Cart_Base/Demo/Textures/Demo_T_Golf_Cart_Base_D}.RGB
   B=VectorParameter() {parameterName=BaseColorMult, defaultValue=<rgba 0.6146 0.5762 0.5762 0>}.RGB
  )
  n4 = TextureSample(UVs=TextureCoordinate() {uTiling=2, vTiling=2}) {texture=/Game/Global/Textures/Default/Masks_and_Gradients/Cracks_02}
  StaticSwitchParameter(True=LinearInterpolate(A=@n3, B=Constant() {r=0.1}, Alpha=@n4.R), False=@n3) {defaultValue=true, parameterName=UseCracks}
  n5 = ScalarParameter() {defaultValue=0.8, parameterName=CrackRoughness}
  StaticSwitchParameter(True=LinearInterpolate(A=ScalarParameter() {defaultValue=0.5, parameterName=Roughness}, B=@n5, Alpha=@n4.G) {constA=1}, False=@n5) {defaultValue=true, parameterName=UseCracks}
  TextureSampleParameter2D() {
   parameterName=Normal
   samplerType=SAMPLERTYPE_Normal
   texture=/Game/Athena/DrivableVehicles/Golf_Cart/Golf_Cart_Base/Demo/Textures/Demo_T_Golf_Cart_Base_N
  }
 ]
}
