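# Masked, lit FX material for the golf cart flag, usable on skeletal meshes and mesh particles, with BaseColor and Normal texture parameters.
# Dynamic parameters drive a vertex bend about a horizontal axis at the object origin, optionally weighted by height, with matching normal correction.
# The bend feeds SimpleGrassWind, weighted by green vertex color, for flutter.
# World-space XY noise plus the particle Alpha feeds temporal dithering for dissolve.
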Material /Game/Athena/DrivableVehicles/FX/Materials/Golf_Cart_Flag_FX
graph {
 properties {
  bUsedWithMeshDeformer=true
  bUsedWithMeshParticles=true
  bUsedWithNiagaraMeshParticles=true
  bUsedWithSkeletalMesh=true
  blendMode=BLEND_Masked
  pixelDepthOffsetMode=PDOM_Legacy
  usedShadingModels=Default Lit
 }
 nodes [
  Constant() {r=0.5}
  Constant()
  n1 = DynamicParameter() {defaultValue=<rgba 0 1 1 1>, paramNames [BlendWPO,SoftBlend,AxisAngle,Alpha]}
  n2 = AppendVector(A=AppendVector(A=Sine(None=@n1.AxisAngle), B=Cosine(None=@n1.AxisAngle)), B=Constant())
  n3 = WorldPosition()
  n4 = TransformPosition(None=Constant3Vector()) {transformType=TRANSFORMPOSSOURCE_World}
  n5 = ComponentMask(None=Divide(A=Subtract(A=@n3.XYZ, B=@n4)) {constB=500}) {b=true}
  n6 = Multiply(A=LinearInterpolate(A=Constant() {r=1}, B=Clamp(None=@n5), Alpha=@n1.SoftBlend), B=@n1.BlendWPO)
  n7 = TextureSampleParameter2D() {
   parameterName=Normal
   samplerType=SAMPLERTYPE_Normal
   texture=/Game/Athena/DrivableVehicles/Golf_Cart/Golf_Cart_Kit_A/Demo/Textures/Demo_T_Golf_Cart_Kit_A_N
  }
  MaterialFunctionCall(Rotation Axis=@n2, Rotation Angle=@n6, World Space Vertex Normals=Transform(None=@n7.RGB)) {materialFunction=&FixRotateAboutAxisNormals}
  n8 = WorldPosition()
  n9 = TextureSample(UVs=Divide(A=ComponentMask(None=@n8.XYZ) {g=true, r=true}) {constB=128}) {samplerType=SAMPLERTYPE_LinearColor, texture=&HighContrastNoise}
  Reroute(None=MaterialFunctionCall(Alpha Threshold=Add(A=@n9.R, B=@n1.Alpha)) {materialFunction=&DitherTemporalAA_Fortnite}.Result)
  n10 = OneMinus(None=VertexColor().G)
  n11 = Multiply(A=@n10, B=@n10)
  n12 = WorldPosition()
  ShadingPathSwitch(Default=MaterialFunctionCall!(
    WindIntensity=Constant() {r=-1}
    WindWeight=OneMinus(None=@n11)
    WindSpeed=Constant() {r=2}
    AdditionalWPO=RotateAboutAxis(NormalizedRotationAxis=@n2, RotationAngle=@n6, PivotPoint=@n4, Position=@n12.XYZ)
   ) {materialFunction=/Engine/Functions/Engine_MaterialFunctions01/WorldPositionOffset/SimpleGrassWind}.Result) {desc=Migrated from FeatureLevelSwitch}
  TextureSampleParameter2D() {parameterName=BaseColor, texture=/Game/Athena/DrivableVehicles/Golf_Cart/Golf_Cart_Kit_A/Demo/Textures/Demo_T_Golf_Cart_Kit_A_D}
 ]
}
