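# Translucent unlit mesh-particle material for engine pilot ring effects on drivable vehicles.
# Two panned smooth-noise samples, offset in time by vertex red and shaped by UV gradients, form the streaks and dark-to-light color blend.
# Dynamic Parameter 1 drives a stepped, smoothed reveal across ring levels encoded in vertex red.
# It also moves a highlight band that boosts brightness and scales a transformed, V-faded offset vector.
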Material /Game/Athena/DrivableVehicles/FX/Materials/M_Engine_Pilot_Rings
graph {
 properties {bUsedWithMeshParticles=true, blendMode=BLEND_Translucent, pixelDepthOffsetMode=PDOM_Legacy, shadingModel=MSM_Unlit, usedShadingModels=Unlit}
 nodes [
  n1 = ParticleColor()
  n2 = Multiply(A=Add(A=Time(), B=VertexColor().R), B=ScalarParameter() {defaultValue=1.5, parameterName=PanSpeed})
  n3 = Panner(Coordinate=TextureCoordinate() {uTiling=6, vTiling=0.1}, Time=@n2) {speedX=-0.2, speedY=1}
  n4 = Panner(Coordinate=TextureCoordinate() {uTiling=3, vTiling=0.2}, Time=@n2) {speedX=0.2, speedY=1.2}
  n5 = ComponentMask(None=TextureCoordinate()) {g=true}
  n6 = Power(Base=@n5) {constExponent=10}
  n7 = Add(
   A=LinearInterpolate(
    A=TextureSample(UVs=@n3) {texture=&T_Smooth_tile}.G
    B=TextureSample(UVs=@n4) {texture=&T_Smooth_tile}.G
   ) {constAlpha=0.35}
   B=Clamp(None=Multiply(A=@n6) {constB=0.7})
  )
  n8 = Power(Base=@n5)
  n9 = Multiply(A=@n7, B=@n8)
  n10 = SphereMask(A=@n5, B=Constant()) {attenuationRadius=1, hardnessPercent=84}
  n11 = VertexColor()
  n12 = ScalarParameter() {defaultValue=3, parameterName=RingLevels}
  n13 = Multiply(A=DynamicParameter().Param1, B=ScalarParameter() {defaultValue=1.2, parameterName=ParamMult})
  n14 = Multiply(A=@n13, B=@n12)
  n15 = Floor(None=@n14)
  n16 = MaterialFunctionCall(
   Alpha=Frac(None=@n14)
   Min=ScalarParameter() {defaultValue=0.7, parameterName=SmoothStepMin}
   Max=ScalarParameter() {defaultValue=0.9, parameterName=SmoothStepMax}
  ) {materialFunction=&SmoothStep}
  n17 = Clamp(None=Subtract(A=Multiply(A=@n11.R, B=@n12), B=Add(A=@n15, B=@n16.Result)))
  Multiply(A=@n1.A, B=Multiply(A=Multiply(A=@n9, B=@n10), B=OneMinus(None=@n17)))
  n18 = VectorParameter() {parameterName=DarkColor, defaultValue=<rgba 1.5 0.7159 0.2727 0>}
  n19 = VectorParameter() {parameterName=LightColor, defaultValue=<rgba 0.3186 1.0503 2 0>}
  n20 = Distance(A=@n11.R, B=Add(A=@n13, B=ScalarParameter()))
  n21 = ScalarParameter() {defaultValue=8}
  n22 = OneMinus(None=Clamp(None=Multiply(A=@n20, B=@n21)))
  n23 = Saturate(None=Subtract(A=Add(A=@n7, B=@n8), B=@n22))
  n24 = ScalarParameter() {defaultValue=5}
  Multiply(A=Multiply(A=@n1.RGB, B=LinearInterpolate(A=@n18.RGB, B=@n19.RGB, Alpha=@n23)), B=Add(A=Multiply(A=@n22, B=@n24)))
  n25 = VectorParameter() {defaultValue=<rgba 0 1 0 1>}
  n26 = ScalarParameter() {defaultValue=-3}
  n27 = TextureCoordinate()
  Multiply(A=Multiply(A=Transform(None=@n25.RGB), B=Multiply(A=@n22, B=@n26)), B=OneMinus(None=ComponentMask(None=@n27) {g=true}))
  ScalarParameter() {defaultValue=1.2, parameterName=DebugPhase}
 ]
}
