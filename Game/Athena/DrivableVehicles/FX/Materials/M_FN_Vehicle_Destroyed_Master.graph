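# Masked, lit master material for destroyed drivable vehicles, rendering a dented, rusted wreck that can fade away.
# A tiled damage AO texture blends metal into desaturated rust across color, metallic and roughness, and drives world position offset, optionally along vertex normals.
# The same texture, weighted by bounds-normalized height, combines with Fade into a dithered dissolve mask.
# The attributes pass through MF_FN_DestructibleDamage, fed destruction amount via custom primitive data.
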Material /Game/Athena/DrivableVehicles/FX/Materials/M_FN_Vehicle_Destroyed_Master
graph {
 properties {
  bUseMaterialAttributes=true
  bUsedWithMeshDeformer=true
  bUsedWithSkeletalMesh=true
  blendMode=BLEND_Masked
  ditherOpacityMask=true
  pixelDepthOffsetMode=PDOM_Legacy
  usedShadingModels=Default Lit
  lightmassSettings {
   bCastShadowAsMasked=true
   bOverrideCastShadowAsMasked=false
   bOverrideDiffuseBoost=false
   bOverrideEmissiveBoost=false
   bOverrideExportResolutionScale=false
   diffuseBoost=1
   emissiveBoost=1
   exportResolutionScale=1
  }
 }
 nodes [
  n1 = Constant3Vector() {constant=<rgba 0.2264 0.25 0.2475 0>}
  n2 = Multiply(A=TextureCoordinate(), B=ScalarParameter() {defaultValue=1.5, parameterName=Damage Tiling})
  n3 = Desaturation(
   None=TextureSample(UVs=@n2) {texture=/Game/Environments/Apollo/Sets/Yacht/Textures/Yacht3/T_Yacht3_Rusty_BelowDeckWalls_D}.RGB
   Fraction=Constant() {r=0.5}
  )
  n4 = TextureSample(UVs=@n2) {samplerType=SAMPLERTYPE_LinearGrayscale, texture=/Game/Vehicles/Valet/ValetDamage_AO}
  n5 = Saturate(None=LinearInterpolate(Alpha=OneMinus(None=@n4.R)) {constA=-0.3, constB=4})
  n6 = Multiply(A=VertexNormalWS(), B=Subtract(A=ScalarParameter() {parameterName=WPOBias}, B=@n4.R))
  n7 = ObjectPositionWS()
  n8 = Divide(
   A=Subtract(
    A=MaterialFunctionCall() {materialFunction=&LocalPosition}.Local Position
    B=TransformPosition(None=@n7) {transformSourceType=TRANSFORMPOSSOURCE_World}
   )
   B=PreSkinnedLocalBounds().Extents
  )
  n9 = OneMinus(None=ComponentMask(None=Add(A=@n8) {constB=0.5}) {b=true})
  n10 = Add(A=Subtract(A=Multiply(A=@n4.R, B=@n9)), B=LinearInterpolate(Alpha=ScalarParameter() {defaultValue=1, parameterName=Fade}) {constB=1.05})
  n11 = SetMaterialAttributes!(
   MaterialAttributes
   Base Color=LinearInterpolate(A=Constant3Vector() {constant=<rgba 0.1979 0.1979 0.1979 1>}, B=Multiply(A=@n1, B=@n3), Alpha=@n5)
   Metallic=OneMinus(None=@n5)
   Specular
   Roughness=LinearInterpolate(A=Constant() {r=0.4}, Alpha=@n5) {constB=0.85}
   Normal=TextureSample(UVs=@n2) {samplerType=SAMPLERTYPE_Normal, texture=/Game/Vehicles/Valet/ValetDamage_N}.RGB
   World Position Offset=Multiply(
    A=StaticSwitchParameter(True=@n6, False=@n4.R) {parameterName=UseVertexNormalWPO}
    B=ScalarParameter() {defaultValue=16, parameterName=DamageDisplacement}
   ) {constB=16}
   Opacity Mask=Clamp(None=Ceil(None=@n10))
  ) {attributeSetTypes [
    #1
    #2
    #3
    #4
    #5
    #6
    #7
   ]}
  MaterialFunctionCall(
   Material_In=@n11
   UseCrackedDamageTextures=StaticBool()
   UseAdditiveWPO=StaticBool() {value=true}
   DestructibleDamageAmount=ScalarParameter() {bUseCustomPrimitiveData=true, group=z_code, parameterName=Code_DestructionAmount, primitiveDataIndex=13}
  ) {materialFunction=&MF_FN_DestructibleDamage}
 ]
}
