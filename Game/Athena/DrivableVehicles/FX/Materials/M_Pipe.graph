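# Default Lit material for pipe mesh particles in the drivable vehicle FX, previewed on a golf cart cylinder mesh.
# It samples a small noise texture using the particle's Random dynamic parameter as U and slightly tiled mesh V as V.
# The WPO dynamic parameter scales that noise, apparently for world position offset, so each particle can get its own distortion along its length.
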Material /Game/Athena/DrivableVehicles/FX/Materials/M_Pipe
graph {
 properties {
  bUsedWithMeshParticles=true
  bUsedWithNiagaraMeshParticles=true
  pixelDepthOffsetMode=PDOM_Legacy
  previewMesh=/Game/Athena/DrivableVehicles/Golf_Cart/Golf_Cart_Base/Mesh/FX/S_CylinderWPOpipe
  usedShadingModels=Default Lit
 }
 nodes [
  Constant()
  Constant() {r=0.4}
  Constant() {r=1}
  n1 = DynamicParameter() {paramNames [Param1,Param2,WPO,Random]}
  n2 = ComponentMask(None=TextureCoordinate() {uTiling=0, vTiling=0.1}) {g=true}
  Multiply(
   A=TextureSample(UVs=AppendVector(A=@n1.Random, B=@n2)) {samplerType=SAMPLERTYPE_Normal, texture=/Game/Global/Textures/Default/Noise/64x64Noise}.RGB
   B=@n1.WPO
  ) {constB=2}
  TextureCoordinate()
 ]
}
