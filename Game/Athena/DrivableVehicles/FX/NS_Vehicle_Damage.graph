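# Vehicle damage Niagara effect that builds engine smoke, flares, sparks, embers and fire as DamageScale crosses three user-set stage thresholds.
# Smoke layers darken or tint by damage, GPU embers with collision join at Stage2, and fire plus extra smoke mark Stage3Critical.
# Sparks read flare particles through a particle attribute reader to spawn at their positions.
# Vehicle_Speed scales spark rate and adjusts fire, smoke and critical-stage ember behavior.
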NiagaraSystem /Game/Athena/DrivableVehicles/FX/NS_Vehicle_Damage
dependencies {
 usedDataInterfaces [
  &NiagaraDataInterfaceCurve
  &NiagaraDataInterfaceColorCurve
  &NiagaraDataInterfaceVectorCurve
  &NiagaraDataInterfaceParticleRead
 ]
 usedDynamicInputs []
 usedModules [&SystemState]
}
emitters {
 Generator_Flare_Engine001 {
  emitterUpdateScript [
   &EmitterState(
    Life Cycle Mode=::NewEnumerator1
    Loop Duration=&MakeCustomFloatFromBool(
     Bool=&SetBoolByFloatComparison(A=$User.DamageScale, B=0.8)
     True Float=&UniformRangedFloat(Minimum=0.4, Maximum=1.2)
     False Float=&UniformRangedFloat(Minimum=0.8, Maximum=2.6)
    )
    UseLoopDelay=false
   )
   &SpawnBurst_Instantaneous(
    Spawn Count=&MakeCustomIntFromBool(
     Bool=&SetBoolByFloatComparison(A=$User.DamageScale, B=$User.Stage1, Comparison Type=::NewEnumerator2)
     True INT=&MakeCustomIntFromBool(
      Bool=&SetBoolByFloatComparison(A=$User.DamageScale, B=$User.Stage3Critical, Comparison Type=::NewEnumerator2)
      True INT=0
      False INT=&UniformRangedInt(Minimum=0, Maximum=1)
     )
     False INT=0
    )
    Use Spawn Probability=false
   )
   &SpawnBurst_Instantaneous(
    Spawn Count=&MakeCustomIntFromBool(
     Bool=&SetBoolByFloatComparison(A=$User.DamageScale, B=$User.Stage2, Comparison Type=::NewEnumerator2)
     True INT=&MakeCustomIntFromBool(
      Bool=&SetBoolByFloatComparison(A=$User.DamageScale, B=$User.Stage3Critical, Comparison Type=::NewEnumerator2)
      True INT=0
      False INT=&UniformRangedInt(Minimum=0, Maximum=1)
     )
     False INT=0
    )
    Spawn Time=0.88
    Use Spawn Probability=false
   ) {module=SpawnBurst_Instantaneous001}
   &SpawnBurst_Instantaneous(
    Spawn Count=&MakeCustomIntFromBool(
     Bool=&SetBoolByFloatComparison(A=$User.DamageScale, B=$User.Stage2, Comparison Type=::NewEnumerator2)
     True INT=&MakeCustomIntFromBool(
      Bool=&SetBoolByFloatComparison(A=$User.DamageScale, B=$User.Stage3Critical, Comparison Type=::NewEnumerator2)
      True INT=0
      False INT=&UniformRangedInt(Minimum=0, Maximum=1)
     )
     False INT=0
    )
    Spawn Time=1.4
    Use Spawn Probability=false
   ) {module=SpawnBurst_Instantaneous002}
   &SpawnRate(
    SpawnRate=&MakeCustomFloatFromBool(Bool=&SetBoolByFloatComparison(A=$User.DamageScale, B=$User.Stage1), True Float=1.5, False Float=0)
    Use Spawn Probability=false
   )
  ]
  particleSpawnScript [
   /Niagara/Modules/Spawn/Initialization/InitializeParticle(
    Color=<rgba 1 0.35 0.01 0.025>
    Lifetime=0.15
    Module.Write Color=true
    Module.Write Lifetime=true
    Module.Write Mass=true
    Module.Write Material Random=true
    Module.Write Position=false
    Module.Write Scale=false
    Module.Write SpriteRotation=false
    Module.Write SpriteSize=true
    Position=&SimulationPosition()
    Sprite Size=&Vector2DFromFloat(Value=&UniformRangedFloat(Minimum=250, Maximum=150))
   )
   /SetVariables(Particles.LightColor=$Particles.Color) {isSetParameters=true}
   /Niagara/Modules/Spawn/Location/SphereLocation(
    Hemisphere X=true
    MaskSpawnGroup=false
    Module.Override Randomness=false
    Non Uniform Scale=&Lerp_Vector(A=<0 0 0>, B=<1 1 0.1>, Alpha=0.5)
    Surface Only=false
   )
   /SetVariables(Particles.ParticleLightPositionUpdate:Vector3f=&Add_Vector(
     A=Linked() {linkedVariable {name=Particles.Position, type=NiagaraPosition}}
     B=&Multiply_VectorByFloat(Vector=$Engine.Owner.SystemZAxis, Float=25)
    )) {isSetParameters=true}
   &CameraOffset(Camera Offset Amount=50)
   &ApplyInitialForces()
  ]
  particleUpdateScript [
   &ParticleState()
   &SolveForcesAndVelocity(Clamp Velocity=false, Limit Acceleration=false)
   &ScaleColor(
    Scale Alpha=&FloatFromCurve(FloatCurve=DataInterface() {propertyValues {Curve {keys [
         {arriveTangent=6.6661, interpMode=RCIM_Cubic, leaveTangent=6.6661, tangentMode=RCTM_User, time=0, value=0}
         <at 0.3239 0.984 RCIM_Cubic>
         <at 1 0 RCIM_Cubic>
        ]}}})
    Scale RGB=&VectorFromFloat(Value=&FloatFromCurve(FloatCurve=DataInterface() {propertyValues {Curve {keys [
          {arriveTangent=28.8238, interpMode=RCIM_Cubic, leaveTangent=28.8239, tangentMode=RCTM_User, time=0, value=0}
          <at 0.4983 15 RCIM_Cubic>
          <at 1 0 RCIM_Cubic>
         ]}}}))
    ScaleA=true
    ScaleRGB=true
   )
   /SetVariables(Particles.LightIntensityMulti:NiagaraFloat=&FloatFromCurve(FloatCurve=DataInterface() {propertyValues {Curve {keys [<at 0 1>,<at 1 0>]}}}, Scale Curve=35)) {isSetParameters=true}
  ]
  rendererClasses [&NiagaraSpriteRendererProperties]
  renderers [{rendererClass=&NiagaraSpriteRendererProperties}]
 }
 Generator_Sparks_Forward001 {
  emitterSpawnScript [/Game/Effects/Niagara/Modules/Attributes/FN_Source_v2(AttributeReader=DataInterface() {propertyValues {EmitterBinding {bindingMode=Other, emitterName=Generator_Flare_Engine001}}})]
  emitterUpdateScript [
   &EmitterState()
   &SpawnRate(
    SpawnRate=&UniformRangedFloat(
     Minimum=&Multiply_Float(
      A=&MakeCustomFloatFromBool(Bool=&SetBoolByFloatComparison(A=$User.DamageScale, B=$User.Stage1), True Float=100, False Float=0)
      B=&FloatFromCurve(FloatCurve=DataInterface() {propertyValues {Curve {keys [<at 0.1 0>,<at 0.5 1>]}}}, CurveIndex=$User.Vehicle_Speed)
     )
     Maximum=&Multiply_Float(
      A=&MakeCustomFloatFromBool(Bool=&SetBoolByFloatComparison(A=$User.DamageScale, B=$User.Stage1), True Float=200, False Float=0)
      B=&FloatFromCurve(FloatCurve=DataInterface() {propertyValues {Curve {keys [<at 0.1 0>,<at 0.5 1>]}}}, CurveIndex=$User.Vehicle_Speed)
     )
    )
    Use Spawn Probability=false
   )
  ]
  particleSpawnScript [
   /Game/Effects/Niagara/Modules/Attributes/FN_SourceParentID()
   &InitializeParticle(
    Color=&UniformRangedLinearColor(Minimum=<rgba 15 4.482 0 1>, Maximum=<rgba 5 1.1228 0 1>)
    Lifetime=&UniformRangedFloat(Minimum=0.12, Maximum=1)
    Module.Write Color=true
    Module.Write Lifetime=true
    Module.Write Mass=true
    Module.Write Material Random=true
    Module.Write Position=true
    Module.Write Scale=false
    Module.Write SpriteRotation=false
    Module.Write SpriteSize=true
    Position=/Game/Effects/Niagara/DynamicInputs/Attributes/FN_GetVectorByID(Attrib name:ENiagaraCommonVectors=::NewEnumerator0, Default on Invalid001:Vector3f=<0 0 0>)
    Sprite Size=&UniformRangedVector2D(Minimum=<3 3>, Maximum=<4 7>)
   )
   &SphereLocation(
    Hemisphere X=true
    MaskSpawnGroup=false
    Module.Override Randomness=false
    Non Uniform Scale=&Lerp_Vector(A=<0 0 0>, B=<1 1 1>, Alpha=0.5)
    Surface Only=false
   ) {module=SphereLocation001}
   &AddVelocity(Velocity=&UniformRangedVector(Minimum=<-100 -300 75>, Maximum=<-250 300 800>)) {module=AddVelocity002}
   &AddVelocity(Velocity=&MakeCustomVectorFromBool(
     Bool=&SetBoolByFloatComparison(A=$User.DamageScale, B=$User.Stage2, Comparison Type=::NewEnumerator2)
     True Vector=&Multiply_Vector(
      A=$Output.SphereLocation.SphereNormal
      B=&VectorFromFloat(Value=&UniformRangedFloat(Minimum=100, Maximum=200))
     )
     False Vector=&Multiply_Vector(
      A=$Output.SphereLocation.SphereNormal
      B=&VectorFromFloat(Value=&UniformRangedFloat(Minimum=50, Maximum=100))
     )
    )) {module=AddVelocity003}
   &SphereLocation(
    Hemisphere X=true
    Hemisphere Z=true
    MaskSpawnGroup=false
    Module.Override Randomness=false
    Sphere Radius=&UniformRangedFloat(Minimum=5, Maximum=15)
    Surface Only=false
   )
  ]
  particleUpdateScript [
   &ParticleState()
   &AccelerationForce(Acceleration=<0 0 -980>)
   &SolveForcesAndVelocity(Clamp Velocity=false, Limit Acceleration=false)
   &ScaleColor(
    Scale Alpha=&FloatFromCurve(FloatCurve=DataInterface() {propertyValues {Curve {keys [{arriveTangent=-6.5058, interpMode=RCIM_Cubic, leaveTangent=-6.5058, tangentMode=RCTM_User, time=0, value=2}, <at 0.2 1 RCIM_Cubic>]}}})
    Scale RGB=&VectorFromCurve(VectorCurve=DataInterface() {propertyValues {
       XCurve {keys [<at 0 16 RCIM_Cubic>,<at 0.8 1 RCIM_Cubic>]}
       YCurve {keys [<at 0 16 RCIM_Cubic>,<at 0.8 0.1249 RCIM_Cubic>]}
       ZCurve {keys [<at 0 16 RCIM_Cubic>,<at 0.8 0.2547 RCIM_Cubic>]}
      }})
    ScaleA=true
    ScaleRGB=true
   )
   &ScaleSpriteSize(Scale Factor=&Vector2DFromFloat(Value=&FloatFromCurve(FloatCurve=DataInterface() {propertyValues {Curve {keys [
          {arriveTangent=7.6285, interpMode=RCIM_Cubic, leaveTangent=7.6285, tangentMode=RCTM_User, time=0, value=0}
          <at 0.2 1 RCIM_Cubic>
          <at 1 0 RCIM_Cubic>
         ]}}})))
   &ScaleSpriteSizeBySpeed(Max Scale Factor=<1 7>, Min Scale Factor=<0.5 0.5>)
  ]
  rendererClasses [&NiagaraSpriteRendererProperties]
  renderers [{rendererClass=&NiagaraSpriteRendererProperties}]
 }
 Local_Embers_Forward001 {
  simTarget=GPUComputeSim
  emitterUpdateScript [
   &EmitterState()
   &SpawnRate(
    SpawnRate=&Multiply_Float(
     A=&MakeCustomFloatFromBool(
      Bool=&SetBoolByFloatComparison(A=$User.DamageScale, B=$User.Stage2, Comparison Type=::NewEnumerator2)
      True Float=&MakeCustomFloatFromBool(
       Bool=&SetBoolByFloatComparison(A=$User.DamageScale, B=$User.Stage3Critical, Comparison Type=::NewEnumerator2)
       True Float=&Lerp_Float(A=150, B=250, Alpha=$User.Vehicle_Speed)
       False Float=250
      )
      False Float=0
     )
     B=&FloatFromCurve(
      FloatCurve=DataInterface() {propertyValues {Curve {keys [
          {interpMode=RCIM_Cubic, tangentMode=RCTM_User, time=0, value=1}
          {arriveTangent=-1.5312, interpMode=RCIM_Cubic, leaveTangent=-1.5312, tangentMode=RCTM_User, time=0.2159, value=0}
          {arriveTangent=8.1348, interpMode=RCIM_Cubic, leaveTangent=8.1348, tangentMode=RCTM_User, time=0.2491, value=0.6186}
          {arriveTangent=-1.4415, interpMode=RCIM_Cubic, leaveTangent=-1.4415, tangentMode=RCTM_User, time=0.3217, value=0.8606}
          <at 0.3597 0>
          {arriveTangent=0.132, interpMode=RCIM_Cubic, leaveTangent=0.132, tangentMode=RCTM_User, time=0.5004, value=0.2712}
          {arriveTangent=-0.2914, interpMode=RCIM_Cubic, leaveTangent=-0.2914, tangentMode=RCTM_User, time=0.6515, value=0.0385}
          {arriveTangent=-0.1105, interpMode=RCIM_Cubic, leaveTangent=-0.1105, tangentMode=RCTM_User, time=0.6954, value=0.7517}
          {interpMode=RCIM_Cubic, tangentMode=RCTM_User, time=1, value=0}
         ]}}}
      CurveIndex=$Emitter.NormalizedLoopAge
     )
    )
    Use Spawn Probability=false
   ) {module=SpawnRate001}
  ]
  particleSpawnScript [
   &InitializeParticle(
    Color=&UniformRangedLinearColor(Minimum=<rgba 5 1.494 0 1>, Maximum=<rgba 1 0.169 0 1>)
    Lifetime=&UniformRangedFloat(Minimum=1, Maximum=2)
    Module.Write Color=true
    Module.Write Lifetime=true
    Module.Write Mass=true
    Module.Write Material Random=true
    Module.Write Position=false
    Module.Write Scale=false
    Module.Write SpriteRotation=false
    Module.Write SpriteSize=true
    Position=&SimulationPosition()
    Sprite Size=&UniformRangedVector2D(Minimum=<0.5 1>, Maximum=<2 3>)
   ) {module=InitializeParticle001}
   &SphereLocation(
    Hemisphere X=true
    MaskSpawnGroup=false
    Module.Override Randomness=false
    Non Uniform Scale=&Lerp_Vector(A=<0.25 0.25 0.1>, B=<1 1 0.2>, Alpha=0.5)
    Offset=<-50 0 0>
    Sphere Radius=150
    Surface Only=false
   )
   &AddVelocity(Velocity=&MakeCustomVectorFromBool(
     Bool=&SetBoolByFloatComparison(A=$User.DamageScale, B=$User.Stage2, Comparison Type=::NewEnumerator2)
     True Vector=&Multiply_Vector(
      A=$Output.SphereLocation.SphereNormal
      B=&VectorFromFloat(Value=&UniformRangedFloat(Minimum=25, Maximum=50))
     )
     False Vector=&Multiply_Vector(
      A=$Output.SphereLocation.SphereNormal
      B=&VectorFromFloat(Value=&UniformRangedFloat(Minimum=25, Maximum=50))
     )
    )) {module=AddVelocity002}
   &AddVelocity(Velocity=&UniformRangedVector(Minimum=<-50 -50 25>, Maximum=<-100 50 100>)) {module=AddVelocity001}
   &CameraOffset(Camera Offset Amount=&UniformRangedFloat(Minimum=10, Maximum=25))
  ]
  particleUpdateScript [
   &ParticleState()
   &AccelerationForce(Acceleration=<0 0 -980>)
   &Collision(
    Advanced Aging Rate=1
    Age Colliding Particles=false
    Friction=0.1
    Kill Offscreen Particles=true
    Module.Control Rest State Via Surface Normal=false
    Module.EnableMaxCollisionCount=false
    Module.Randomize Collision Normal Vector=false
    Restitution=0.5
   )
   /Niagara/Modules/Update/Forces/CurlNoiseForce(Noise Frequency=0.005, Noise Strength=250, Pan Noise=true, Pan Noise Field=<0 0 1>)
   &CurlNoiseForce(
    Noise Frequency=0.01
    Noise Strength=&FloatFromCurve(FloatCurve=DataInterface() {propertyValues {Curve {keys [<at 0 1>,<at 0.1954 0.0398>]}}}, Scale Curve=2500)
    Pan Noise=true
    Pan Noise Field=<0 0 1>
   ) {module=CurlNoiseForce001}
   &SolveForcesAndVelocity(Clamp Velocity=false, Limit Acceleration=false)
   &ScaleSpriteSize(Scale Factor=&Vector2DFromFloat(Value=&FloatFromCurve(FloatCurve=DataInterface() {propertyValues {Curve {keys [
          {arriveTangent=56.1861, interpMode=RCIM_Cubic, leaveTangent=56.1859, tangentMode=RCTM_User, time=0, value=0}
          {arriveTangent=0.8166, interpMode=RCIM_Cubic, leaveTangent=0.8166, time=0.03, value=1}
          {arriveTangent=-1.7925, interpMode=RCIM_Cubic, leaveTangent=-1.7925, tangentMode=RCTM_User, time=0.7111, value=0.5807}
          <at 1 0 RCIM_Cubic>
         ]}}})))
   &ScaleSpriteSizeBySpeed(Max Scale Factor=<1 10>)
   &ScaleColor(
    Scale Alpha=&FloatFromCurve(FloatCurve=DataInterface() {propertyValues {Curve {keys [
         {interpMode=RCIM_Cubic, tangentMode=RCTM_User, time=0.0048, value=0.438}
         {arriveTangent=0.3021, interpMode=RCIM_Cubic, leaveTangent=0.3021, tangentMode=RCTM_User, time=0.0983, value=1.5996}
         {arriveTangent=-1.0514, interpMode=RCIM_Cubic, leaveTangent=-1.0514, tangentMode=RCTM_User, time=0.3368, value=0.5383}
         {arriveTangent=-0.8117, interpMode=RCIM_Cubic, leaveTangent=-0.8117, tangentMode=RCTM_User, time=0.5811, value=1.092}
         <at 1 0>
        ]}}})
    Scale RGB=&VectorFromCurve(VectorCurve=DataInterface() {propertyValues {
       XCurve {keys [<at 0 16 RCIM_Cubic>,<at 0.8 1 RCIM_Cubic>]}
       YCurve {keys [<at 0 16 RCIM_Cubic>,<at 0.8 0.1249 RCIM_Cubic>]}
       ZCurve {keys [<at 0 16 RCIM_Cubic>,<at 0.8 0.2547 RCIM_Cubic>]}
      }})
    ScaleA=true
    ScaleRGB=true
   )
  ]
  rendererClasses [&NiagaraSpriteRendererProperties]
  renderers [{rendererClass=&NiagaraSpriteRendererProperties}]
 }
 Local_Fire_Engine {
  emitterUpdateScript [
   &EmitterState(Life Cycle Mode=::NewEnumerator1, Loop Duration=1, UseLoopDelay=false)
   &SpawnRate(
    SpawnRate=&MakeCustomFloatFromBool(
     Bool=&SetBoolByFloatComparison(A=$User.DamageScale, B=$User.Stage3Critical, Comparison Type=::NewEnumerator2)
     True Float=25
     False Float=0
    )
    Use Spawn Probability=false
   )
  ]
  particleSpawnScript [
   &InitializeParticle(
    Color=<rgba 2 0.9857 0.2812 1>
    Lifetime=&Lerp_Float(A=&UniformRangedFloat(Minimum=0.4, Maximum=0.65), B=0.325, Alpha=$User.Vehicle_Speed)
    Module.Write Color=true
    Module.Write Lifetime=true
    Module.Write Mass=true
    Module.Write Material Random=true
    Module.Write Position=false
    Module.Write Scale=false
    Module.Write SpriteRotation=true
    Module.Write SpriteSize=true
    Position=&SimulationPosition()
    Sprite Rotation=&NormalizedAngleToDegrees(NormalizedAngle=&UniformRangedFloat(Minimum=0, Maximum=1))
    Sprite Size=&Vector2DFromFloat(Value=&UniformRangedFloat(Minimum=100, Maximum=150))
   )
   &CylinderLocation(
    Cylinder Height=10
    Cylinder Radius=70
    Lathe Profile=DataInterface() {propertyValues {Curve {keys []}}}
    MaskSpawnGroup=false
    Module.Enable Lathe Profile=false
    Module.Override Randomness=false
    Module.Surface Only=false
    Non Uniform Scale:Vector3f=<1 1 -1>
   )
   &AddVelocity(Velocity=&Multiply_Vector(
     A=$Output.CylinderLocation.CylinderNormal
     B=&VectorFromFloat(Value=&UniformRangedFloat(Minimum=50, Maximum=100))
    ))
   &AddVelocity(
    Scale Added Velocity=&Lerp_Vector(A=<1 1 1>, B=<-3 0 2>, Alpha=$User.Vehicle_Speed)
    Velocity=&UniformRangedVector(Minimum=<-100 0 75>, Maximum=<-250 1 100>)
   ) {module=AddVelocity001}
   &DynamicMaterialParameters(
    Index 0 Param 2=&UniformRangedFloat(Minimum=0.333, Maximum=1)
    Index 0 Param 4=&UniformRangedFloat(Minimum=1.333, Maximum=2)
    Module.Param0WriteEnabled=false
    Module.Param1WriteEnabled=true
    Module.Param2WriteEnabled=false
    Module.Param3WriteEnabled=true
   )
  ]
  particleUpdateScript [
   &ParticleState()
   &Drag(Drag=2, Use Linear Drag=true, Use Rotational Drag=false)
   &AccelerationForce(Acceleration=<0 0 100>)
   &DynamicMaterialParameters(
    Index 0 Param 1=&FloatFromCurve(FloatCurve=DataInterface() {propertyValues {Curve {keys [
         {arriveTangent=4.229, interpMode=RCIM_Cubic, leaveTangent=4.229, tangentMode=RCTM_User, time=0, value=0.5415}
         {arriveTangent=0.917, interpMode=RCIM_Cubic, leaveTangent=0.917, time=0.1, value=0.9}
         <at 0.5 1 RCIM_Cubic>
        ]}}})
    Index 0 Param 3=&FloatFromCurve(FloatCurve=DataInterface() {propertyValues {Curve {keys [{arriveTangent=5.3935, interpMode=RCIM_Cubic, leaveTangent=3.8884, tangentMode=RCTM_User, time=0.5, value=1}, <at 1 2 RCIM_Cubic>]}}})
    Module.Param0WriteEnabled=true
    Module.Param1WriteEnabled=false
    Module.Param2WriteEnabled=true
    Module.Param3WriteEnabled=false
   ) {module=DynamicMaterialParameters001}
   &SolveForcesAndVelocity(Clamp Velocity=false, Limit Acceleration=false)
   &ScaleColor(
    Scale Alpha=&FloatFromCurve(FloatCurve=DataInterface() {propertyValues {Curve {keys [
         {arriveTangent=6.233, interpMode=RCIM_Cubic, leaveTangent=6.233, tangentMode=RCTM_User, time=0, value=0}
         {arriveTangent=3.1104, interpMode=RCIM_Cubic, leaveTangent=3.1104, time=0.2, value=1.2445}
         {arriveTangent=-1.5556, interpMode=RCIM_Cubic, leaveTangent=-1.5556, time=0.6833, value=2.1254}
         {interpMode=RCIM_Cubic, leaveTangent=-3.4385, time=1, value=0}
        ]}}})
    Scale RGB=&MakeVectorFromLinearColorRGB(Color=&ColorFromCurve(ColorCurve=DataInterface() {propertyValues {
        AlphaCurve {keys [<at -0.0023 1 RCIM_Cubic>,<at 1 1 RCIM_Cubic>]}
        BlueCurve {keys [<at -0.002 0 RCIM_Cubic>,<at 0.998 0 RCIM_Cubic>]}
        GreenCurve {keys [<at -0.002 1.3836 RCIM_Cubic>,<at 0.998 0.0572 RCIM_Cubic>]}
        RedCurve {keys [<at -0.002 4 RCIM_Cubic>,<at 0.998 1 RCIM_Cubic>]}
       }}))
    ScaleA=true
    ScaleRGB=true
   )
   &ScaleSpriteSize(Scale Factor=&Vector2DFromFloat(Value=&FloatFromCurve(
      FloatCurve=DataInterface() {propertyValues {Curve {keys [
          {arriveTangent=1.067, interpMode=RCIM_Cubic, leaveTangent=1.067, tangentMode=RCTM_User, time=0, value=0.5357}
          {arriveTangent=0.0093, interpMode=RCIM_Cubic, leaveTangent=0.0093, tangentMode=RCTM_User, time=1, value=1}
         ]}}}
      Scale Curve=&Lerp_Float(A=1, B=2, Alpha=$User.Vehicle_Speed)
     )))
   /SetVariables(Particles.SubImageIndex=&FloatFromCurve(FloatCurve=DataInterface() {propertyValues {Curve {keys [{arriveTangent=87.6094, time=0, value=0}, <at 1.009 50>]}}})) {isSetParameters=true}
  ]
  rendererClasses [&NiagaraSpriteRendererProperties]
  renderers [{rendererClass=&NiagaraSpriteRendererProperties}]
 }
 Local_Smoke_Forward001 {
  emitterUpdateScript [
   &EmitterState()
   &SpawnRate(
    SpawnRate=&MakeCustomFloatFromBool(
     Bool=&SetBoolByFloatComparison(A=$User.DamageScale, B=$User.Stage1, Comparison Type=::NewEnumerator2)
     True Float=&MakeCustomFloatFromBool(
      Bool=&SetBoolByFloatComparison(A=$User.DamageScale, B=$User.Stage2, Comparison Type=::NewEnumerator2)
      True Float=14
      False Float=10
     )
     False Float=0
    )
    Use Spawn Probability=false
   )
  ]
  particleSpawnScript [
   &InitializeParticle(
    Lifetime=&UniformRangedFloat(Minimum=0.6, Maximum=1)
    Module.Write Color=true
    Module.Write Lifetime=true
    Module.Write Mass=true
    Module.Write Material Random=true
    Module.Write Position=false
    Module.Write Scale=false
    Module.Write SpriteRotation=true
    Module.Write SpriteSize=true
    Position=&SimulationPosition()
    Sprite Rotation=&NormalizedAngleToDegrees(NormalizedAngle=&UniformRangedFloat(Minimum=0, Maximum=1))
    Sprite Size=&Vector2DFromFloat(Value=&UniformRangedFloat(Minimum=75, Maximum=90))
   )
   &InheritVelocity(Module.Limit Speed=true, Velocity Limit=1500, Velocity Scale=<0.04 0.04 -0.02>)
   &SphereLocation(
    Hemisphere X=true
    MaskSpawnGroup=false
    Module.Override Randomness=false
    Non Uniform Scale=&Lerp_Vector(A=<0 0 0>, B=<1 1 0.1>, Alpha=0.5)
    Sphere Radius=40
    Surface Only=false
   )
   &AddVelocity(Velocity=&Multiply_Vector(
     A=$Output.SphereLocation.SphereNormal
     B=&VectorFromFloat(Value=&UniformRangedFloat(Minimum=50, Maximum=100))
    ))
   &AddVelocity(Velocity=&UniformRangedVector(Minimum=<-100 -300 75>, Maximum=<-250 300 100>)) {module=AddVelocity001}
   &CameraOffset(Camera Offset Amount=25)
   &Color(Color=&MakeLinearColorFromVectorAndFloat(
     "Vector (RGB)"=&MakeVectorFromLinearColorRGB(Color=&ColorFromCurve(
       ColorCurve=DataInterface() {propertyValues {
         AlphaCurve {keys [<at -0.0023 1 RCIM_Constant>,<at 0.0995 1 RCIM_Constant>,<at 1 1 RCIM_Constant>]}
         BlueCurve {keys [<at 0 0.1 RCIM_Constant>,<at 0.45 0.05 RCIM_Constant>,<at 1 0.0464 RCIM_Constant>]}
         GreenCurve {keys [<at 0 0.1 RCIM_Constant>,<at 0.45 0.05 RCIM_Constant>,<at 1 0.2858 RCIM_Constant>]}
         RedCurve {keys [<at 0 0.1 RCIM_Constant>,<at 0.45 0.05 RCIM_Constant>,<at 1 2 RCIM_Constant>]}
        }}
       CurveIndex=$User.DamageScale
      ))
     "Float (Alpha)"=1
    ))
   &BoxLocation(Box Size=<100 100 1>, MaskSpawnGroup=false, Module.Override Seed=false, Rotate Box=false, Surface Only=false)
   &AddVelocity(Velocity=&Lerp_Vector(A=<0 0 0>, B=<-605 0 0>, Alpha=$User.Vehicle_Speed)) {module=AddVelocity002}
  ]
  particleUpdateScript [
   &ParticleState()
   &Drag(Drag=2, Use Linear Drag=true, Use Rotational Drag=false)
   &AccelerationForce(Acceleration=<0 0 100>)
   &SolveForcesAndVelocity(Clamp Velocity=false, Limit Acceleration=false)
   &ScaleColor(
    Scale Alpha=&FloatFromCurve(
     FloatCurve=DataInterface() {propertyValues {Curve {keys [
         {interpMode=RCIM_Cubic, tangentMode=RCTM_User, time=0, value=0}
         {arriveTangent=0.201, interpMode=RCIM_Cubic, leaveTangent=0.201, tangentMode=RCTM_User, time=0.1, value=0.575}
         {arriveTangent=-1.4222, interpMode=RCIM_Cubic, leaveTangent=-1.4222, tangentMode=RCTM_User, time=0.5526, value=0.434}
         {interpMode=RCIM_Cubic, tangentMode=RCTM_User, time=1, value=0}
        ]}}}
     Scale Curve=&MakeCustomFloatFromBool(
      Bool=&SetBoolByFloatComparison(A=$User.DamageScale, B=$User.Stage2, Comparison Type=::NewEnumerator2)
      True Float=0.5
      False Float=0.25
     )
    )
    Scale RGB=&VectorFromFloat(Value=&FloatFromCurve(FloatCurve=DataInterface() {propertyValues {Curve {keys [{arriveTangent=-3.2082, interpMode=RCIM_Cubic, leaveTangent=-1.8602, tangentMode=RCTM_User, time=0, value=1.25}, <at 1 0.15 RCIM_Cubic>]}}}))
    ScaleA=true
    ScaleRGB=true
   )
   &DynamicMaterialParameters(
    Index 0 Param 1=&FloatFromCurve(FloatCurve=DataInterface() {propertyValues {Curve {keys [<at 0 1>,<at 1 0>]}}})
    Index 0 Param 2=&FloatFromCurve(FloatCurve=DataInterface() {propertyValues {Curve {keys [
         {arriveTangent=3.5758, interpMode=RCIM_Cubic, leaveTangent=3.5758, tangentMode=RCTM_User, time=-0.0037, value=-0.957}
         {arriveTangent=1.3602, interpMode=RCIM_Cubic, leaveTangent=1.3602, tangentMode=RCTM_User, time=1.0011, value=1}
        ]}}})
    Index 0 Param 4=0.5
    Module.Param0WriteEnabled=true
    Module.Param1WriteEnabled=true
    Module.Param2WriteEnabled=true
    Module.Param3WriteEnabled=true
   )
   &ScaleSpriteSize(Scale Factor=&Vector2DFromFloat(Value=&FloatFromCurve(FloatCurve=DataInterface() {propertyValues {Curve {keys [{arriveTangent=4.3948, interpMode=RCIM_Cubic, leaveTangent=4.3948, tangentMode=RCTM_User, time=0, value=0.5}, <at 1 3 RCIM_Cubic>]}}})))
   &SpriteRotationRate(Rotation Rate=&UniformRangedFloat(Minimum=-1, Maximum=1))
  ]
  rendererClasses [&NiagaraSpriteRendererProperties]
  renderers [{rendererClass=&NiagaraSpriteRendererProperties}]
 }
 World_Smoke_Critical001 {
  emitterUpdateScript [
   &EmitterState()
   &SpawnRate(
    SpawnRate=&MakeCustomFloatFromBool(
     Bool=&SetBoolByFloatComparison(A=$User.DamageScale, B=$User.Stage3Critical, Comparison Type=::NewEnumerator2)
     True Float=25
     False Float=0
    )
    Use Spawn Probability=false
   ) {module=SpawnRate001}
  ]
  particleSpawnScript [
   &InitializeParticle(
    Color=<rgba 1 1 1 0.5>
    Lifetime=&UniformRangedFloat(Minimum=0.6, Maximum=1)
    Module.Write Color=true
    Module.Write Lifetime=true
    Module.Write Mass=true
    Module.Write Material Random=true
    Module.Write Position=false
    Module.Write Scale=false
    Module.Write SpriteRotation=true
    Module.Write SpriteSize=true
    Position=&SimulationPosition()
    Sprite Rotation=&NormalizedAngleToDegrees(NormalizedAngle=&UniformRangedFloat(Minimum=0, Maximum=1))
    Sprite Size=&Vector2DFromFloat(Value=&UniformRangedFloat(Minimum=30, Maximum=50))
   )
   &SphereLocation(
    Hemisphere X=true
    MaskSpawnGroup=false
    Module.Override Randomness=false
    Non Uniform Scale=&Lerp_Vector(A=<0 0 0>, B=<1 1 0.1>, Alpha=0.5)
    Surface Only=false
   )
   &AddVelocity(Velocity=&Multiply_Vector(
     A=$Output.SphereLocation.SphereNormal
     B=&VectorFromFloat(Value=&UniformRangedFloat(Minimum=50, Maximum=100))
    ))
   &AddVelocity(Velocity=&UniformRangedVector(Minimum=<-100 -300 100>, Maximum=<-250 300 200>)) {module=AddVelocity001}
   &CameraOffset(Camera Offset Amount=50)
   &Color(Color=&MakeLinearColorFromVectorAndFloat(
     "Vector (RGB)"=&MakeVectorFromLinearColorRGB(Color=&ColorFromCurve(
       ColorCurve=DataInterface() {propertyValues {
         AlphaCurve {keys [<at -0.0023 1 RCIM_Constant>,<at 0.0995 1 RCIM_Constant>,<at 1 1 RCIM_Constant>]}
         BlueCurve {keys [<at 0 0.1 RCIM_Constant>,<at 0.45 0.05 RCIM_Constant>,<at 1 0.0464 RCIM_Constant>]}
         GreenCurve {keys [<at 0 0.1 RCIM_Constant>,<at 0.45 0.05 RCIM_Constant>,<at 1 0.2858 RCIM_Constant>]}
         RedCurve {keys [<at 0 0.1 RCIM_Constant>,<at 0.45 0.05 RCIM_Constant>,<at 1 2 RCIM_Constant>]}
        }}
       CurveIndex=$User.DamageScale
      ))
     "Float (Alpha)"=0.75
    ))
  ]
  particleUpdateScript [
   &ParticleState()
   &Drag(Drag=2, Use Linear Drag=true, Use Rotational Drag=false)
   &AccelerationForce(Acceleration=<0 0 500>)
   &SolveForcesAndVelocity(Clamp Velocity=false, Limit Acceleration=false)
   &ScaleColor(
    Scale Alpha=&FloatFromCurve(FloatCurve=DataInterface() {propertyValues {Curve {keys [
         {interpMode=RCIM_Cubic, tangentMode=RCTM_User, time=0, value=0}
         {interpMode=RCIM_Cubic, tangentMode=RCTM_User, time=0.1007, value=0.6723}
         {arriveTangent=-2.0134, interpMode=RCIM_Cubic, leaveTangent=-2.0134, tangentMode=RCTM_User, time=0.5526, value=0.5348}
         {interpMode=RCIM_Cubic, tangentMode=RCTM_User, time=1, value=0}
        ]}}})
    Scale RGB=&VectorFromFloat(Value=&FloatFromCurve(FloatCurve=DataInterface() {propertyValues {Curve {keys [
          {arriveTangent=-3.2082, interpMode=RCIM_Cubic, leaveTangent=-1.8602, tangentMode=RCTM_User, time=0, value=1.25}
          <at 0.3977 0.25>
          <at 0.5013 0.2904 RCIM_Cubic>
         ]}}}))
    ScaleA=true
    ScaleRGB=true
   )
   &DynamicMaterialParameters(
    Index 0 Param 1=&FloatFromCurve(FloatCurve=DataInterface() {propertyValues {Curve {keys [<at 0 0.15>,<at 1 -1>]}}})
    Index 0 Param 2=&FloatFromCurve(FloatCurve=DataInterface() {propertyValues {Curve {keys [{arriveTangent=2.2631, interpMode=RCIM_Cubic, time=0, value=0}, {interpMode=RCIM_Cubic, leaveTangent=1.4308, time=1, value=0.125}]}}})
    Index 0 Param 4=0.5
    Module.Param0WriteEnabled=true
    Module.Param1WriteEnabled=false
    Module.Param2WriteEnabled=true
    Module.Param3WriteEnabled=true
   )
   &ScaleSpriteSize(Scale Factor=&Vector2DFromFloat(Value=&FloatFromCurve(FloatCurve=DataInterface() {propertyValues {Curve {keys [{arriveTangent=4.3948, interpMode=RCIM_Cubic, leaveTangent=4.3948, tangentMode=RCTM_User, time=0, value=0.5}, <at 1 3 RCIM_Cubic>]}}})))
   &SpriteRotationRate(Rotation Rate=&UniformRangedFloat(Minimum=-1, Maximum=1))
  ]
  rendererClasses [&NiagaraSpriteRendererProperties]
  renderers [{rendererClass=&NiagaraSpriteRendererProperties}]
 }
 World_Smoke_Forward001 {
  emitterUpdateScript [
   &EmitterState()
   &SpawnRate(
    SpawnRate=&MakeCustomFloatFromBool(
     Bool=&SetBoolByFloatComparison(A=$User.DamageScale, B=$User.Stage1, Comparison Type=::NewEnumerator2)
     True Float=&MakeCustomFloatFromBool(
      Bool=&SetBoolByFloatComparison(A=$User.DamageScale, B=$User.Stage2, Comparison Type=::NewEnumerator2)
      True Float=18
      False Float=12
     )
     False Float=0
    )
    Use Spawn Probability=false
   )
  ]
  particleSpawnScript [
   &InitializeParticle(
    Color=<rgba 1 1 1 0.25>
    Lifetime=&UniformRangedFloat(Minimum=0.8, Maximum=1)
    Module.Write Color=true
    Module.Write Lifetime=true
    Module.Write Mass=true
    Module.Write Material Random=true
    Module.Write Position=false
    Module.Write Scale=false
    Module.Write SpriteRotation=true
    Module.Write SpriteSize=true
    Position=&SimulationPosition()
    Sprite Rotation=&NormalizedAngleToDegrees(NormalizedAngle=&UniformRangedFloat(Minimum=0, Maximum=1))
    Sprite Size=&Vector2DFromFloat(Value=&Multiply_Float(
      A=&UniformRangedFloat(Minimum=30, Maximum=50)
      B=&Lerp_Float(A=1, B=1.5, Alpha=$User.Vehicle_Speed)
     ))
   )
   &CylinderLocation(
    Cylinder Height=20
    Cylinder Radius=10
    Hemisphere X=true
    Hemisphere Y=true
    Lathe Profile=DataInterface() {propertyValues {Curve {keys []}}}
    MaskSpawnGroup=false
    Module.Enable Lathe Profile=false
    Module.Override Randomness=false
    Module.Surface Only=false
    Non Uniform Scale:Vector3f=<1 1 -1>
    Orientation Axis:ENiagaraOrientationAxis=::XAxis
   )
   &AddVelocity(Velocity=&Multiply_Vector(
     A=$Output.CylinderLocation.CylinderNormal
     B=&VectorFromFloat(Value=&UniformRangedFloat(Minimum=50, Maximum=100))
    ))
   &AddVelocity(Velocity=&UniformRangedVector(Minimum=<-100 0 100>, Maximum=<-250 0 200>)) {module=AddVelocity001}
   &CameraOffset(Camera Offset Amount=50)
   &Color(Color=&MakeLinearColorFromVectorAndFloat(
     "Vector (RGB)"=&MakeVectorFromLinearColorRGB(Color=&ColorFromCurve(
       ColorCurve=DataInterface() {propertyValues {
         AlphaCurve {keys [<at -0.0023 1 RCIM_Constant>,<at 0.0995 1 RCIM_Constant>,<at 1 1 RCIM_Constant>]}
         BlueCurve {keys [<at 0 0.1 RCIM_Constant>,<at 0.45 0.05 RCIM_Constant>,<at 1 0.0464 RCIM_Constant>]}
         GreenCurve {keys [<at 0 0.1 RCIM_Constant>,<at 0.45 0.05 RCIM_Constant>,<at 1 0.2858 RCIM_Constant>]}
         RedCurve {keys [<at 0 0.1 RCIM_Constant>,<at 0.45 0.05 RCIM_Constant>,<at 1 2 RCIM_Constant>]}
        }}
       CurveIndex=$User.DamageScale
      ))
     "Float (Alpha)"=1
    ))
   &BoxLocation(Box Size=<100 100 1>, MaskSpawnGroup=false, Module.Override Seed=false, Rotate Box=false, Surface Only=false)
  ]
  particleUpdateScript [
   &ParticleState()
   &AccelerationForce(Acceleration=<0 0 500>)
   &CurlNoiseForce(
    Noise Frequency=0.5
    Noise Strength=&Lerp_Float(A=0, B=500, Alpha=&Power(a=$Particles.NormalizedAge, b=0.25))
    Pan Noise=true
    Pan Noise Field=<0 0 100>
   )
   &Drag(Drag=3, Use Linear Drag=true, Use Rotational Drag=false)
   &SolveForcesAndVelocity(Clamp Velocity=false, Limit Acceleration=false)
   &ScaleColor(
    Scale Alpha=&FloatFromCurve(
     FloatCurve=DataInterface() {propertyValues {Curve {keys [
         {interpMode=RCIM_Cubic, tangentMode=RCTM_User, time=0, value=0}
         {interpMode=RCIM_Cubic, tangentMode=RCTM_User, time=0.0958, value=0.9565}
         {arriveTangent=-2.0134, interpMode=RCIM_Cubic, leaveTangent=-2.0134, tangentMode=RCTM_User, time=0.594, value=0.7182}
         {interpMode=RCIM_Cubic, tangentMode=RCTM_User, time=1, value=0}
        ]}}}
     Scale Curve=&MakeCustomFloatFromBool(
      Bool=&SetBoolByFloatComparison(A=$User.DamageScale, B=$User.Stage2, Comparison Type=::NewEnumerator2)
      True Float=0.5
      False Float=0.25
     )
    )
    Scale RGB=&VectorFromFloat(Value=&FloatFromCurve(FloatCurve=DataInterface() {propertyValues {Curve {keys [{arriveTangent=-3.2082, interpMode=RCIM_Cubic, leaveTangent=-1.8602, tangentMode=RCTM_User, time=0, value=1.25}, <at 1 0.15 RCIM_Cubic>]}}}))
    ScaleA=true
    ScaleRGB=true
   )
   &DynamicMaterialParameters(
    Index 0 Param 1=&FloatFromCurve(FloatCurve=DataInterface() {propertyValues {Curve {keys [
         {arriveTangent=1.3596, interpMode=RCIM_Cubic, leaveTangent=1.3596, tangentMode=RCTM_User, time=0, value=0}
         {arriveTangent=0.1028, interpMode=RCIM_Cubic, leaveTangent=0.1028, tangentMode=RCTM_User, time=1, value=0.75}
        ]}}})
    Index 0 Param 2=&FloatFromCurve(FloatCurve=DataInterface() {propertyValues {Curve {keys [
         {arriveTangent=0.6717, interpMode=RCIM_Cubic, leaveTangent=0.6717, tangentMode=RCTM_User, time=-0.0003, value=-0.9768}
         {arriveTangent=3.037, interpMode=RCIM_Cubic, leaveTangent=3.037, tangentMode=RCTM_User, time=1.0011, value=1}
        ]}}})
    Index 0 Param 4=0.5
    Index 2 Param 1=&FloatFromCurve(FloatCurve=DataInterface() {propertyValues {Curve {keys [
         {arriveTangent=2.5015, interpMode=RCIM_Cubic, leaveTangent=2.5015, tangentMode=RCTM_User, time=0, value=0}
         {arriveTangent=1.1306, interpMode=RCIM_Cubic, leaveTangent=1.1306, tangentMode=RCTM_User, time=0.25, value=0.5}
         <at 1 1>
        ]}}})
    Index 2 Param 2=&FloatFromCurve(FloatCurve=DataInterface() {propertyValues {Curve {keys [
         {arriveTangent=2.5015, interpMode=RCIM_Cubic, leaveTangent=2.5015, tangentMode=RCTM_User, time=0, value=0}
         {arriveTangent=1.1306, interpMode=RCIM_Cubic, leaveTangent=1.1306, tangentMode=RCTM_User, time=0.25, value=0.5}
         <at 1 1>
        ]}}})
    Index 2 Param 3=&FloatFromCurve(FloatCurve=DataInterface() {propertyValues {Curve {keys [
         {arriveTangent=0.6717, interpMode=RCIM_Cubic, leaveTangent=0.6717, tangentMode=RCTM_User, time=-0.0028, value=-0.0101}
         {arriveTangent=3.037, interpMode=RCIM_Cubic, leaveTangent=3.037, tangentMode=RCTM_User, time=1.0011, value=2}
        ]}}})
    Index 2 Param 4=&FloatFromCurve(FloatCurve=DataInterface() {propertyValues {Curve {keys [
         {arriveTangent=-0.4045, interpMode=RCIM_Cubic, leaveTangent=-0.4045, tangentMode=RCTM_User, time=-0.0192, value=1.0634}
         {arriveTangent=-0.318, interpMode=RCIM_Cubic, leaveTangent=-0.318, tangentMode=RCTM_User, time=0.9751, value=0.415}
        ]}}})
    Module.Param0WriteEnabled=true
    Module.Param0WriteEnabled2=true
    Module.Param1WriteEnabled=true
    Module.Param1WriteEnabled2=true
    Module.Param2WriteEnabled=true
    Module.Param2WriteEnabled2=true
    Module.Param3WriteEnabled=true
    Module.Param3WriteEnabled2=true
    Write Parameter Index 2=true
   )
   &ScaleSpriteSize(Scale Factor=&Vector2DFromFloat(Value=&FloatFromCurve(FloatCurve=DataInterface() {propertyValues {Curve {keys [{arriveTangent=4.3948, interpMode=RCIM_Cubic, leaveTangent=4.3948, tangentMode=RCTM_User, time=0, value=1}, <at 1 3 RCIM_Cubic>]}}})))
   &SpriteRotationRate(Rotation Rate=&UniformRangedFloat(Minimum=-1, Maximum=1))
  ]
  rendererClasses [&NiagaraSpriteRendererProperties]
  renderers [{rendererClass=&NiagaraSpriteRendererProperties}]
 }
}
summary {
 systemName=NS_Vehicle_Damage
 emitters [
  {emitterName=Local_Fire_Engine, rendererClasses [&NiagaraSpriteRendererProperties]}
  {emitterName=Local_Embers_Forward001, simTarget=GPUComputeSim, rendererClasses [&NiagaraSpriteRendererProperties]}
  {emitterName=World_Smoke_Forward001, rendererClasses [&NiagaraSpriteRendererProperties]}
  {emitterName=Local_Smoke_Forward001, rendererClasses [&NiagaraSpriteRendererProperties]}
  {emitterName=World_Smoke_Critical001, rendererClasses [&NiagaraSpriteRendererProperties]}
  {emitterName=Generator_Flare_Engine001, rendererClasses [&NiagaraSpriteRendererProperties]}
  {emitterName=Generator_Sparks_Forward001, rendererClasses [&NiagaraSpriteRendererProperties]}
 ]
 userVariables [
  User.DamageScale {type=NiagaraFloat, defaultValue=1, description=""}
  User.SimRT {
   type=Object
   description=""
   defaultValue {struct=NiagaraExt_VariableValue_Object, value {object=null, objectClass=&Object}}
  }
  User.Stage1 {type=NiagaraFloat, defaultValue=0.33, description=""}
  User.Stage2 {type=NiagaraFloat, defaultValue=0.66, description=""}
  User.Stage3Critical {type=NiagaraFloat, defaultValue=1, description=""}
  User.Vehicle_Speed {type=NiagaraFloat, defaultValue=0.5, description=""}
 ]
}
