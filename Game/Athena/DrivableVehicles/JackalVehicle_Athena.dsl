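; Blueprint for the Jackal drivable hoverboard in Athena, layering effects, HUD and hit responses onto the native vehicle.
; Enter/exit events track the driver, toggle quickbar visibility and manage the boost cooldown gameplay effect.
; Boost, jump, collision, water and death events drive particles, audio, camera shakes or haptics, with most cosmetics kept off dedicated servers.
; Dedicated-server mesh hits damage weak props at speed and bump players when the board is unoccupied.
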(event EventBeginPlay
  ;[1 1 1] EVENT -> Begin Play
  (|CallParentFunction)
  ;;[1 1 1] Nuke Audio/FX on server
  (if (Networking|IsDedicatedServer)
    (CallFunction|DeactivateComponents)
    (else
      (CallFunction|CacheFX (Variables|Default|GetBoostFX) (Variables|Default|GetLoopingFX))
      (if (Variables|Sounds|GetbUseAudio)
        (CallFunction|CacheAudioPointers (Variables|Default|GetAudioMovement) (Variables|Default|GetAudioBoost) (Variables|Default|GetAudioWind))))))

(event Vehicle|EventOnIsSprintingChanged
  ;[1 1 1] EVENT ->  On Sprinting Changed - Camera Shake Management
  (if (Vehicle|IsSprinting)
    (if true
      (CallFunction|StopDriverCameraShake)
      (else
        ;;[1 1 1] Sprinting. Start Camera shake if we don't have one, update shake data.
        (Utilities|IsValid (Variables|Hoverboard|GetDriverCameraShake)
          (:Is%20Not%20Valid
            (CallFunction|StartDriverCameraShake)))))
    (else
      ;;[1 1 1] Stop shakin, bro.
      (Utilities|IsValid (Variables|Hoverboard|GetDriverCameraShake)
        (:Is%20Valid
          (CallFunction|StopDriverCameraShake))))))

(event EventOnPawnEnterVehicle
  (params (PlayerPawn :type Fort%20Player%20Pawn%20Object%20Reference) (SeatIdx :type Integer))
  ;[1 1 1] EVENT -> Enter Vehicle
  (|CallParentFunction PlayerPawn SeatIdx)
  (Utilities|FlowControl|Sequence
    (:then_0
      (if (Pawn|IsLocallyControlled PlayerPawn)
        (Variables|Hoverboard|SetLocalPlayerPawn PlayerPawn)))
    (:then_1
      (Variables|Hoverboard|SetDriverPawn PlayerPawn)
      (CallFunction|HideQuickBarforSecondaryAbility true)
      (CallFunction|SetupCosmetics)
      ;;[1 1 1] Only true on local driving client and server
      (bind _asplayer_controller (Utilities|Casting|CastToPlayerController (Pawn|GetController (Variables|Hoverboard|GetDriverPawn))))
      (Variables|Hoverboard|SetDrivingPlayerController _asplayer_controller))
    (:then_2
      (bind _returnvalue (Environment|ApplyGlobalEnvironmentGameplayEffecttoActor PlayerPawn (Variables|Default|GetGE_Boost))) ; Apply GE for mobile Boost radial
      (Variables|Default|SetCooldownCueEvent _returnvalue))))

(event Custom|StartDriverCameraShake
  (bind _drivingplayercontroller (Variables|Hoverboard|GetDrivingPlayerController))
  ;[1 1 1] Start Driver Camera Shake
  (if (not (Variables|Hoverboard|GetDebug_NoCameraShake))
    (Utilities|IsValid _drivingplayercontroller
      (:Is%20Valid
        (bind _returnvalue (CameraShakes|StartCameraShake :self (Variables|/Script/Engine.PlayerController|GetPlayerCameraManager _drivingplayercontroller) :ShakeClass "/Game/Athena/DrivableVehicles/Speed_CameraShake.Speed_CameraShake_C" :Scale 1.0 :PlaySpace "CameraLocal"))
        (Variables|Hoverboard|SetDriverCameraShake (CameraShakes|ConvLegacyCameraShake _returnvalue))))))

(event EventOnHoldExitStarted
  (params (ExitingPawn :type Fort%20Pawn%20Object%20Reference) (ExitDuration :type Float%20%28single-precision%29))
  ;[1 1 1] EVENT -> On Exit Hold
  (Ability|SendGameplayEventtoActor ExitingPawn "(TagName=\"Athena.Vehicle.Biplane.OnHoldExitStarted\")" (Utilities|Struct|MakeGameplayEventData :ContextHandle "()" :InstigatorTags "(GameplayTags=)" :TargetTags "(GameplayTags=)" :EventMagnitude ExitDuration :TargetData "()")))

(event EventOnPawnExitVehicle
  (params (PlayerPawn :type Fort%20Player%20Pawn%20Object%20Reference) (ExitSocketName :type Name :default "None"))
  ;[1 1 1] EVENT -> Exit Vehicle
  (|CallParentFunction PlayerPawn ExitSocketName)
  (Utilities|FlowControl|Sequence
    (:then_0
      (if (Utilities|Equal(Object) PlayerPawn (Variables|Hoverboard|GetLocalPlayerPawn))
        (Variables|Hoverboard|SetLocalPlayerPawn 0)))
    (:then_1
      (CallFunction|StopDriverCameraShake)
      (CallFunction|HideQuickBarforSecondaryAbility false)
      (CallFunction|RemoveCooldownCueEvent PlayerPawn) ; Remove GE for mobile Boost radial
    )
    (:then_2
      (Variables|Hoverboard|SetDriverPawn 0)
      (Variables|Hoverboard|SetDrivingPlayerController 0)
      (CallFunction|DeactivateComponents)
      (Rendering|SetVisibility (Variables|Default|GetSM_Hoverboard_Lightbeams) true false)
      (if (Variables|Sounds|GetbUseAudio)
        (Audio|SpawnSoundAttached (Variables|Sounds|GetDismountSound) (Variables|Vehicle|GetSkeletalMesh))))
    (:then_3
      (Networking|CanExecuteCosmeticEvents
        (:True
          (Utilities|Time|ClearandInvalidateTimerbyHandle (Variables|Default|GetMovementVisualTimer)))))))

(event EventOnHoldExitStopped
  (params (ExitingPawn :type Fort%20Pawn%20Object%20Reference))
  ;[1 1 1] EVENT -> On Exit Hold
  (Ability|SendGameplayEventtoActor ExitingPawn "(TagName=\"Athena.Vehicle.Biplane.OnHoldExitStopped\")"))

(event Custom|StopDriverCameraShake
  (bind _drivingplayercontroller (Variables|Hoverboard|GetDrivingPlayerController))
  ;[1 1 1] Stop Driver Camera Shake
  (Utilities|IsValid (Variables|Hoverboard|GetDriverCameraShake)
    (:Is%20Valid
      (Utilities|IsValid _drivingplayercontroller
        (:Is%20Valid
          (CameraShakes|StopCameraShake (Variables|/Script/Engine.PlayerController|GetPlayerCameraManager _drivingplayercontroller) (Variables|Hoverboard|GetDriverCameraShake) false)
          (Variables|Hoverboard|SetDriverCameraShake 0))))))

(event Custom|SetupCosmetics
  ;[1 1 1] Setup Cosmetics
  (Networking|CanExecuteCosmeticEvents
    (:True
      (Components|Activation|Activate (Variables|Default|GetLoopingFX) false)
      (Rendering|SetVisibility (Variables|Default|GetFrontLight) true false)
      (Rendering|SetVisibility (Variables|Default|GetSM_Hoverboard_Lightbeams) false false)
      ;;[0.9201 0.2258 1] Sounds
      (if (Variables|Sounds|GetbUseAudio)
        (Audio|SpawnSoundAttached (Variables|Sounds|GetMountSound) (Variables|Vehicle|GetSkeletalMesh))))))

(event EventOnStopTick
  ;[1 1 1] Event on stop tick
  (Class|FortVehicleAudioVoice|StopAllAudio (Variables|Default|GetAudioMovement)))

(event Custom|ChangeOutMaterials
  ;[1 1 1] Update materials because the driver changed or ADS changed
  (Class|FortVehicleInterface|ApplyMaterialandUseWrapifPossible (Variables|Default|GetWrapBaseMaterial))
  (CallFunction|SendDamageValuetoMaterials (Variables|Hoverboard|GetSavedDamageValueForMID))
  (CallFunction|SendHeadlightValuetoMaterials (Variables|Hoverboard|GetSavedHeadlightValueForMID))
  (CallFunction|SendBoostTimetoMaterials (Variables|Hoverboard|GetBoostTimeValueForMID)))

(event EventK2_ApplyCosmeticWrap
  (params (LoadedWrap :type Athena%20Item%20Wrap%20Definition%20Object%20Reference))
  ;[1 1 1] Update materials because the driver changed or ADS changed
  (Class|FortVehicleInterface|ApplyMaterialandUseWrapifPossible (Variables|Default|GetWrapBaseMaterial))
  (CallFunction|SendDamageValuetoMaterials (Variables|Hoverboard|GetSavedDamageValueForMID))
  (CallFunction|SendHeadlightValuetoMaterials (Variables|Hoverboard|GetSavedHeadlightValueForMID))
  (CallFunction|SendBoostTimetoMaterials (Variables|Hoverboard|GetBoostTimeValueForMID)))

(event EventOnDeathPlayEffects
  (params (Damage :type Float%20%28single-precision%29) (DamageTags :type Gameplay%20Tag%20Container%20Structure%20%28by%20ref%29) (Momentum :type Vector :default "0, 0, 0") (HitInfo :type Hit%20Result%20Structure%20%28by%20ref%29) (InstigatedBy :type Fort%20Pawn%20Object%20Reference) (DamageCauser :type Actor%20Object%20Reference) (EffectContext :type Gameplay%20Effect%20Context%20Handle%20Structure))
  (bind _returnvalue (Transformation|GetActorLocation))
  (bind (_returnvalue_location _returnvalue_rotation) (Math|Transform|BreakTransform (Transformation|GetSocketTransform (Variables|Vehicle|GetSkeletalMesh) "FrameCenter" "RTS_World")))
  ;[1 1 1] EVENT -> On Death Play Effects
  (bind _returnvalue_2 (Effects|SpawnEmitteratLocation "/Game/Athena/DrivableVehicles/FX/Hoverboard/P_Hoverboard_Athena_Destroy.P_Hoverboard_Athena_Destroy" _returnvalue_location _returnvalue_rotation "1.000000,1.000000,1.000000" true "AutoRelease"))
  (Effects|Components|ParticleSystem|SetVectorParameter _returnvalue_2 "Velocity" (Vehicle|GetVehicleVelocity))
  (Camera|PlayWorldCameraShake "/Game/Effects/Fort_Effects/BluePrints/CameraShakes/B_Small_Vertical_Jolt_CameraShake.B_Small_Vertical_Jolt_CameraShake_C" _returnvalue 1.5e+02 6e+02)
  (if (Variables|Sounds|GetbUseAudio)
    (Audio|PlaySoundatLocation (Variables|Sounds|GetOnDeathSound) _returnvalue)
    (else
      _))
  (Development|SetHiddeninGame (Variables|Vehicle|GetSkeletalMesh) true false))

(event EventOnBoostBegin
  (bind _driverpawn (Variables|Hoverboard|GetDriverPawn))
  (bind _returnvalue (Animation|GetCurrentMontage _driverpawn :!class /Script/Engine.Character))
  ;[1 1 1] EVENT -> Boost Begin/End
  (CallFunction|SlipStreamCheck)
  (Utilities|IsValid _driverpawn
    (:Is%20Valid
      (if (Utilities|Array|ContainsItem (Variables|Default|GetMontagesToStopOnBoost) _returnvalue)
        (Animation|StopAnimMontage _driverpawn _returnvalue) ; Stop Pickaxe Swinging- Fixes FORT-146008 & FORT-379313 
        (else
          _))
      (Networking|CanExecuteCosmeticEvents
        (:True
          (Components|Activation|SetActive (Variables|Default|GetBoostFX) true false)
          (Variables|FX|SetFXBoostEnd_0 (* (Transformation|GetActorForwardVector) -9e+01))
          (CallFunction|StartDriverCameraShake))))))

(event EventOnBoostEnd
  ;[1 1 1] EVENT -> Boost Begin/End
  (Networking|CanExecuteCosmeticEvents
    (:True
      (Components|Activation|SetActive (Variables|Default|GetBoostFX) false false)
      (CallFunction|StopDriverCameraShake))))

(event Custom|HideQuickBarForSecondaryAbility
  (params (Hide :type Boolean))
  (bind _driverpawn (Variables|Hoverboard|GetDriverPawn))
  ;[1 1 1] Hide/Show Quickbar in HUD
  (Utilities|IsValid _driverpawn
    (:Is%20Valid
      (bind _asfort_player_controller_gameplay (Utilities|Casting|CastToFortPlayerControllerGameplay (Pawn|GetController _driverpawn)))
      (HUD|SetHUDElementVisibilitywithContext _asfort_player_controller_gameplay (GameplayTags|MakeGameplayTagContainerfromTag "(TagName=\"HUD.Athena.HideSecondaryForAbility\")") Hide "None" ""))))

(event EventOnJumpChargeBegin
  ;[1 1 1] EVENT -> Jump Charge Begin/End
  (Variables|Hoverboard|SetbJumpCharging true)
  (Variables|Hoverboard|SetJumped false))

(event EventOnJumped
  (params (Charge :type Float%20%28single-precision%29))
  (bind _drivingplayercontroller (Variables|Hoverboard|GetDrivingPlayerController))
  ;[1 1 1] EVENT -> Jump Charge Begin/End
  (Variables|Hoverboard|SetJumped true)
  (Variables|Hoverboard|SetbJumpCharging false)
  (Networking|CanExecuteCosmeticEvents
    (:True
      (Effects|SpawnEmitterAttached "/Game/Athena/DrivableVehicles/FX/Hoverboard/P_Hoverboard_Athena_Jump.P_Hoverboard_Athena_Jump" (Variables|Vehicle|GetSkeletalMesh) "FrameCenter")
      (if (Variables|Sounds|GetbUseAudio)
        (Components|Activation|Activate (Variables|Default|GetAudioJump) true)
        (else
          _))
      (Utilities|IsValid _drivingplayercontroller
        (:Is%20Valid
          (CameraShakes|StartCameraShake :self (Variables|/Script/Engine.PlayerController|GetPlayerCameraManager _drivingplayercontroller) :ShakeClass (Variables|Hoverboard|GetJumpCameraShake) :Scale 1.0 :PlaySpace "CameraLocal"))))))

(event EventOnJumpChargeEnd
  (params (Charge :type Float%20%28single-precision%29)))

(event EventEndPlay
  (params (EndPlayReason :type EEndPlayReason%20Enum :default "Destroyed"))
  (CallFunction|HideQuickBarforSecondaryAbility false))

(event EventOnCollisionHitEffects
  (params (HitLocation :type Vector%20%28by%20ref%29 :default "0, 0, 0") (HitNormalImpulse :type Vector%20%28by%20ref%29 :default "0, 0, 0") (HitFrictionImpulse :type Vector%20%28by%20ref%29 :default "0, 0, 0") (HitNormal :type Vector%20%28by%20ref%29 :default "0, 0, 0") (HitActor :type Actor%20Object%20Reference) (HitSurfaceType :type EPhysicalSurface%20Enum :default "SurfaceType_Default"))
  (bind (_asfort_pawn _bsuccess) (Utilities|Casting|CastToFortPawn HitActor))
  (bind _landingfximpactthreshold (Variables|FX|GetLandingFXImpactThreshold))
  ;[1 1 1] EVENT -> Collision Hit FX/Audio
  (bind (_outresult _outmagnitude) (Audio|OnCollision (Variables|Default|GetFortCollisionAudioImpacts) HitLocation HitNormalImpulse))
  (if _outresult
    (Utilities|FlowControl|Sequence
      (:then_0
        (CallFunction|HapticsonHit _outmagnitude)
        (if (> _outmagnitude 1.5e+04)
          (bind _returnvalue (Niagara|SpawnSystematLocation (select _bsuccess "/Game/Athena/DrivableVehicles/FX/NS_Shopping_Cart_Vehicle_HitPlayer.NS_Shopping_Cart_Vehicle_HitPlayer" "/Game/Athena/DrivableVehicles/FX/NS_ShoppingCart/ns_Shopping_Cart_Vehicle_HitDust.NS_Shopping_Cart_Vehicle_HitDust") "0, 0, 0" (Math|Conversions|RotationFromXVector HitNormal)))
          (Effects|Components|ParticleSystem|SetEmitterEnable _returnvalue "Sparks" (> _outmagnitude 4.5e+04)) ; Turn on sparks for heavy hits
        ))
      (:then_1
        (if (and (> _outmagnitude _landingfximpactthreshold) (> (Math|Vector|DotProduct HitNormal (Transformation|GetActorUpVector)) 0.6))
          (bind _returnvalue_2 (Effects|SpawnEmitterAttached (Variables|FX|GetLandingFX) (Variables|Vehicle|GetSkeletalMesh) "None" "0, 0,-20.000000")) ; Spawn landing dust if landing flat with enough force
          (Effects|Components|ParticleSystem|SetEmitterEnable _returnvalue_2 "Lightning" (> _outmagnitude (* _landingfximpactthreshold 1.25))) ; Only do electricity for heavy lands
          (CallFunction|UpdateSurfaceType HitSurfaceType)
          (if (Variables|Sounds|GetbUseAudio)
            (Audio|Components|Audio|Play (Variables|Default|GetAudioLand) 0.0)))))))

(event EventOnEnteredWaterBody
  (params (WaterBody :type Fort%20Water%20Body%20Actor%20Object%20Reference))
  (bind (_returnvalue_x _returnvalue_y _returnvalue_z) (Math|Vector|BreakVector (Transformation|GetActorLocation)))
  (bind _waterfxheightoffset (Variables|Default|GetWaterFXHeightOffset))
  (bind _waterheight (Variables|Default|GetWaterHeight))
  (bind _returnvalue (+ _waterheight _waterfxheightoffset))
  (bind _returnvalue_2 (- _returnvalue _returnvalue_z))
  ;[0.2379 0.87 1] EVENT -> Entered Water Volume
  (bind _outwatersurfaceinfo (WaterBody|GetWaterSurfaceInfoStruct WaterBody (Transformation|GetSocketLocation (Variables|Vehicle|GetSkeletalMesh) (Variables|Default|GetWater%20Test%20Socket%20Name)) true false false))
  (bind (_waterplanelocation _waterplanenormal _watersurfacelocation) (Utilities|Struct|BreakWaterSurfaceInfo _outwatersurfaceinfo))
  (bind (_watersurfacelocation_x _watersurfacelocation_y _watersurfacelocation_z) (Math|Vector|BreakVector _watersurfacelocation))
  (Variables|Default|SetWaterHeight _watersurfacelocation_z)
  (Networking|CanExecuteCosmeticEvents
    (:True
      (if (> (Rendering|GetRenderingDetailMode) 0)
        (bind _returnvalue_3 (Utilities|Time|SetTimerbyEvent (delegate Custom|CheckWaterEvent) 0.25 true))
        (Variables|Hoverboard|SetCheckWaterTimer _returnvalue_3)
        (bind _inwaterfx (Variables|Default|GetInWaterFX)
          (:then
            (Components|DestroyComponent _inwaterfx))
          (:else
            _))
        (bind _returnvalue_4 (Effects|SpawnEmitterAttached :EmitterTemplate "/Game/WaterAndWind/Particles/WaterInteraction/P_Vehicle_In_Water_Athena.P_Vehicle_In_Water_Athena" :AttachToComponent (Variables|Vehicle|GetSkeletalMesh) :AttachPointName "None" :Location (Math|Vector|MakeVector 0.0 0.0 _returnvalue_2) :Scale "1.000000,1.000000,1.000000" :LocationType "KeepRelativeOffset" :bAutoDestroy true :PoolingMethod "None" :bAutoActivate true)) ; Looping water FX
        (Variables|Default|SetInWaterFX _returnvalue_4)
        (Effects|SpawnEmitterAttached :EmitterTemplate "/Game/WaterAndWind/Particles/WaterInteraction/P_WaterSplash_Athena_Vehicle.P_WaterSplash_Athena_Vehicle" :AttachToComponent (Variables|Vehicle|GetSkeletalMesh) :AttachPointName "None" :Location (Math|Vector|MakeVector 0.0 0.0 _returnvalue_2) :Scale "1.000000,1.000000,1.000000" :LocationType "KeepRelativeOffset" :bAutoDestroy true :PoolingMethod "None" :bAutoActivate true) ; Splash burst FX
        (bind _inwaterloop (Variables|Default|GetInWaterLoop)
          (:then
            (Audio|Components|Audio|Stop _inwaterloop))
          (:else
            _))
        (Audio|SpawnSoundAttached "/Game/Athena/Sounds/Vehicles/GolfCart/GolfCart_Movement_Impact_Water_Cue.GolfCart_Movement_Impact_Water_Cue" (Variables|Vehicle|GetSkeletalMesh))
        (bind _returnvalue_5 (Audio|SpawnSoundAttached "/Game/Athena/Sounds/Vehicles/GolfCart/GolfCart_Movement_InWater_Loop_Cue.GolfCart_Movement_InWater_Loop_Cue" (Variables|Vehicle|GetSkeletalMesh)))
        (Variables|Default|SetInWaterLoop _returnvalue_5)))))

(event Custom|CheckWaterEvent
  ;[0.2379 0.87 1] EVENT -> Entered Water Volume
  (Variables|Default|SetWaterHeight (CallFunction|GetWaterLevel)) ; Get the latest WaterLevel from native
  (Vehicle|CheckWater (+ (Variables|Default|GetWaterHeight) (Variables|Default|GetWaterFXHeightOffset)) (Variables|Default|GetInWaterFX) (Variables|Default|GetInWaterLoop)))

(event Vehicle|Water|EventOnEnteredWaterVolume
  (params (WaterSurfacePoint :type Vector%20%28by%20ref%29 :default "0, 0, 0"))
  (bind (_returnvalue_x _returnvalue_y _returnvalue_z) (Math|Vector|BreakVector (Transformation|GetActorLocation)))
  (bind _waterfxheightoffset (Variables|Default|GetWaterFXHeightOffset))
  (bind _waterheight (Variables|Default|GetWaterHeight))
  (bind _returnvalue (+ _waterheight _waterfxheightoffset))
  (bind _returnvalue_2 (- _returnvalue _returnvalue_z))
  ;[0.2379 0.87 1] EVENT -> Entered Water Volume
  (Variables|Default|SetWaterHeight (.z WaterSurfacePoint))
  (Networking|CanExecuteCosmeticEvents
    (:True
      (if (> (Rendering|GetRenderingDetailMode) 0)
        (bind _returnvalue_3 (Utilities|Time|SetTimerbyEvent (delegate Custom|CheckWaterEvent) 0.25 true))
        (Variables|Hoverboard|SetCheckWaterTimer _returnvalue_3)
        (bind _inwaterfx (Variables|Default|GetInWaterFX)
          (:then
            (Components|DestroyComponent _inwaterfx))
          (:else
            _))
        (bind _returnvalue_4 (Effects|SpawnEmitterAttached :EmitterTemplate "/Game/WaterAndWind/Particles/WaterInteraction/P_Vehicle_In_Water_Athena.P_Vehicle_In_Water_Athena" :AttachToComponent (Variables|Vehicle|GetSkeletalMesh) :AttachPointName "None" :Location (Math|Vector|MakeVector 0.0 0.0 _returnvalue_2) :Scale "1.000000,1.000000,1.000000" :LocationType "KeepRelativeOffset" :bAutoDestroy true :PoolingMethod "None" :bAutoActivate true)) ; Looping water FX
        (Variables|Default|SetInWaterFX _returnvalue_4)
        (Effects|SpawnEmitterAttached :EmitterTemplate "/Game/WaterAndWind/Particles/WaterInteraction/P_WaterSplash_Athena_Vehicle.P_WaterSplash_Athena_Vehicle" :AttachToComponent (Variables|Vehicle|GetSkeletalMesh) :AttachPointName "None" :Location (Math|Vector|MakeVector 0.0 0.0 _returnvalue_2) :Scale "1.000000,1.000000,1.000000" :LocationType "KeepRelativeOffset" :bAutoDestroy true :PoolingMethod "None" :bAutoActivate true) ; Splash burst FX
        (bind _inwaterloop (Variables|Default|GetInWaterLoop)
          (:then
            (Audio|Components|Audio|Stop _inwaterloop))
          (:else
            _))
        (Audio|SpawnSoundAttached "/Game/Athena/Sounds/Vehicles/GolfCart/GolfCart_Movement_Impact_Water_Cue.GolfCart_Movement_Impact_Water_Cue" (Variables|Vehicle|GetSkeletalMesh))
        (bind _returnvalue_5 (Audio|SpawnSoundAttached "/Game/Athena/Sounds/Vehicles/GolfCart/GolfCart_Movement_InWater_Loop_Cue.GolfCart_Movement_InWater_Loop_Cue" (Variables|Vehicle|GetSkeletalMesh)))
        (Variables|Default|SetInWaterLoop _returnvalue_5)))))

(event EventOnExitedWaterBody
  (params (WaterBody :type Fort%20Water%20Body%20Actor%20Object%20Reference))
  ;[0.2379 0.87 1] EVENT -> Entered Water Volume
  ;;[0.2541 0.3275 1] EVENT -> Exited Water Volume
  (Networking|CanExecuteCosmeticEvents
    (:True
      (Utilities|Time|ClearandInvalidateTimerbyHandle (Variables|Hoverboard|GetCheckWaterTimer))
      (bind _inwaterfx (Variables|Default|GetInWaterFX)
        (:then
          (Components|Activation|SetActive _inwaterfx false false)
          (bind _inwaterloop (Variables|Default|GetInWaterLoop)
            (:then
              (Audio|Components|Audio|FadeOut _inwaterloop 3.0 0.0 "Linear" :!class /Script/Engine.AudioComponent))
            (:else)))
        (:else)))))

(event Vehicle|Water|EventOnExitedWaterVolume
  (params (WaterSurfacePoint :type Vector%20%28by%20ref%29 :default "0, 0, 0"))
  ;[0.2379 0.87 1] EVENT -> Entered Water Volume
  ;;[0.2541 0.3275 1] EVENT -> Exited Water Volume
  (Networking|CanExecuteCosmeticEvents
    (:True
      (Utilities|Time|ClearandInvalidateTimerbyHandle (Variables|Hoverboard|GetCheckWaterTimer))
      (bind _inwaterfx (Variables|Default|GetInWaterFX)
        (:then
          (Components|Activation|SetActive _inwaterfx false false)
          (bind _inwaterloop (Variables|Default|GetInWaterLoop)
            (:then
              (Audio|Components|Audio|FadeOut _inwaterloop 3.0 0.0 "Linear" :!class /Script/Engine.AudioComponent))
            (:else)))
        (:else)))))

(event OnComponentHit(SkeletalMesh)
  (params (HitComponent :type Primitive%20Component%20Object%20Reference) (OtherActor :type Actor%20Object%20Reference) (OtherComp :type Primitive%20Component%20Object%20Reference) (NormalImpulse :type Vector :default "0, 0, 0") (Hit_bBlockingHit :type Boolean) (Hit_bInitialOverlap :type Boolean) (Hit_Time :type Float%20%28single-precision%29) (Hit_Distance :type Float%20%28single-precision%29) (Hit_Location :type Vector :default "0, 0, 0") (Hit_ImpactPoint :type Vector :default "0, 0, 0") (Hit_Normal :type Vector :default "0, 0, 0") (Hit_ImpactNormal :type Vector :default "0, 0, 0") (Hit_PhysMat :type Physical%20Material%20Object%20Reference) (Hit_HitActor :type Actor%20Object%20Reference) (Hit_HitComponent :type Primitive%20Component%20Object%20Reference) (Hit_HitBoneName :type Name :default "None") (Hit_BoneName :type Name :default "None") (Hit_HitItem :type Integer) (Hit_ElementIndex :type Integer) (Hit_FaceIndex :type Integer) (Hit_TraceStart :type Vector :default "0, 0, 0") (Hit_TraceEnd :type Vector :default "0, 0, 0"))
  ;[1 1 1] EVENT -> On Hit Skel Mesh: Destroy Props
  (bind _output_get (Variables|Hoverboard|SetVehicleHitActor OtherActor))
  (if (Networking|IsDedicatedServer)
    (Utilities|FlowControl|Sequence
      (:then_0
        (if (CallFunction|ShouldDestroyProp? Hit_Normal)
          (bind _asbuilding_prop (Utilities|Casting|CastToBuildingProp (Variables|Hoverboard|GetVehicleHitActor)))
          (if (not (Actor|ActorHasTag _asbuilding_prop "NoVehicleDamage"))
            (CallFunction|DestroyBuildingPiece _asbuilding_prop (Variables|Hoverboard|GetMaxHealthToDestroy)))))
      (:then_1
        (bind _asfort_player_pawn (Utilities|Casting|CastToFortPlayerPawn _output_get))
        (if (CallFunction|ShouldBump Hit_Normal _asfort_player_pawn)
          (Audio|BroadcastAbilityImpactatLocation "/Game/Sounds/Fort_Cinematics/10_seconds_silence_Cinematic_Cue.10_seconds_silence_Cinematic_Cue" self (* Hit_Normal (Variables|/Script/FortniteGame.FortAthenaJackalVehicleConfigs|GetBumpForce (Variables|Vehicle|GetFortJackalVehicleConfigs))) 771.0))))))

(event EventShowCooldownCue
  (params (Pawn :type Fort%20Player%20Pawn%20Object%20Reference) (Duration :type Float%20%28single-precision%29))
  ;[1 1 1] Boost Cooldown Cue
  (Ability|SendGameplayEventtoActor Pawn "(TagName=\"GameplayEffect.Vehicle.Boost\")" (Utilities|Struct|MakeGameplayEventData :ContextHandle "()" :InstigatorTags "(GameplayTags=)" :TargetTags "(GameplayTags=)" :EventMagnitude Duration :TargetData "()")))

(fn CurrentBoostCharge
  (bind _returnvalue (CallFunction|GetCurrentBoostTimer))
  (return (.Charge (Utilities|Struct|BreakFortRechargingActionTimer _returnvalue))))

(fn DeactivateComponents
  (Components|Activation|Deactivate (Variables|Default|GetLoopingFX))
  (Rendering|SetVisibility (Variables|Default|GetFrontLight) false false))

(fn DestroyBuildingPiece
  (params (BuildingActor :type Building%20Actor%20Object%20Reference) (DestroyThreshold :type Float%20%28double-precision%29))
  (if (Variables|/Script/Engine.Actor|GetbCanBeDamaged BuildingActor)
    (if (<= (Health|GetMaxHealth BuildingActor :!class /Script/FortniteGame.BuildingActor) DestroyThreshold)
      (GameplayEffects|ApplyGameplayEffectToTarget (Ability|GetAbilitySystemComponent (Variables|Hoverboard|GetDriverPawn)) "/Game/Athena/DrivableVehicles/Hoverboard/GE_VehicleDamage150.GE_VehicleDamage150_C" (Ability|GetAbilitySystemComponent (Variables|Hoverboard|GetVehicleHitActor)) 0.0 (GameplayEffects|MakeEffectContext (Ability|GetAbilitySystemComponent (Variables|Hoverboard|GetDriverPawn)))))))

(fn GetIconPlacement
  (params (SelfActor :type Actor%20Object%20Reference) (ViewingActor :type Actor%20Object%20Reference))
  (Utilities|IsValid SelfActor
    (:Is%20Valid
      (return (Transformation|GetActorLocation SelfActor) "0, 0, 0"))
    (:Is%20Not%20Valid
      (return "0, 0, 0" "0, 0, 0"))))

(fn GetMaxAudibleDistance
  (bind _audiomovement (Variables|Default|GetAudioMovement))
  (bind _sound (Variables|/Script/Engine.AudioComponent|GetSound _audiomovement))
  (Utilities|IsValid _sound
    (:Is%20Valid
      (return (Variables|/Script/Engine.SoundBase|GetMaxDistance _sound)))
    (:Is%20Not%20Valid
      (return 0.0))))

(fn HapticsOnHit
  (params (HitMagnitude :type Float%20%28double-precision%29))
  (bind _drivingplayercontroller (Variables|Hoverboard|GetDrivingPlayerController))
  (bind _drivingplayercontroller_2 (Variables|Hoverboard|GetDrivingPlayerController))
  (Utilities|IsValid (Variables|Hoverboard|GetDriverPawn)
    (:Is%20Valid
      (if (> (Variables|GetHitMagnitude) 1.2e+04)
        (Utilities|FlowControl|Sequence
          (:then_0
            (LocalVariables|SetCameraShakeMagnitude (Math|Float|MapRangeClamped (Variables|GetHitMagnitude) 1.5e+04 5e+04 0.1 1.0))
            (Utilities|IsValid _drivingplayercontroller_2
              (:Is%20Valid
                (Game|Feedback|ClientStartCameraShake :self _drivingplayercontroller_2 :Shake "/Game/Athena/DrivableVehicles/CameraShake_HitImpact_ShoppingCart.CameraShake_HitImpact_ShoppingCart_C" :Scale (LocalVariables|GetCameraShakeMagnitude) :PlaySpace "CameraLocal")
                (if (> (Variables|GetHitMagnitude) 2e+04)
                  (Game|Feedback|ClientPlayForceFeedback _drivingplayercontroller "/Game/Weapons/ForceFeedback/FF_Generic_Impact_Near.FF_Generic_Impact_Near" "None" false false false)
                  (else
                    (Game|Feedback|ClientPlayForceFeedback _drivingplayercontroller "/Game/Weapons/ForceFeedback/FF_Generic_Build.FF_Generic_Build" "None" false false false))))))
          (:then_1
            (if (< (Math|Float|Absolute(Float) (Math|Vector|DotProduct (Math|Vector|Normalize (Vehicle|GetPreviousLinearVelocity)) "0, 0,1.000000")) 0.5)
              (bind _output_get (LocalVariables|SetHitDelta (- (Variables|Vehicle|GetCachedSpeed) (Math|Vector|VectorLength (Transformation|GetComponentVelocity (Variables|Vehicle|GetSkeletalMesh))))))
              (if (> _output_get (LocalVariables|GetCameraPushVelocityDelta))
                (Game|Feedback|ClientStartCameraShake (Variables|Hoverboard|GetDrivingPlayerController) "/Game/Athena/DrivableVehicles/CameraShake_BigHit.CameraShake_BigHit_C" (Math|Float|MapRangeClamped (LocalVariables|GetHitDelta) 3e+02 9e+02 0.5 1.0) "UserDefined" (Math|Rotator|MakeRotfromX (Vehicle|GetPreviousLinearVelocity)))))))))))

(fn IsBoostCharging
  (bind _returnvalue (CallFunction|GetCurrentBoostTimer))
  (return (.bIsCharging (Utilities|Struct|BreakFortRechargingActionTimer _returnvalue))))

(fn OnRep_bUseAudio
  ;[0.9227 0.2723 1] Update the audio components on the client when Use Audio changes
  (Networking|CanExecuteCosmeticEvents
    (:True
      (if (Variables|Sounds|GetbUseAudio)
        (CallFunction|CacheAudioPointers (Variables|Default|GetAudioMovement) (Variables|Default|GetAudioBoost) (Variables|Default|GetAudioWind))
        (else
          (CallFunction|CacheAudioPointers 0)
          (Class|FortVehicleAudioVoice|StopAllAudio (Variables|Default|GetAudioMovement)))))))

(fn RemoveCooldownCueEvent
  (params (PlayerPawn :type Fort%20Player%20Pawn%20Object%20Reference))
  (bind _returnvalue (Ability|GetAbilitySystemComponent PlayerPawn))
  (Ability|SendGameplayEventtoActor PlayerPawn (Variables|Default|GetGamePlayEvent) (Utilities|Struct|MakeGameplayEventData :ContextHandle "()" :InstigatorTags "(GameplayTags=)" :TargetTags "(GameplayTags=)" :EventMagnitude 0.0 :TargetData "()"))
  (if (not (Class|ItemUIFunctionLibrary|IsNullItem _returnvalue))
    (GameplayEffects|RemoveActiveGameplayEffect _returnvalue (Variables|Default|GetCooldownCueEvent) -1)))

(fn SendBoostTimeToMaterials
  (params (ParameterValue :type Float%20%28double-precision%29))
  (bind _output_get (Variables|Hoverboard|SetBoostTimeValueForMID ParameterValue))
  (Rendering|Material|SetScalarParameterValueonMaterials (Variables|Vehicle|GetSkeletalMesh) "BoostTime" _output_get))

(fn SendDamageValueToMaterials
  (params (ParameterValue :type Float%20%28double-precision%29))
  (bind _output_get (Variables|Hoverboard|SetSavedDamageValueForMID ParameterValue))
  (Rendering|Material|SetScalarParameterValueonMaterials (Variables|Vehicle|GetSkeletalMesh) "Damage" _output_get))

(fn SendHeadlightValueToMaterials
  (params (ParameterValue :type Float%20%28double-precision%29))
  (bind _output_get (Variables|Hoverboard|SetSavedHeadlightValueForMID ParameterValue))
  (Rendering|Material|SetScalarParameterValueonMaterials (Variables|Vehicle|GetSkeletalMesh) "HeadLights" _output_get))

(fn ShouldBump
  (params (Hit%20Normal :type Vector) (PlayerPawn :type Fort%20Player%20Pawn%20Object%20Reference))
  (return (Math|Boolean|ANDBoolean (< (Math|Float|Absolute(Float) (Math|Vector|DotProduct Hit%20Normal (Transformation|GetActorUpVector))) (Variables|Hoverboard|GetBumpForceAngle)) (not (Vehicle|ContainsPlayers)) (not (Vehicle|IsinVehicle PlayerPawn)))))

(fn ShouldDestroyProp?
  (params (HitNormal :type Vector))
  (bind _vehiclehitactor (Variables|Hoverboard|GetVehicleHitActor)
    (:then
      (return (and (Variables|/Script/Engine.Actor|GetbCanBeDamaged _vehiclehitactor) (and (<= (Math|Vector|DotProduct (Math|Vector|Normalize (Vehicle|GetPreviousLinearVelocity)) HitNormal) (Variables|Hoverboard|GetDestructionAngle)) (>= (Vehicle|GetPreviousForwardSpeedKmh) (Variables|Hoverboard|GetMaxSpeedToDestroy))))))
    (:else
      (return false))))

(fn SlipStreamCheck
  (if (SkyTube|IsinSkyTube (Variables|Hoverboard|GetDriverPawn) :!class /Script/FortniteGame.FortPlayerPawn)
    (Audio|BroadcastAbilityImpactatLocation "/Game/Sounds/Fort_Cinematics/10_seconds_silence_Cinematic_Cue.10_seconds_silence_Cinematic_Cue" self (* (Vehicle|GetVehicleForwardVector) (Variables|Default|GetSlipStreamImpulseForce)) 771.0)))

(fn UpdateSurfaceType
  (params (Surface :type EPhysicalSurface%20Enum :default "SurfaceType_Default"))
  ;[1 1 1] 0 = default, 1 = water, 2 = snow/ice
  (bind _output_get (Variables|Hoverboard|SetFXSurfaceType (select (Utilities|Enum|Equal(Enum) Surface "SurfaceType8") 1 (select (or (Utilities|Enum|Equal(Enum) Surface "SurfaceType23") (Utilities|Enum|Equal(Enum) Surface "SurfaceType24")) 2 0))))
  (return _output_get))

(fn ConstructionScript
  (Variables|FortAthenaVehicle|SetOverlapComponent (Variables|Default|GetOverlapVolume))
  (bind _returnvalue (Rendering|Material|CreateDynamicMaterialInstance (Variables|Vehicle|GetSkeletalMesh) 1 "/Game/Athena/DrivableVehicles/Quad/M_BoostMeter_Bar.M_BoostMeter_Bar" "None" :!class /Script/Engine.PrimitiveComponent))
  (Variables|FX|SetBoostMeterMID _returnvalue))
