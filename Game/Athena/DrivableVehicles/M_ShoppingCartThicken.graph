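# Shopping cart surface material that thickens the mesh at distance, likely to keep its thin parts readable.
# FOV-scaled pixel depth pushes vertices along their normals and blends the normal map toward axis-snapped normals derived from local position.
# A Damage value, from dynamic and scalar parameters, drives color tint, roughness and a tiling dent normal.
# Depth also raises brightness and roughness, and the attributes pass through MF_CreativeMoveTool.
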Material /Game/Athena/DrivableVehicles/M_ShoppingCartThicken
graph {
 properties {
  bTangentSpaceNormal=false
  bUseMaterialAttributes=true
  bUsedWithMeshDeformer=true
  bUsedWithMeshParticles=true
  bUsedWithNanite=true
  bUsedWithNiagaraMeshParticles=true
  bUsedWithSkeletalMesh=true
  pixelDepthOffsetMode=PDOM_Legacy
  usedShadingModels=Default Lit
 }
 nodes [
  MaterialFunctionCall() {materialFunction=&LocalPosition}
  n1 = TextureSample() {texture=/Game/Environments/Props/Street/Textures/Prop_Shopping_Cart_01_D}
  n2 = Add(A=Multiply(A=@n1.RGB, B=@n1.RGB), B=Multiply(A=@n1.R, B=Constant3Vector() {constant=<rgba 0.401 0.1969 0.057 0>}))
  n3 = DynamicParameter() {defaultValue=<rgba 0 1 1 1>, paramNames [Damage,Param2,Param3,Param4]}
  n4 = ScalarParameter() {parameterName=Damage}
  n5 = Clamp(None=Power(Base=Add(A=@n3.Damage, B=@n4)) {constExponent=2.5})
  n6 = Clamp(
   None=Divide(A=PixelDepth().R, B=ScalarParameter() {defaultValue=1500, parameterName=RoughnessDistanceFade})
   Max=ScalarParameter() {defaultValue=1, parameterName=RoughAddMAx}
  )
  n7 = Multiply(A=ScalarParameter() {defaultValue=1}, B=@n6)
  n8 = TextureSample() {samplerType=SAMPLERTYPE_Masks, texture=/Game/Environments/Props/Street/Textures/Prop_Shopping_Cart_01_S}
  n9 = Add(A=Multiply(A=@n5) {constB=0.35}, B=ScalarParameter() {defaultValue=0.1, parameterName=RoughnessMin})
  n10 = ScalarParameter() {defaultValue=1, parameterName=RoughnessMax}
  n11 = TextureSample(UVs=TextureCoordinate() {uTiling=0.5, vTiling=0.5}) {samplerType=SAMPLERTYPE_Normal, texture=/Game/Global/Textures/Default/Tileable_Patterns/CarDamageNormal}
  n12 = Constant3Vector() {constant=<rgba 1.5 1.5 0 0>}
  n13 = Add(
   A=Multiply(A=Multiply(A=@n11.RGB, B=@n12), B=@n5)
   B=TextureSample() {samplerType=SAMPLERTYPE_Normal, texture=/Game/Environments/Props/Street/Textures/Prop_Shopping_Cart_01_N}.RGB
  )
  n14 = ScalarParameter() {defaultValue=0.6857, parameterName=BlendThickness}
  n15 = MaterialFunctionCall() {materialFunction=&LocalPosition}
  n16 = VectorParameter() {parameterName=Scale, defaultValue=<rgba 0.5824 0.9691 0.2367 1>}
  n17 = Normalize(VectorInput=Add(A=Multiply(A=@n15.Local Position, B=@n16.RGB), B=VectorParameter() {parameterName=Bias, defaultValue=<rgba -5.9347 0 -12.7106 0>}.RGB))
  n18 = MaterialFunctionCall(Float3=Abs(None=@n17)) {materialFunction=&BreakOutFloat3Components}
  n19 = Add(A=Divide(A=@n14), B=Subtract(A=@n18.R, B=@n18.R))
  n20 = LinearInterpolate(A=Constant3Vector() {constant=<rgba 0 1 0 0>}, B=Constant3Vector() {constant=<rgba 1 0 0 0>}, Alpha=Saturate(None=Divide(A=@n19, B=@n14)))
  n21 = Constant3Vector() {constant=<rgba 0 0 1 0>}
  n22 = Divide(A=@n14)
  n23 = Subtract(A=@n18.B, B=Max(A=@n18.R, B=@n18.R))
  n24 = Saturate(None=Divide(A=Add(A=@n22, B=@n23), B=@n14))
  n25 = Transform(None=Multiply(A=LinearInterpolate(A=@n20, B=@n21, Alpha=@n24), B=Sign(None=@n17))) {transformSourceType=TRANSFORMSOURCE_Local}
  n26 = ViewProperty() {property=MEVP_TanHalfFieldOfView}
  n27 = Multiply(A=Multiply(A=ComponentMask(None=@n26.Property) {r=true}) {constB=1.4}, B=PixelDepth().R)
  n28 = ScalarParameter() {defaultValue=900, parameterName=StartingBias}
  n29 = Divide(A=Clamp(None=Subtract(A=@n27, B=@n28)) {clampMode=CMODE_ClampMin}, B=ScalarParameter() {defaultValue=1600, parameterName=FalloffDistance})
  n30 = Multiply(A=@n29, B=ScalarParameter() {defaultValue=0.3, parameterName=NormalBlendDistanceMultiply})
  n31 = ScalarParameter() {defaultValue=0.8, parameterName=NormalLerpMin}
  n32 = VertexNormalWS()
  n33 = Clamp(None=@n29, Max=ScalarParameter() {defaultValue=4, parameterName=MaxThicken})
  n34 = SetMaterialAttributes!(
   MaterialAttributes
   Base Color=Multiply(A=LinearInterpolate(A=@n1.RGB, B=@n2, Alpha=@n5), B=Add(B=@n7) {constA=1})
   Metallic=@n8.G
   Roughness=Add(A=LinearInterpolate(A=@n9, B=@n10, Alpha=@n8.B), B=@n6)
   Normal=LinearInterpolate(A=Transform(None=@n13), B=VertexInterpolator(VS=@n25).PS, Alpha=Clamp(None=@n30, Max=@n31))
   World Position Offset=Multiply(A=Multiply(A=@n32, B=@n33), B=ScalarParameter() {defaultValue=1, parameterName=ThickeningMultiply})
  ) {attributeSetTypes [
    #1
    #2
    #3
    #4
    #5
   ]}
  MaterialFunctionCall(MA In=@n34) {materialFunction=&MF_CreativeMoveTool}
  MaterialFunctionCall() {materialFunction=&DebugTimeSine}
  n35 = MaterialFunctionCall() {materialFunction=&BoundingBoxBased_0-1_UVW}
  Min(A=Min(A=@n35.R, B=@n35.R), B=@n35.B)
  n36 = TransformPosition(None=WorldPosition() {worldPositionShaderOffset=WPT_ExcludeAllShaderOffsets}.XYZ) {transformSourceType=TRANSFORMPOSSOURCE_World}
  Multiply(A=MaterialFunctionCall(Float3=@n36) {materialFunction=&BreakOutFloat3Components}.R)
  n37 = MaterialFunctionCall() {materialFunction=&ObjectLocalBounds}
  Multiply(A=@n37.Local Bounds Size, B=VectorParameter() {defaultValue=<rgba 0 0 0 1>}.RGB)
  Transform(None=Divide(A=@n36, B=@n37.Local Bounds Size)) {transformSourceType=TRANSFORMSOURCE_Local}
 ]
}
