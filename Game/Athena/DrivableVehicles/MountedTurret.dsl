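; Mounted turret, a drivable Athena vehicle that pawns man to fire a dual weapon.
; Entering applies weapon-grant and fire-block effects, hides quickbar and held weapon, and binds fire montages; exiting restores them.
; Overheat replicates from server as a quantized integer, while a platform-dependent timer drives native FX and audio updates.
; Interaction waits for the start animation and a clear trace; destroying the linked actor destroys the turret.
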(event EventOnHoldExitStarted
  (params (ExitingPawn :type Fort%20Pawn%20Object%20Reference) (ExitDuration :type Float%20%28single-precision%29))
  ;[1 1 1] EVENT -> OnHoldExitStarted/Stopped
  (Ability|SendGameplayEventtoActor ExitingPawn "(TagName=\"Athena.Vehicle.Biplane.OnHoldExitStarted\")" (Utilities|Struct|MakeGameplayEventData :ContextHandle "()" :InstigatorTags "(GameplayTags=)" :TargetTags "(GameplayTags=)" :EventMagnitude ExitDuration :TargetData "()")))

(event EventBeginPlay
  ;[1 1 1] EVENT -> Begin Play
  (Utilities|FlowControl|Sequence
    (:then_0
      (bind _structout (Utilities|Struct|SetmembersinTurretCosmeticData (Variables|MountedTurret|GetTurretCosmeticData) (Variables|Default|GetOverheatSmoke) (Variables|Default|GetOverHeatScreen_L) (Variables|Default|GetOverHeatScreen_R) (Variables|Default|GetAudioOverheat) (Variables|Default|GetAudioOnOverheat) (Variables|Sounds|GetOverheatAudioCurve) (Variables|Default|GetServoAudioComponent) (Variables|Default|GetFortVehicleAimingWeaponComp)))
      (Utilities|Array|Add (.OverheatParticleSystems (Utilities|Struct|BreakTurretCosmeticData _structout)) (Variables|Default|GetOverHeatScreen_L))
      (Utilities|Array|Add (.OverheatParticleSystems (Utilities|Struct|BreakTurretCosmeticData _structout)) (Variables|Default|GetOverHeatScreen_R))
      (Utilities|Array|Add (.OverheatParticleSystems (Utilities|Struct|BreakTurretCosmeticData _structout)) (Variables|Default|GetBarrelGlowL))
      (Utilities|Array|Add (.OverheatParticleSystems (Utilities|Struct|BreakTurretCosmeticData _structout)) (Variables|Default|GetBarrelGlowR))
      (Utilities|Array|Add (.OverheatParticleSystems (Utilities|Struct|BreakTurretCosmeticData _structout)) (Variables|Default|GetMagazineGlowL))
      (Utilities|Array|Add (.OverheatParticleSystems (Utilities|Struct|BreakTurretCosmeticData _structout)) (Variables|Default|GetMagazineGlowR))
      (bind _turretstartanimation (Variables|Animations|GetTurretStartAnimation)
        (:then
          (Variables|InteractionGuardGroup|SetfInteractionDelay (+ (Animation|GetPlayLength _turretstartanimation) (Variables|InteractionGuardGroup|GetfInteractionDelayOffset)))
          (Animation|Montage|PlayMontage (Variables|Vehicle|GetSkeletalMesh) (Variables|Animations|GetTurretStartAnimation)))
        (:else
          _))
      ;;[1 1 1] If server, nuke audio and FX
      (if (Networking|IsDedicatedServer)
        (Components|DestroyComponent (Variables|Default|GetBarrelGlowR))
        (Variables|MountedTurret|SetTurretRotation (Transformation|GetWorldRotation (Variables|Vehicle|GetSkeletalMesh)))
        (Networking|SwitchHasAuthority
          (:Authority
            (Utilities|FlowControl|Delay (Variables|InteractionGuardGroup|GetfInteractionDelay))
            (Variables|InteractionGuardGroup|SetbCanInteract true)))))
    (:then_2
      ;;[1 1 1] Wake up turret if set to start awake.
      (if (Variables|Default|GetbStartMountedTurretAwake)
        (Utilities|FlowControl|DelayUntilNextTick) ; Wait until next tick so that native AthenaVehicle doesn't override us.  (See ShouldSleepAtSpawn())
        (Vehicle|WakeUp)))))

(event EventOnHoldExitStopped
  (params (ExitingPawn :type Fort%20Pawn%20Object%20Reference))
  ;[1 1 1] EVENT -> OnHoldExitStarted/Stopped
  (Ability|SendGameplayEventtoActor ExitingPawn "(TagName=\"Athena.Vehicle.Biplane.OnHoldExitStopped\")"))

(event Custom|overheat%20changed
  (params (CurrentValue :type Float%20%28single-precision%29) (MaxValue :type Float%20%28single-precision%29) (State :type EFortWeaponOverheatState%20Enum) (bIsInCooldown :type Boolean))
  ;[1 1 1] Update Server replication of overheat
  (Networking|SwitchHasAuthority
    (:Authority
      (if (and (<= CurrentValue 0.0) (Utilities|Enum|Equal(Enum) State "None"))
        (Utilities|Time|ClearandInvalidateTimerbyHandle (Variables|Default|GetOverheatOnServerTimer))
        (Variables|Default|SetReplicateOverheatOf20 0)
        (CallFunction|SetOverheating false)
        (elif (Utilities|Time|IsTimerActivebyHandle (Variables|Default|GetOverheatOnServerTimer))
          _
          (else
            (bind _returnvalue (Utilities|Time|SetTimerbyEvent (delegate Custom|UpdateOverheatOnServer) 0.1 true))
            (Variables|Default|SetOverheatOnServerTimer _returnvalue)
            (CallFunction|UpdateOverheatonServer)
            (CallFunction|SetOverheating true)))))))

(event Custom|UpdateOverheatOnServer
  ;[1 1 1] Update Server replication of overheat
  (Variables|Default|SetReplicateOverheatOf20 (Math|Float|Truncate (* (CallFunction|CurrentOverheatScalar) 2e+01))))

(event EventOnPawnEnterVehicle
  (params (PlayerPawn :type Fort%20Player%20Pawn%20Object%20Reference) (SeatIdx :type Integer))
  ;[0 0.625 0.5012] EVENT -> Pawn Enter Vehicle
  (Utilities|FlowControl|Sequence
    (:then_0
      (if (Pawn|IsLocallyControlled PlayerPawn)
        (Variables|MountedTurret|SetLocalPlayerPawn PlayerPawn)))
    (:then_1
      (CallFunction|SetManned true)
      (Variables|MountedTurret|SetDriverPawn PlayerPawn)
      (Variables|MountedTurret|SetDriverControllerGameplay (Utilities|Casting|CastToFortPlayerControllerGameplay (Pawn|GetController PlayerPawn)))
      (CallFunction|StartFXandAudioTimer)
      (CallFunction|HideQuickBarforAbility true)
      (bind _output_get (Variables|MountedTurret|SetPlayerWeapon (Variables|/Script/FortniteGame.FortPawn|GetCurrentWeapon PlayerPawn)))
      (Rendering|SetActorHiddenInGame _output_get true)
      (if (Vehicle|IsOperatingTurret PlayerPawn)
        (Utilities|FlowControl|Sequence
          (:then_0
            (Animation|Montage|PlayMontage (Variables|/Script/Engine.Character|GetMesh (Variables|MountedTurret|GetDriverPawn)) (Variables|Animations|GetPlayerEnterAnimation))
            (Animation|Montage|PlayMontage (Variables|Vehicle|GetSkeletalMesh) (Variables|Animations|GetTurretEnterAnimation)))
          (:then_1
            (bind _returnvalue (Environment|ApplyGlobalEnvironmentGameplayEffecttoActor PlayerPawn (Variables|MountedTurret|GetGE_GrantWeapon)))
            (Variables|MountedTurret|SetGE_EnterTurret _returnvalue)
            (GameplayEffects|ApplyGameplayEffectToTarget (Ability|GetAbilitySystemComponent PlayerPawn) (Variables|MountedTurret|GetGE_BlockWeaponFire) (Ability|GetAbilitySystemComponent PlayerPawn) 1.0 (Utilities|Struct|MakeGameplayEffectContextHandle))
            ;[1 1 1] Do animations when player fires weapon
            (bind _returnvalue_2 (Utilities|Time|SetTimerbyEvent (delegate Custom|Get%20Turret%20Weapon) 0.05 true))
            (Variables|MountedTurret|SetGetWeaponTimer _returnvalue_2)))))))

(event Custom|Get%20Turret%20Weapon
  ;[1 1 1] Do animations when player fires weapon
  (Utilities|IsValid (Variables|MountedTurret|GetDriverPawn)
    (:Is%20Valid
      (bind _asfort_weapon_ranged_dual (Utilities|Casting|CastToFortWeaponRangedDual (Variables|/Script/FortniteGame.FortPawn|GetCurrentWeapon (Variables|MountedTurret|GetDriverPawn))))
      (if (not (Variables|Default|GetbSkipOverheatBind))
        (Overheating|BindEventtoOnOverheatValueChanged _asfort_weapon_ranged_dual (EventDispatchers|CreateEvent :!event Custom|overheat%20changed)))
      (bind _output_get (Variables|MountedTurret|SetTurretWeapon _asfort_weapon_ranged_dual))
      (RangedWeaponDual|BindEventtoOnHandFired _output_get (delegate Custom|Event%20On%20Weapon%20Fired))
      (Utilities|Time|ClearandInvalidateTimerbyHandle (Variables|MountedTurret|GetGetWeaponTimer)))
    (:Is%20Not%20Valid
      (Utilities|Time|ClearandInvalidateTimerbyHandle (Variables|MountedTurret|GetGetWeaponTimer)))))

(event Custom|Start%20FX%20And%20Audio%20Timer
  ;[1 1 1] Timer to control FX scaling
  (Utilities|FlowControl|DoOnce false
    (:Completed
      (if (not (Variables|Default|GetbSkipOverheatBind))
        (bind _returnvalue (Utilities|Time|SetTimerbyEvent (delegate Custom|CallNativeUpdateOverheatCosmetics) (select (Fortnite|IsMobilePlatform) 0.1 0.03) true))
        (Variables|MountedTurret|SetOverheatCostmeticTimer _returnvalue)))))

(event Custom|Event%20On%20Weapon%20Fired
  (params (Hand :type EDualWeaponHand%20Enum :default "LEFT") (bPersistantFire :type Boolean))
  (bind _driverpawn (Variables|MountedTurret|GetDriverPawn))
  ;[1 1 1] Do animations when player fires weapon
  (Utilities|IsValid _driverpawn
    (:Is%20Valid
      (Utilities|FlowControl|Sequence
        (:then_0
          (Animation|Montage|PlayMontage (Variables|/Script/Engine.Character|GetMesh _driverpawn) (Variables|Animations|GetPlayerFireAnimation))
          (Animation|Montage|PlayMontage (Variables|Vehicle|GetSkeletalMesh) (select (Utilities|Enum|Equal(Enum) Hand "LEFT") (Variables|Animations|GetTurretFireLeftAnimation) (Variables|Animations|GetTurretFireRightAnimation)))
          ;[1 1 1] Timer to control FX scaling
          (Utilities|FlowControl|DoOnce false
            (:Completed
              (if (not (Variables|Default|GetbSkipOverheatBind))
                (bind _returnvalue (Utilities|Time|SetTimerbyEvent (delegate Custom|CallNativeUpdateOverheatCosmetics) (select (Fortnite|IsMobilePlatform) 0.1 0.03) true))
                (Variables|MountedTurret|SetOverheatCostmeticTimer _returnvalue)))))))))

(event Custom|Clear%20FX%20And%20AudioTimer
  ;[1 1 1] Timer to control FX scaling
  (Utilities|Time|ClearandInvalidateTimerbyHandle (Variables|MountedTurret|GetOverheatCostmeticTimer))
  (Utilities|FlowControl|DoOnce false
    (:Completed
      (if (not (Variables|Default|GetbSkipOverheatBind))
        (bind _returnvalue (Utilities|Time|SetTimerbyEvent (delegate Custom|CallNativeUpdateOverheatCosmetics) (select (Fortnite|IsMobilePlatform) 0.1 0.03) true))
        (Variables|MountedTurret|SetOverheatCostmeticTimer _returnvalue)))))

(event EventOnPawnExitVehicle
  (params (PlayerPawn :type Fort%20Player%20Pawn%20Object%20Reference) (ExitSocketName :type Name :default "None"))
  (bind _skeletalmesh (Variables|Vehicle|GetSkeletalMesh))
  (bind _returnvalue (Components|SkeletalMesh|GetAnimInstance _skeletalmesh))
  (bind _asanim_bp_mounted_turret (Utilities|Casting|CastToAnimBP_MountedTurret _returnvalue))
  ;[0.8438 0.4558 0.3799] EVENT -> Pawn Exit Vehicle
  (Utilities|IsValid _asanim_bp_mounted_turret
    (:Is%20Valid
      (Animation|Montage|PlayMontage (Variables|Vehicle|GetSkeletalMesh) (Variables|Animations|GetTurretExitAnimation) 1.0 (Math|Float|MapRangeClamped (Math|Rotator|NormalizeAxis (Variables|/Script/FortniteGame.FortMountedTurretAnimInstance|GetAimingPitch _asanim_bp_mounted_turret)) 9e+01 -3e+01 0.0 0.75)
        (:then
          (CallFunction|HideQuickBarforAbility false)
          (Utilities|IsValid (Variables|MountedTurret|GetDriverPawn)
            (:Is%20Valid
              ;;[1 1 1] Remove Ability that grants weapon to player
              (GameplayEffects|RemoveActiveGameplayEffect (Ability|GetAbilitySystemComponent PlayerPawn) (Variables|MountedTurret|GetGE_EnterTurret) -1)
              (Rendering|SetActorHiddenInGame (Variables|MountedTurret|GetPlayerWeapon) false)
              (Variables|MountedTurret|SetTurretRotation (Pawn|GetControlRotation (Variables|MountedTurret|GetDriverPawn)))
              (Variables|MountedTurret|SetDriverPawn 0)
              (Variables|MountedTurret|SetDriverControllerGameplay 0)
              (Class|FortLayeredAudioComponent|ResetAllFloatParameters (Variables|Default|GetServoAudioComponent))
              (Audio|Parameter|SetFloatParameter (Variables|Default|GetAudioOverheat) "Overheat" 0.0 :!class /Script/AudioExtensions.AudioParameterControllerInterface))))
        (:OnCompleted
          _)
        (:OnInterrupted
          _))
      (CallFunction|SetManned false))
    (:Is%20Not%20Valid
      (Animation|Montage|PlayMontage (Variables|Vehicle|GetSkeletalMesh) (Variables|Animations|GetTurretExitAnimation)
        (:then
          _)
        (:OnCompleted
          _)
        (:OnInterrupted
          _)))))

(event Custom|CallNativeUpdateOverheatCosmetics
  ;[1 1 1] Timer to control FX scaling
  (CallFunction|NativeUpdateOverheatCosmetics (Variables|MountedTurret|GetDriverPawn) (Math|Conversions|ToFloat(Integer) (Variables|Default|GetReplicateOverheatOf20)) (Variables|MountedTurret|GetTurretRotation)))

(event Custom|HideQuickBarForAbility
  (params (Hide :type Boolean))
  ;[1 1 1] Hide/Show Quickbar in HUD
  (Utilities|IsValid (Variables|MountedTurret|GetDriverControllerGameplay)
    (:Is%20Valid
      (HUD|SetHUDElementVisibilitywithContext (Variables|MountedTurret|GetDriverControllerGameplay) (GameplayTags|MakeGameplayTagContainerfromArray (Variables|MountedTurret|GetHide%20HUD%20Gameplay%20Tags)) Hide "None" ""))))

(event EventEndPlay
  (params (EndPlayReason :type EEndPlayReason%20Enum :default "Destroyed"))
  (bind _driverpawn (Variables|MountedTurret|GetDriverPawn))
  (bind _returnvalue (Ability|GetAbilitySystemComponent _driverpawn))
  ;[1 1 1] EVENT -> End Play
  (CallFunction|HideQuickBarforAbility false)
  ;;[1 1 1] Remove Ability that grants weapon to player
  (Networking|SwitchHasAuthority
    (:Authority
      (Utilities|IsValid _driverpawn
        (:Is%20Valid
          (Utilities|IsValid _returnvalue
            (:Is%20Valid
              (GameplayEffects|RemoveActiveGameplayEffect _returnvalue (Variables|MountedTurret|GetGE_EnterTurret) -1))))))))

(event EventOnDeathPlayEffects
  (params (Damage :type Float%20%28single-precision%29) (DamageTags :type Gameplay%20Tag%20Container%20Structure%20%28by%20ref%29) (Momentum :type Vector :default "0, 0, 0") (HitInfo :type Hit%20Result%20Structure%20%28by%20ref%29) (InstigatedBy :type Fort%20Pawn%20Object%20Reference) (DamageCauser :type Actor%20Object%20Reference) (EffectContext :type Gameplay%20Effect%20Context%20Handle%20Structure))
  ;[1 1 1] EVENT -> Play Death Effects
  (Effects|SpawnEmitteratLocation "/Game/Gadgets/Assets/Turret_Gun/Particles/P_TurretGun_Destroyed.P_TurretGun_Destroyed" (Transformation|GetActorLocation) (Variables|MountedTurret|GetTurretRotation) "1.000000,1.000000,1.000000" true "AutoRelease")
  (Audio|PlaySoundatLocation (Variables|Sounds|GetOnDeathSound) (Transformation|GetActorLocation))
  (Class|FortLayeredAudioComponent|ResetAllFloatParameters (Variables|Default|GetServoAudioComponent)))

(event Custom|PlayDeathEffects
  ;[1 1 1] EVENT -> Play Death Effects
  (Networking|CanExecuteCosmeticEvents
    (:True
      (Effects|SpawnEmitteratLocation "/Game/Gadgets/Assets/Turret_Gun/Particles/P_TurretGun_Destroyed.P_TurretGun_Destroyed" (Transformation|GetActorLocation) (Variables|MountedTurret|GetTurretRotation) "1.000000,1.000000,1.000000" true "AutoRelease")
      (Audio|PlaySoundatLocation (Variables|Sounds|GetOnDeathSound) (Transformation|GetActorLocation))
      (Class|FortLayeredAudioComponent|ResetAllFloatParameters (Variables|Default|GetServoAudioComponent)))))

(event OnOverheatStateChanged(FortVehicleAimingWeaponComp)
  (params (OverheatState :type EFortWeaponOverheatState%20Enum))
  ;[1 1 1] EVENT -> OnOverheatStateChanged (FortVehicleAimingWeaponComp)
  (CallFunction|SetOverheating (Utilities|Enum|NotEqual(Enum) OverheatState "None")))

(event OnLinkedActorDestroyed(FortLinkToActorComp)
  (params (DamageCauser :type Actor%20Object%20Reference))
  (CallFunction|HandleLinkedActorDestroyed))

(fn CanInteract
  (params (InController :type Controller%20Object%20Reference))
  (bind _returnvalue (Pawn|GetControlledPawn InController))
  (Utilities|IsValid InController
    (:Is%20Valid
      (Utilities|IsValid _returnvalue
        (:Is%20Valid
          (bind (_outhit _returnvalue_2) (Collision|LineTraceByChannel (Transformation|GetActorLocation _returnvalue) (Transformation|GetSocketLocation (Variables|Vehicle|GetSkeletalMesh) "Center") "TraceTypeQuery10"))
          (if _returnvalue_2
            (return false)
            (else
              _)))
        (:Is%20Not%20Valid
          _))
      (return (Variables|InteractionGuardGroup|GetbCanInteract)))
    (:Is%20Not%20Valid
      _)))

(fn CurrentOverheatScalar
  (bind _fortvehicleaimingweaponcomp (Variables|Default|GetFortVehicleAimingWeaponComp))
  (bind _returnvalue (Overheat|GetOverheatingMaxValue _fortvehicleaimingweaponcomp :!class /Script/FortniteGame.FortVehicleSeatWeaponComponent))
  (Networking|SwitchHasAuthority
    (:Authority
      (if (> _returnvalue 0.0)
        (return (/ (Overheat|GetCurrentOverheatValue (Variables|Default|GetFortVehicleAimingWeaponComp) :!class /Script/FortniteGame.FortVehicleSeatWeaponComponent) _returnvalue))))
    (:Remote
      (return (/ (Math|Conversions|ToFloat(Integer) (Variables|Default|GetReplicateOverheatOf20)) 2e+01)))))

(fn HandleLinkedActorDestroyed
  (Vehicle|ForceDestroyVehicle true)
  (Actor|SetLifeSpan 0.2))

(fn OnGetExitSocketTransform
  (params (ExitSocket :type Name :default "None") (SocketIndex :type Integer) (InTransform :type Transform))
  (return true (Math|Transform|MakeTransform (+ (Math|Vector|RotateVectorAroundAxis (- (.location InTransform) (Transformation|GetActorLocation)) 9e+01 "0, 0,1.000000") (Transformation|GetActorLocation)) (.rotation InTransform) (.scale InTransform))))

(fn SetMagazineSizeCoefficient
  (params (Ammo%20Pool :type Float%20%28double-precision%29))
  (bind (_value _curve) (Utilities|Struct|BreakScalableFloat (Variables|/Script/FortniteGame.FortVehicleSeatWeaponComponent|GetTotalShotsAllowedThenShutdown (Variables|Default|GetFortVehicleAimingWeaponComp))))
  (bind (_curvetable _rowname) (Utilities|Struct|BreakCurveTableRowHandle _curve))
  (ScalableFloat|SetScalingValue (Variables|/Script/FortniteGame.FortVehicleSeatWeaponComponent|GetTotalShotsAllowedThenShutdown (Variables|Default|GetFortVehicleAimingWeaponComp)) Ammo%20Pool _rowname _curvetable))

(fn ConstructionScript
  (Audio|Components|Audio|SetSound (Variables|Default|GetServoAudioComponent) (Variables|Sounds|GetServoSound))
  (Variables|FortAthenaVehicle|SetOverlapComponent (Variables|Default|GetOverlapVolume))
  ;[1 1 1] For animation to not lag, skeletal mesh tick group has to be after the turret's tick group
  (Components|Tick|SetTickGroup (Variables|Vehicle|GetSkeletalMesh) "TG_DuringPhysics"))
