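; Vehicle component that turns a passenger's impulse-weapon shots into a push on the vehicle.
; Server-side, it tracks the passenger's equipped weapon and binds projectile spawns for matching weapon classes, skipping the push when a driven SKPush vehicle is not coasting.
; Grounded impulses follow the vehicle's forward axis by passenger facing, plus lift; airborne ones oppose passenger facing.
; The vector replicates through a RepNotify, with a dormancy flush forcing it out.
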(event Custom|OnPassengerWeaponEquipped
  (params (NewWeapon :type Fort%20Weapon%20Object%20Reference) (PreviousWeapon :type Fort%20Weapon%20Object%20Reference))
  (bind _asfort_weapon_ranged (Utilities|Casting|CastToFortWeaponRanged NewWeapon))
  (CallFunction|SetImpulseWeapon _asfort_weapon_ranged))

(event Custom|OnImpluseWeaponProjectileSpawned
  (params (Projectile :type Fort%20Projectile%20Base%20Object%20Reference))
  (bind _returnvalue (Components|GetOwner))
  (bind _asfort_vehicle_interface (Utilities|Casting|CastToFortVehicleInterface _returnvalue))
  (Utilities|IsValid (Vehicle|GetDriver _asfort_vehicle_interface)
    (:Is%20Valid
      (bind _asfort_athena_skpush_vehicle (Utilities|Casting|CastToFortAthenaSKPushVehicle _returnvalue)
        (:then
          (if (Vehicle|IsCoasting _asfort_athena_skpush_vehicle)
            _))
        (:CastFailed
          _))
      (CallFunction|ApplyImpulse))
    (:Is%20Not%20Valid
      _)))

(event Custom|ApplyImpulse
  (bind _returnvalue (Components|GetOwner))
  (Networking|CanExecuteCosmeticEvents
    (:False
      (bind _impulse (CallFunction|CalculateImpulse))
      (Variables|Default|SetImpulseVector _impulse) ; Notify so movement replicates to all clients
      (Networking|FlushNetDormancy _returnvalue) ; Force Replication for push the notify through \n(needed for Repnotifies on components)
      (bind _asfort_vehicle_interface (Utilities|Casting|CastToFortVehicleInterface _returnvalue))
      (if (not (Vehicle|HasDriver _asfort_vehicle_interface))
        (Default|CallOnPassengerFiredImpulseWeapon)))))

(fn CalculateImpulse
  (Utilities|FlowControl|Sequence
    (:then_0
      (bind _output_get (LocalVariables|SetLocalOwner (Components|GetOwner)))
      (bind _asfort_vehicle_interface (Utilities|Casting|CastToFortVehicleInterface _output_get))
      (bind _passengerpawn (Variables|Default|GetPassengerPawn)
        (:then
          (return (+ (* (+ (* (Transformation|GetActorForwardVector (LocalVariables|GetLocalOwner)) (* (ScalableFloat|GetValueatLevel (Variables|Default|GetImpulsePower)) (Math|Vector|DotProduct (* (Transformation|GetActorForwardVector (LocalVariables|GetLocalOwner)) -1.0) (Transformation|GetActorForwardVector _passengerpawn)))) "0, 0,200.000000") (Math|Conversions|ToFloat(Boolean) (not (Vehicle|VehicleIsinAir _asfort_vehicle_interface)))) (* (* (Transformation|GetActorForwardVector _passengerpawn) -2e+02) (Math|Conversions|ToFloat(Boolean) (Vehicle|VehicleIsinAir _asfort_vehicle_interface))))))
        (:else)))
    (:then_1
      (return "0.000000, 0, 0"))))

(fn IsAnImpulseWeapon
  (params (Weapon :type Fort%20Weapon%20Object%20Reference))
  (Utilities|FlowControl|Sequence
    (:then_0
      (bind _asfort_weapon_ranged (Utilities|Casting|CastToFortWeaponRanged Weapon))
      (for _array_element (Variables|Default|GetImpluseWeaponClasses)
        (if (Utilities|ClassIsChildOf (Utilities|GetClass _asfort_weapon_ranged) _array_element)
          (return true _asfort_weapon_ranged))))
    (:then_1
      (return false 0))))

(fn OnRep_ImpulseVector
  (Audio|BroadcastAbilityImpactatLocation 0 (Components|GetOwner) (Variables|Default|GetImpulseVector) 777.0))

(fn SetImpulseWeapon
  (params (NewImpulseWeapon :type Fort%20Weapon%20Ranged%20Object%20Reference))
  (if (Utilities|NotEqual(Object) NewImpulseWeapon (Variables|Default|GetImpulseWeapon))
    (Utilities|FlowControl|Sequence
      (:then_0
        (bind _impulseweapon (Variables|Default|GetImpulseWeapon)
          (:then
            (EventDispatchers|UnbindEventfromOnProjectileSpawned _impulseweapon (EventDispatchers|CreateEvent :!event OnImpluseWeaponProjectileSpawned))
            (Variables|Default|SetImpulseWeapon 0))
          (:else)))
      (:then_1
        (bind (_breturnvalue _impulseweapon_2) (CallFunction|IsAnImpulseWeapon (Variables|GetNewImpulseWeapon)))
        (if _breturnvalue
          (bind _output_get (Variables|Default|SetImpulseWeapon _impulseweapon_2))
          (Utilities|IsValid _output_get
            (:Is%20Valid
              (EventDispatchers|BindEventtoOnProjectileSpawned _output_get (EventDispatchers|CreateEvent :!event OnImpluseWeaponProjectileSpawned)))))))))

(fn SetPassengerPawn
  (params (NewPassengerPawn :type Fort%20Player%20Pawn%20Object%20Reference))
  (if (Utilities|NotEqual(Object) NewPassengerPawn (Variables|Default|GetPassengerPawn))
    (Networking|CanExecuteCosmeticEvents
      (:False
        (Utilities|FlowControl|Sequence
          (:then_0
            ;[1 1 1] Remove the old
            (Utilities|FlowControl|Sequence
              (:then_0
                (bind _passengerpawn (Variables|Default|GetPassengerPawn)
                  (:then
                    (FortPawn|UnbindEventfromOnWeaponEquippedDelegate _passengerpawn (EventDispatchers|CreateEvent :!event OnPassengerWeaponEquipped)))
                  (:else)))
              (:then_1
                (CallFunction|SetImpulseWeapon 0))))
          (:then_1
            ;[1 1 1] Add the new
            (bind _output_get (Variables|Default|SetPassengerPawn (Variables|GetNewPassengerPawn)))
            (Utilities|IsValid _output_get
              (:Is%20Valid
                (FortPawn|BindEventtoOnWeaponEquippedDelegate _output_get (EventDispatchers|CreateEvent :!event OnPassengerWeaponEquipped)) ; Check to see if player is changing to a rocket
                ;;[1 1 1] If player already has an impulse weapon equipped, set stuff up.
                (CallFunction|OnPassengerWeaponEquipped (Variables|/Script/FortniteGame.FortPawn|GetCurrentWeapon (Variables|Default|GetPassengerPawn)))))))))))
