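; Drivable push cannon vehicle Blueprint for Athena, with a driver seat and a barrel passenger seat that can be launched.
; Two water-depth thresholds disable seats or eject riders, barrel first, then everyone.
; Pre-launch hits apply a building-destroy GameplayEffect and are ignored by the launched pawn's movement.
; Helpers attach weapon pickups to cart sockets, push health into a material damage parameter, and scale camera shake and haptics by speed, suspension and impacts.
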(fn AttachWeaponToCart
  (params (Selection :type Integer))
  (bind (_returnvalue_roll _returnvalue_pitch _returnvalue_yaw) (Math|Rotator|BreakRotator (Transformation|GetActorRotation (Variables|PushCannon|GetFortPickup))))
  (bind _skeletalmesh (Variables|Vehicle|GetSkeletalMesh))
  (bind _fortpickup (Variables|PushCannon|GetFortPickup))
  (switch int Selection
    (:Default
      _)
    (:0
      _)
    (:1
      (Transformation|AttachActorToComponent _fortpickup _skeletalmesh "Weapon2" "SnapToTarget" "SnapToTarget" "SnapToTarget" true)
      (Transformation|SetActorRotation (Variables|PushCannon|GetFortPickup) (Math|Rotator|MakeRotator (Math|Random|RandomFloatinRange (+ _returnvalue_roll -1e+02) (+ _returnvalue_roll -8e+01)) (Math|Random|RandomFloatinRange (+ _returnvalue_pitch -5.0) (+ _returnvalue_pitch 5.0)) (Math|Random|RandomFloatinRange (+ _returnvalue_yaw -1e+02) (+ _returnvalue_yaw -8e+01))) false :!label _label))
    (:2
      (Transformation|AttachActorToComponent _fortpickup _skeletalmesh "Weapon3" "SnapToTarget" "SnapToTarget" "SnapToTarget" true)
      (goto _label))
    (:3
      (Transformation|AttachActorToComponent _fortpickup _skeletalmesh "Weapon4" "SnapToTarget" "SnapToTarget" "SnapToTarget" true)
      (goto _label))
    (:4
      (Transformation|AttachActorToComponent _fortpickup _skeletalmesh "Weapon5" "SnapToTarget" "SnapToTarget" "SnapToTarget" true)
      (goto _label)))
  (Transformation|AttachActorToComponent _fortpickup _skeletalmesh "Weapon1" "SnapToTarget" "SnapToTarget" "SnapToTarget" true)
  (goto _label))

(fn CanInteract
  (params (InController :type Controller%20Object%20Reference))
  (bind _returnvalue (GameplayTags|HasMatchingGameplayTag(Message) (Ability|GetAbilitySystemComponent InController) "(TagName=\"MovementMode.Sliding\")"))
  (return (not _returnvalue)))

(fn IsSeatEnabled
  (params (SeatIdx :type Integer))
  (bind _returnvalue (|CallParentFunction SeatIdx))
  (Utilities|FlowControl|Sequence
    (:then_0
      ;[1 1 1] Check to see if the vehicle is in water and the seat is currently enabled. Only continue if the vehicle is in water and the seat is enabled
      (if (not (and (Class|FortVehicleInterface|IsinWater) _returnvalue))
        (return _returnvalue)))
    (:then_1
      ;[1 1 1] If vehicle is in water, calculate how deep in the water it is based on the water level and current Z location of the cannon
      (LocalVariables|SetVehicleSubmergedWaterLevel (- (+ (CallFunction|GetWaterLevel) (Variables|Default|GetWaterVFXOffset)) (.z (Transformation|GetActorLocation)))))
    (:then_2
      ;[1 1 1] Check if the Vehicle is submerged deep enough underwater to disable all seats on the vehicle
      (if (Math|Float|InRange(Float) (LocalVariables|GetVehicleSubmergedWaterLevel) (LocalVariables|GetWaterThreshold%202) 1e+04)
        (return false)))
    (:then_3
      ;[1 1 1] Check if the Vehicle is submerged deep enough underwater to disable the Passenger seat of the Cannon. If not just return the parent value
      (if (Math|Float|InRange(Float) (LocalVariables|GetVehicleSubmergedWaterLevel) (LocalVariables|GetWaterThreshold%201) 1e+04)
        (return (!= (Variables|GetSeatIdx) 1))
        (else
          (return _returnvalue))))))

(fn CheckIfBurning
  (return (GameplayTags|HasMatchingGameplayTag self "(TagName=\"Curie.State.ElementalStatus.Burning\")")))

(fn CheckPassengers
  (return (Math|Boolean|NORBoolean (Class|ItemUIFunctionLibrary|IsNullItem (Variables|PushCannon|GetDriverPawn)) (Class|ItemUIFunctionLibrary|IsNullItem (Variables|PushCannon|GetPassengerPawn)))))

(fn ClearAudioParams
  (bind _audioupdatetimerhandle (Variables|Default|GetAudioUpdateTimerHandle))
  (if (and (Math|Boolean|NORBoolean (Utilities|IsValid (Variables|PushCannon|GetDriverPawn) :!class /Script/Engine.KismetSystemLibrary) (Utilities|IsValid (Variables|PushCannon|GetPassengerPawn) :!class /Script/Engine.KismetSystemLibrary)) (Utilities|Time|IsValidTimerHandle _audioupdatetimerhandle))
    (Utilities|Time|ClearandInvalidateTimerbyHandle _audioupdatetimerhandle)
    (Audio|Components|Audio|FadeOut (Variables|Default|GetAudioCannonAim) 0.25 0.0 "Linear" :!class /Script/Engine.AudioComponent)))

(fn HandleOnHealthChanged
  (Networking|CanExecuteCosmeticEvents
    (:True
      (CallFunction|SendDamageStatetoMaterial (- 1.0 (/ (Class|FortVehicleInterface|GetHealth) (Class|FortVehicleInterface|GetMaxHealth)))))))

(fn HandleWaterDepth
  (params (Water%20Surface%20Info :type Water%20Surface%20Info%20Structure))
  (bind _waterplanelocation (Utilities|Struct|BreakWaterSurfaceInfo Water%20Surface%20Info))
  (bind _returnvalue (Transformation|GetActorLocation))
  (bind _z (.z _returnvalue))
  (bind _z_2 (.z _waterplanelocation))
  (bind _returnvalue_2 (- _z_2 _z))
  (if (Math|Float|InRange(Float) _returnvalue_2 (LocalVariables|GetWater%20Threshold%202) 1e+04)
    ;[1 1 1] If below Threshold 2
    ; Eject all Players
    (Vehicle|EjectAllPlayers)
    (elif (Math|Float|InRange(Float) _returnvalue_2 (LocalVariables|GetWater%20Threshold%201) 1e+04)
      ;[1 1 1] If below Threshold 1
      ; Eject Player in Barrel
      (Vehicle|EjectSeat 1))))

(fn HandleWaterDepthRufus
  (params (Water%20Surface%20Height :type Float%20%28double-precision%29))
  (bind _returnvalue (Transformation|GetActorLocation))
  (bind _z (.z _returnvalue))
  (bind _returnvalue_2 (- Water%20Surface%20Height _z))
  ;[0 0.9133 1] Eject Passenger when Cannon is too Deep in Rufus Water Volumes
  (if (Math|Float|InRange(Float) _returnvalue_2 (LocalVariables|GetWater%20Threshold%202) 1e+04)
    ;;[1 1 1] If below Threshold 2
    ;; Eject all Players
    (Vehicle|EjectAllPlayers)
    (elif (Math|Float|InRange(Float) _returnvalue_2 (LocalVariables|GetWater%20Threshold%201) 1e+04)
      ;;[1 1 1] If below Threshold 1
      ;; Eject Player in Barrel
      (Vehicle|EjectSeat 1))))

(fn IsLocalPlayerRiding
  (bind _passengerpawn (Variables|PushCannon|GetPassengerPawn))
  (bind _driverpawn (Variables|PushCannon|GetDriverPawn))
  (Utilities|FlowControl|Sequence
    (:then_0
      (Utilities|IsValid _driverpawn
        (:Is%20Valid
          (if (Pawn|IsLocallyControlled _driverpawn)
            _))))
    (:then_1
      (Utilities|IsValid _passengerpawn
        (:Is%20Valid
          (if (Pawn|IsLocallyControlled _passengerpawn)
            (return true)
            (else
              _)))
        (:Is%20Not%20Valid
          _))
      (return false)))
  (return true))

(fn PlayHapticsOnHit
  (params (HitMagnitude :type Float%20%28double-precision%29))
  (bind _passengerplayercontroller (Variables|PushCannon|GetPassengerPlayerController))
  (bind _drivingplayercontroller (Variables|PushCannon|GetDrivingPlayerController))
  (bind _validplayercontrollers (LocalVariables|GetValidPlayerControllers))
  (Utilities|FlowControl|Sequence
    (:then_0
      (Utilities|IsValid _drivingplayercontroller
        (:Is%20Valid
          (Utilities|Array|Add (LocalVariables|GetValidPlayerControllers) _drivingplayercontroller))))
    (:then_1
      (Utilities|IsValid _passengerplayercontroller
        (:Is%20Valid
          (Utilities|Array|Add (LocalVariables|GetValidPlayerControllers) _passengerplayercontroller))))
    (:then_2
      (if (> (Utilities|Array|Length (LocalVariables|GetValidPlayerControllers)) 0)
        (if (> (Variables|GetHitMagnitude) 1e+04)
          (Utilities|FlowControl|Sequence
            (:then_0
              (LocalVariables|SetCameraShakeMagnitude (Math|Float|MapRangeClamped (Variables|GetHitMagnitude) 1.5e+04 5e+04 0.1 1.0))
              (Utilities|IsValid (Variables|PushCannon|GetDrivingPlayerController)
                (:Is%20Valid
                  (Game|Feedback|ClientStartCameraShake :self (LocalVariables|GetValidPlayerControllers) :Shake "/Game/Athena/DrivableVehicles/CameraShake_HitImpact_ShoppingCart.CameraShake_HitImpact_ShoppingCart_C" :Scale (LocalVariables|GetCameraShakeMagnitude) :PlaySpace "CameraLocal")
                  (if (> (Variables|GetHitMagnitude) 2e+04)
                    (Game|Feedback|ClientPlayForceFeedback _validplayercontrollers "/Game/Weapons/ForceFeedback/FF_Generic_Impact_Near.FF_Generic_Impact_Near" "None" false false false)
                    (else
                      (Game|Feedback|ClientPlayForceFeedback _validplayercontrollers "/Game/Weapons/ForceFeedback/FF_Generic_Build.FF_Generic_Build" "None" false false false))))))
            (:then_1
              (if (< (Math|Float|Absolute(Float) (Math|Vector|DotProduct (Math|Vector|Normalize (Vehicle|GetPreviousLinearVelocity)) "0, 0,1.000000")) 0.5)
                (bind _output_get (LocalVariables|SetHitDelta (- (Variables|Vehicle|GetCachedSpeed) (Math|Vector|VectorLength (Transformation|GetComponentVelocity (Variables|Vehicle|GetSkeletalMesh))))))
                (if (> _output_get (LocalVariables|GetCameraPushVelocityDelta))
                  (Game|Feedback|ClientStartCameraShake (LocalVariables|GetValidPlayerControllers) "/Game/Athena/DrivableVehicles/CameraShake_BigHit.CameraShake_BigHit_C" (Math|Float|MapRangeClamped (LocalVariables|GetHitDelta) 3e+02 9e+02 0.5 1.0) "UserDefined" (Math|Rotator|MakeRotfromX (Vehicle|GetPreviousLinearVelocity))))))))))))

(fn PrelaunchDestroyBuildings
  (params (HitActors :type Array%20of%20Hit%20Result%20Structures) (LaunchedPawn :type Fort%20Player%20Pawn%20Object%20Reference))
  (bind _hitactor (LocalVariables|GetHitActor))
  (bind _hitactor_2 (LocalVariables|GetHitActor))
  (bind _launchedpawn (Variables|GetLaunchedPawn))
  (bind _array_element (Utilities|Array|ForEachLoopwithBreak HitActors)
    (:LoopBody
      (if (Variables|Actor|GetbCanBeDamaged (.HitActor (Collision|BreakHitResult _array_element)))
        (LocalVariables|SetHitActor (.HitActor (Collision|BreakHitResult _array_element)))
        (bind _bdidactorhavetags (Class|AthenaQuestFunctionLibrary|AthenaCheckActorHasTags (.HitActor (Collision|BreakHitResult _array_element)) "(GameplayTags=((TagName=\"Deimos.Spawner\")))"))
        (if (not _bdidactorhavetags)
          (Utilities|Casting|CastToBuildingActor _hitactor_2
            (:then
              _)
            (:CastFailed
              (Utilities|Casting|CastToBuildingProp _hitactor_2
                (:then
                  _)
                (:CastFailed
                  (Utilities|Casting|CastToAthena_Prop_ParentBuildingContainerBlueprint _hitactor_2)))))
          (Utilities|Casting|CastToAthenaBarrierObjective (LocalVariables|GetHitActor)
            (:CastFailed
              (GameplayEffects|ApplyGameplayEffectToTarget :self (Ability|GetAbilitySystemComponent _launchedpawn) :GameplayEffectClass (Variables|Default|GetBuildingDestroyGE) :Target (Ability|GetAbilitySystemComponent _hitactor) :Level 0.0)
              (Collision|IgnoreActorwhenMoving (Variables|/Script/Engine.Character|GetMesh _launchedpawn) _hitactor true))))))))

(fn ResetWheelFXParams
  (bind _ns_ship_cannon_vehicle_wheeldust (Variables|Default|GetNS_Ship_Cannon_Vehicle_WheelDust))
  (Effects|Components|ParticleSystem|SetFloatParameter _ns_ship_cannon_vehicle_wheeldust "LeftSlow" 0.0)
  (Effects|Components|ParticleSystem|SetFloatParameter _ns_ship_cannon_vehicle_wheeldust "RightSlow" 0.0)
  (if (not (Fortnite|IsMobilePlatform))
    (Effects|Components|ParticleSystem|SetFloatParameter _ns_ship_cannon_vehicle_wheeldust "LeftFast" 0.0)
    (Effects|Components|ParticleSystem|SetFloatParameter _ns_ship_cannon_vehicle_wheeldust "RightFast" 0.0)
    (Effects|Components|ParticleSystem|SetFloatParameter _ns_ship_cannon_vehicle_wheeldust "LeftVFast" 0.0)
    (Effects|Components|ParticleSystem|SetFloatParameter _ns_ship_cannon_vehicle_wheeldust "RightVFast" 0.0)
    (Effects|Components|ParticleSystem|SetFloatParameter _ns_ship_cannon_vehicle_wheeldust "Fast" 0.0)))

(fn SendDamageStateToMaterial
  (params (Parameter%20Value :type Float%20%28double-precision%29))
  (bind _output_get (Variables|Default|SetSaved%20Damage%20for%20MID Parameter%20Value))
  (Rendering|Material|SetScalarParameterValueonMaterials (Variables|Vehicle|GetSkeletalMesh) "Damage" _output_get))

(fn UpdateDynamicCameraShake
  (Utilities|FlowControl|Sequence
    (:then_0
      (Variables|CameraShake|SetInAirSmoothed (Math|Interpolation|FInterpTo (Variables|CameraShake|GetInAirSmoothed) (- 1.0 (Math|Conversions|ToFloat(Boolean) (Vehicle|VehicleIsinAir))) (Utilities|Time|GetWorldDeltaSeconds) 2e+01))
      (Variables|PushCannon|SetSmoothedSpringCompression (Math|Float|Clamp(Float) (Math|Interpolation|FInterpTo (Variables|PushCannon|GetSmoothedSpringCompression) (* (Variables|Vehicle|GetCurrentMaxSpringCompression) (LocalVariables|GetSpringFudgeValue)) (Utilities|Time|GetWorldDeltaSeconds) 3.0) 0.33))
      (LocalVariables|SetShakeAmount (Math|Float|float*float (Variables|CameraShake|GetInAirSmoothed) (Math|Float|Power (Math|Float|Clamp(Float) (/ (Math|Float|Absolute(Float) (Vehicle|GetForwardSpeedKmh)) (Variables|CameraShake|GetNormalizedSpeed))) (Variables|CameraShake|GetCameraShakeSpeedCurvePow)) (ScalableFloat|GetValueatLevel (Variables|Haptics|GetScreenShakeMult))))
      (LocalVariables|SetCamerShakeAmplitude (Math|Float|Lerp (Variables|CameraShake|GetAmplitudeMin) (Variables|CameraShake|GetAmplitudeMax) (* (LocalVariables|GetShakeAmount) (Variables|PushCannon|GetSmoothedSpringCompression))))
      (LocalVariables|SetShakeFrequency (Math|Float|Lerp (Variables|CameraShake|GetScreenShakeFrequencyMin) (Variables|CameraShake|GetScreenShakeFrequencyMax) (LocalVariables|GetShakeAmount))))
    (:then_2
      (Utilities|FlowControl|Sequence
        (:then_0
          ;[1 1 1] Driver
          (Utilities|IsValid (Variables|PushCannon|GetCameraShake)
            (:Is%20Valid
              (if (Variables|PushCannon|GetbCoastingCameraShake)
                (Variables|/Script/EngineCameras.LegacyCameraShake|SetRotOscillation (Utilities|Struct|MakeROscillator (Utilities|Struct|MakeFOscillator (LocalVariables|GetCamerShakeAmplitude) (LocalVariables|GetShakeFrequency)) (Utilities|Struct|MakeFOscillator (LocalVariables|GetCamerShakeAmplitude) (* (LocalVariables|GetShakeFrequency) (Variables|CameraShake|GetScreenShakeYawFrequencyMultipier)))) (Variables|PushCannon|GetCameraShake))
                (else
                  (Variables|/Script/EngineCameras.LegacyCameraShake|SetShakeScale (/ (Vehicle|GetForwardSpeedKmh) 2e+01) (Variables|PushCannon|GetCameraShake)))))))
        (:then_1
          ;[1 1 1] Passenger
          (Utilities|IsValid (Variables|PushCannon|GetPassengerCameraShake_0)
            (:Is%20Valid
              (Variables|/Script/EngineCameras.LegacyCameraShake|SetRotOscillation (Utilities|Struct|MakeROscillator (Utilities|Struct|MakeFOscillator (* (LocalVariables|GetCamerShakeAmplitude) (Variables|PushCannon|GetPassengerCameraShakeMultiplier)) (LocalVariables|GetShakeFrequency) "PerlinNoise") (Utilities|Struct|MakeFOscillator (* (LocalVariables|GetCamerShakeAmplitude) (Variables|PushCannon|GetPassengerCameraShakeMultiplier)) (* (LocalVariables|GetShakeFrequency) (Variables|CameraShake|GetScreenShakeYawFrequencyMultipier)) "PerlinNoise")) (Variables|PushCannon|GetPassengerCameraShake_0)))))))))

(fn ConstructionScript
  (Variables|FortAthenaVehicle|SetOverlapComponent (Variables|Default|GetOverlapTrigger))
  ;[1 1 1] For animation to not lag, skeletal mesh tick group has to be after the turret's tick group
  (Components|Tick|SetTickGroup (Variables|Vehicle|GetSkeletalMesh) "TG_DuringPhysics")
  (Transformation|SetRelativeLocation (Variables|Default|GetReticle) "0, 0, 0" false true)
  (Variables|Tricks|SetbTricksEnabled false))
