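# Physics asset that defines collision for the drivable shopping cart vehicle.
# A single root body holds two capsules, one for the driver behind and one for the higher passenger position, plus a thin, slightly tilted box named as query collision for the bottom tray.
# It has no constraints, so the shapes form one rigid compound body with no articulated parts.
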PhysicsAsset /Game/Athena/DrivableVehicles/ShoppingCartVehicleCollision
bodies {root [
  {shapeType=Capsule, length=57.5732, radius=46.3665, shapeName=DriverCollision, center=<-82.334 -0.0007 98.2765>}
  {shapeType=Capsule, length=67.4782, radius=48.316, shapeName=PassengerCollision, center=<15.1923 -0.0006 135.0327>}
  {
   shapeType=Box
   extentX=99.7122
   extentY=71.5617
   extentZ=7.5
   shapeName=QueryCollisionBottomTray
   center=<12.9249 -0.0001 18.4429>
   rotation=<rot -2.8125 0 0>
  }
 ]}
constraints []
