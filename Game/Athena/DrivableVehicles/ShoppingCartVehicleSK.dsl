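; Drivable two-seat shopping cart vehicle Blueprint, covering feedback, damage visuals, quest updates and launching of hit pawns.
; Seat enter/exit tracks driver and passenger, feeding knockback credit and camera shakes.
; Tick and collision events drive audio, FX, rumble and shakes scaled by speed, springs and impacts.
; Water events spawn Niagara FX, with a repeating timer switching shallow foam and deep bubbles by depth below the surface.
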(event EventOnEnteredWaterBody
  (params (WaterBody :type Fort%20Water%20Body%20Actor%20Object%20Reference))
  (bind _watersurfacedatacache (Variables|FortAthenaVehicle|GetWaterSurfaceDataCache))
  (bind (_waterplanelocation _waterplanenormal _watersurfacelocation _waterdepth _waterbodyidx) (Utilities|Struct|BreakWaterSurfaceInfo _watersurfacedatacache))
  (bind (_returnvalue_x _returnvalue_y _returnvalue_z) (Math|Vector|BreakVector (Transformation|GetActorLocation)))
  (bind (_returnvalue_x_2 _returnvalue_y_2 _returnvalue_z_2) (Math|Vector|BreakVector (Transformation|GetActorLocation)))
  (bind _inwaterfx (Variables|WaterInteraction|GetInWaterFX))
  ;[0.3146 0.5848 1] FX for water ineraction
  (Variables|Default|SetWater%20Body WaterBody)
  (Networking|CanExecuteCosmeticEvents
    (:True
      (if (> (Rendering|GetRenderingDetailMode) 0)
        (Utilities|FlowControl|DoOnce false
          (:Completed
            (Audio|PlaySoundatLocation "/Game/Athena/Sounds/Vehicles/ShoppingCart/Shopping_Water_Impact_Cue.Shopping_Water_Impact_Cue" (Transformation|GetActorLocation)) ; re adding this after it was nuked
            (bind _returnvalue (Utilities|Time|SetTimerbyEvent (delegate Custom|CheckWater) 0.2 true))
            (Variables|WaterInteraction|SetTimer%20Snap%20Foam%20to%20Water%20Surface _returnvalue)
            (if (not (Networking|IsServer))
              (bind _returnvalue_2 (Niagara|SpawnSystemAttached "/Game/Athena/DrivableVehicles/FX/NS_ShoppingCart/NS_Shopping_Cart_Vehicle_InWater.NS_Shopping_Cart_Vehicle_InWater" (Variables|Vehicle|GetSkeletalMesh) "WaterTest" "0, 0, 0" "0, 0, 0" "KeepRelativeOffset" true))
              (Variables|WaterInteraction|SetInWaterFX _returnvalue_2)
              (Utilities|FlowControl|Sequence
                (:then_0
                  ;;[1 1 1] Trigger large splash base on falling speed
                  (Niagara|SetNiagaraVariable(Float) _inwaterfx "LoopFoamSpawn" 0.0)
                  (Variables|Default|SetVehicle%20Fall%20Speed _returnvalue_z)
                  (Utilities|FlowControl|Delay 0.08)
                  (if (> (Math|Float|Absolute(Float) (- (Variables|Default|GetVehicle%20Fall%20Speed) _returnvalue_z_2)) 1e+01)
                    (Niagara|SetNiagaraVariable(Int32) _inwaterfx "SplashSpawn" 1)
                    (Niagara|SetNiagaraVariable(Float) _inwaterfx "BubbleSpawn" 0.0)
                    (Niagara|SetNiagaraVariable(Int32) _inwaterfx "BigBubbleSpawn" 0)
                    (Niagara|SetNiagaraVariable(Vector3) _inwaterfx "WaterPlaneLocation" (Math|Vector|MakeVector (.x (Transformation|GetActorLocation)) (.y (Transformation|GetActorLocation)) (CallFunction|GetWaterLevel)))
                    (Niagara|SetNiagaraVariable(Float) _inwaterfx "WaterDepth" _waterdepth)
                    (Niagara|SetNiagaraVariable(Vector3) _inwaterfx "WaterPlaneNormal" _waterplanenormal)
                    (Niagara|SetNiagaraVariable(Float) _inwaterfx "WaterBodyIndex" (Math|Conversions|ToFloat(Integer) _waterbodyidx))
                    (else
                      (Niagara|SetNiagaraVariable(Int32) _inwaterfx "SplashSpawn" 0)
                      (Niagara|SetNiagaraVariable(Float) _inwaterfx "BubbleSpawn" 0.0)
                      (Niagara|SetNiagaraVariable(Int32) _inwaterfx "BigBubbleSpawn" 0))))
                (:then_1
                  ;;[1 1 1] Deactivate Water bubble FX while sinking to certain point and no driver in vehicle
                  (Utilities|FlowControl|Delay 12.0)
                  (Variables|Default|GetDriverPawn
                    (:then)
                    (:else
                      (Components|Activation|Deactivate _inwaterfx)
                      (Utilities|Time|ClearandInvalidateTimerbyHandle (Variables|WaterInteraction|GetTimer%20Snap%20Foam%20to%20Water%20Surface))
                      (Development|PrintString "DeactiveWaterFXwhenDriverIsNotValid"))))))))))))

(event Custom|CE_PawnEnterVehicleWhileInWater
  (bind _watersurfacedatacache (Variables|FortAthenaVehicle|GetWaterSurfaceDataCache))
  (bind (_waterplanelocation _waterplanenormal _watersurfacelocation _waterdepth _waterbodyidx) (Utilities|Struct|BreakWaterSurfaceInfo _watersurfacedatacache))
  (bind (_returnvalue_x _returnvalue_y _returnvalue_z) (Math|Vector|BreakVector (Transformation|GetActorLocation)))
  (bind (_returnvalue_x_2 _returnvalue_y_2 _returnvalue_z_2) (Math|Vector|BreakVector (Transformation|GetActorLocation)))
  (bind _inwaterfx (Variables|WaterInteraction|GetInWaterFX))
  ;[0.3146 0.5848 1] FX for water ineraction
  (if (CallFunction|IsinWaterBody)
    (bind _inwaterfx_2 (Variables|WaterInteraction|GetInWaterFX)
      (:then
        (Components|Activation|Activate _inwaterfx_2 false))
      (:else
        (bind _returnvalue (Utilities|Time|SetTimerbyEvent (delegate Custom|CheckWater) 0.2 true))
        (Variables|WaterInteraction|SetTimer%20Snap%20Foam%20to%20Water%20Surface _returnvalue)
        (if (not (Networking|IsServer))
          (bind _returnvalue_2 (Niagara|SpawnSystemAttached "/Game/Athena/DrivableVehicles/FX/NS_ShoppingCart/NS_Shopping_Cart_Vehicle_InWater.NS_Shopping_Cart_Vehicle_InWater" (Variables|Vehicle|GetSkeletalMesh) "WaterTest" "0, 0, 0" "0, 0, 0" "KeepRelativeOffset" true))
          (Variables|WaterInteraction|SetInWaterFX _returnvalue_2)
          (Utilities|FlowControl|Sequence
            (:then_0
              ;;[1 1 1] Trigger large splash base on falling speed
              (Niagara|SetNiagaraVariable(Float) _inwaterfx "LoopFoamSpawn" 0.0)
              (Variables|Default|SetVehicle%20Fall%20Speed _returnvalue_z)
              (Utilities|FlowControl|Delay 0.08)
              (if (> (Math|Float|Absolute(Float) (- (Variables|Default|GetVehicle%20Fall%20Speed) _returnvalue_z_2)) 1e+01)
                (Niagara|SetNiagaraVariable(Int32) _inwaterfx "SplashSpawn" 1)
                (Niagara|SetNiagaraVariable(Float) _inwaterfx "BubbleSpawn" 0.0)
                (Niagara|SetNiagaraVariable(Int32) _inwaterfx "BigBubbleSpawn" 0)
                (Niagara|SetNiagaraVariable(Vector3) _inwaterfx "WaterPlaneLocation" (Math|Vector|MakeVector (.x (Transformation|GetActorLocation)) (.y (Transformation|GetActorLocation)) (CallFunction|GetWaterLevel)))
                (Niagara|SetNiagaraVariable(Float) _inwaterfx "WaterDepth" _waterdepth)
                (Niagara|SetNiagaraVariable(Vector3) _inwaterfx "WaterPlaneNormal" _waterplanenormal)
                (Niagara|SetNiagaraVariable(Float) _inwaterfx "WaterBodyIndex" (Math|Conversions|ToFloat(Integer) _waterbodyidx))
                (else
                  (Niagara|SetNiagaraVariable(Int32) _inwaterfx "SplashSpawn" 0)
                  (Niagara|SetNiagaraVariable(Float) _inwaterfx "BubbleSpawn" 0.0)
                  (Niagara|SetNiagaraVariable(Int32) _inwaterfx "BigBubbleSpawn" 0))))
            (:then_1
              ;;[1 1 1] Deactivate Water bubble FX while sinking to certain point and no driver in vehicle
              (Utilities|FlowControl|Delay 12.0)
              (Variables|Default|GetDriverPawn
                (:then)
                (:else
                  (Components|Activation|Deactivate _inwaterfx)
                  (Utilities|Time|ClearandInvalidateTimerbyHandle (Variables|WaterInteraction|GetTimer%20Snap%20Foam%20to%20Water%20Surface))
                  (Development|PrintString "DeactiveWaterFXwhenDriverIsNotValid"))))))))))

(event Custom|CheckWater
  (bind _watersurfacedatacache (Variables|FortAthenaVehicle|GetWaterSurfaceDataCache))
  (bind (_waterplanelocation _waterplanenormal _watersurfacelocation _waterdepth _waterbodyidx) (Utilities|Struct|BreakWaterSurfaceInfo _watersurfacedatacache))
  (bind _returnvalue (Transformation|GetActorLocation))
  (bind _z (.z _returnvalue))
  (bind _z_2 (.z _waterplanelocation))
  (bind _returnvalue_2 (- _z_2 _z))
  (bind _inwaterfx (Variables|WaterInteraction|GetInWaterFX))
  (bind _inwaterfx_2 (Variables|WaterInteraction|GetInWaterFX))
  ;[0.3146 0.5848 1] FX for water ineraction
  (if (Math|Float|InRange(Float) _returnvalue_2 -8e+01 8e+02)
    (bind _inwaterfx_3 (Variables|WaterInteraction|GetInWaterFX)
      (:then
        (if (Math|Float|InRange(Float) _returnvalue_2 -2e+01 1e+02)
          ;;[1 1 1] Shallow Water
          (Variables|Default|GetDriverPawn
            (:then
              (Components|Activation|Activate _inwaterfx_3 false))
            (:else
              _))
          (Niagara|SetNiagaraVariable(Int32) _inwaterfx_3 "SplashSpawn" 0)
          (Niagara|SetNiagaraVariable(Float) _inwaterfx_3 "BubbleSpawn" 1.0)
          (Niagara|SetNiagaraVariable(Float) _inwaterfx_3 "LoopFoamSpawn" 1.0)
          (Niagara|SetNiagaraVariable(Int32) _inwaterfx_3 "BigBubbleSpawn" 0)
          (CallFunction|CEWaterFXResetDeepWaterEffectActivate)
          (Utilities|FlowControl|Sequence
            (:then_0
              ;;[1 1 1] Foam Snap Function
              (Niagara|SetNiagaraVariable(Vector3) _inwaterfx_2 "WaterPlaneLocation" (Math|Vector|MakeVector (.x (Transformation|GetActorLocation)) (.y (Transformation|GetActorLocation)) (CallFunction|GetWaterLevel))))
            (:then_1
              (Utilities|FlowControl|DoOnce false
                (:Completed
                  (Niagara|SetNiagaraVariable(Float) _inwaterfx_2 "WaterDepth" _waterdepth)
                  (Niagara|SetNiagaraVariable(Vector3) _inwaterfx_2 "WaterPlaneNormal" _waterplanenormal)
                  (Niagara|SetNiagaraVariable(Float) _inwaterfx_2 "WaterBodyIndex" (Math|Conversions|ToFloat(Integer) _waterbodyidx))))))
          (else
            ;;[1 1 1] Deep Water
            (Utilities|FlowControl|Sequence
              (:then_0
                (Utilities|FlowControl|DoOnce false
                  (:Completed
                    (Variables|Default|GetDriverPawn
                      (:then
                        (Components|Activation|Activate _inwaterfx true))
                      (:else)))))
              (:then_1
                (Niagara|SetNiagaraVariable(Int32) _inwaterfx "SplashSpawn" 0)
                (Niagara|SetNiagaraVariable(Float) _inwaterfx "BubbleSpawn" 1.0)
                (Niagara|SetNiagaraVariable(Float) _inwaterfx "LoopFoamSpawn" 0.0)
                (Niagara|SetNiagaraVariable(Int32) _inwaterfx "BigBubbleSpawn" 1))))))
      (:else))
    (else
      (Components|Activation|Deactivate _inwaterfx))))

(event Custom|CE_ResetWaterFXSnapping
  (bind _watersurfacedatacache (Variables|FortAthenaVehicle|GetWaterSurfaceDataCache))
  (bind (_waterplanelocation _waterplanenormal _watersurfacelocation _waterdepth _waterbodyidx) (Utilities|Struct|BreakWaterSurfaceInfo _watersurfacedatacache))
  (bind _inwaterfx (Variables|WaterInteraction|GetInWaterFX))
  ;[0.3146 0.5848 1] FX for water ineraction
  ;;[1 1 1] Foam Snap Function
  (Utilities|FlowControl|DoOnce false
    (:Completed
      (Niagara|SetNiagaraVariable(Float) _inwaterfx "WaterDepth" _waterdepth)
      (Niagara|SetNiagaraVariable(Vector3) _inwaterfx "WaterPlaneNormal" _waterplanenormal)
      (Niagara|SetNiagaraVariable(Float) _inwaterfx "WaterBodyIndex" (Math|Conversions|ToFloat(Integer) _waterbodyidx)))))

(event Custom|CE_WaterFX_ResetDeepWaterEffectActivate
  ;[0.3146 0.5848 1] FX for water ineraction
  ;;[1 1 1] Deep Water
  (Utilities|FlowControl|DoOnce false
    (:Completed
      (Variables|Default|GetDriverPawn
        (:then
          (Components|Activation|Activate (Variables|WaterInteraction|GetInWaterFX) true))
        (:else)))))

(event EventOnExitedWaterBody
  (params (WaterBody :type Fort%20Water%20Body%20Actor%20Object%20Reference))
  (bind _watersurfacedatacache (Variables|FortAthenaVehicle|GetWaterSurfaceDataCache))
  (bind (_waterplanelocation _waterplanenormal _watersurfacelocation _waterdepth _waterbodyidx) (Utilities|Struct|BreakWaterSurfaceInfo _watersurfacedatacache))
  (bind (_returnvalue_x _returnvalue_y _returnvalue_z) (Math|Vector|BreakVector (Transformation|GetActorLocation)))
  (bind (_returnvalue_x_2 _returnvalue_y_2 _returnvalue_z_2) (Math|Vector|BreakVector (Transformation|GetActorLocation)))
  (bind _inwaterfx (Variables|WaterInteraction|GetInWaterFX))
  ;[0.3146 0.5848 1] FX for water ineraction
  (Networking|CanExecuteCosmeticEvents
    (:True
      (Utilities|Time|ClearandInvalidateTimerbyHandle (Variables|WaterInteraction|GetTimer%20Snap%20Foam%20to%20Water%20Surface))
      (Utilities|FlowControl|Sequence
        (:then_0
          (bind _inwaterfx_2 (Variables|WaterInteraction|GetInWaterFX)
            (:then
              (Components|Activation|Deactivate _inwaterfx_2))
            (:else)))
        (:then_1
          (CallFunction|CEResetWaterFXSnapping)
          (Variables|Default|GetDriverPawn
            (:then
              (Utilities|FlowControl|DoOnce false
                (:Completed
                  (Audio|PlaySoundatLocation "/Game/Athena/Sounds/Vehicles/ShoppingCart/Shopping_Water_Impact_Cue.Shopping_Water_Impact_Cue" (Transformation|GetActorLocation)) ; re adding this after it was nuked
                  (bind _returnvalue (Utilities|Time|SetTimerbyEvent (delegate Custom|CheckWater) 0.2 true))
                  (Variables|WaterInteraction|SetTimer%20Snap%20Foam%20to%20Water%20Surface _returnvalue)
                  (if (not (Networking|IsServer))
                    (bind _returnvalue_2 (Niagara|SpawnSystemAttached "/Game/Athena/DrivableVehicles/FX/NS_ShoppingCart/NS_Shopping_Cart_Vehicle_InWater.NS_Shopping_Cart_Vehicle_InWater" (Variables|Vehicle|GetSkeletalMesh) "WaterTest" "0, 0, 0" "0, 0, 0" "KeepRelativeOffset" true))
                    (Variables|WaterInteraction|SetInWaterFX _returnvalue_2)
                    (Utilities|FlowControl|Sequence
                      (:then_0
                        ;;[1 1 1] Trigger large splash base on falling speed
                        (Niagara|SetNiagaraVariable(Float) _inwaterfx "LoopFoamSpawn" 0.0)
                        (Variables|Default|SetVehicle%20Fall%20Speed _returnvalue_z)
                        (Utilities|FlowControl|Delay 0.08)
                        (if (> (Math|Float|Absolute(Float) (- (Variables|Default|GetVehicle%20Fall%20Speed) _returnvalue_z_2)) 1e+01)
                          (Niagara|SetNiagaraVariable(Int32) _inwaterfx "SplashSpawn" 1)
                          (Niagara|SetNiagaraVariable(Float) _inwaterfx "BubbleSpawn" 0.0)
                          (Niagara|SetNiagaraVariable(Int32) _inwaterfx "BigBubbleSpawn" 0)
                          (Niagara|SetNiagaraVariable(Vector3) _inwaterfx "WaterPlaneLocation" (Math|Vector|MakeVector (.x (Transformation|GetActorLocation)) (.y (Transformation|GetActorLocation)) (CallFunction|GetWaterLevel)))
                          (Niagara|SetNiagaraVariable(Float) _inwaterfx "WaterDepth" _waterdepth)
                          (Niagara|SetNiagaraVariable(Vector3) _inwaterfx "WaterPlaneNormal" _waterplanenormal)
                          (Niagara|SetNiagaraVariable(Float) _inwaterfx "WaterBodyIndex" (Math|Conversions|ToFloat(Integer) _waterbodyidx))
                          (else
                            (Niagara|SetNiagaraVariable(Int32) _inwaterfx "SplashSpawn" 0)
                            (Niagara|SetNiagaraVariable(Float) _inwaterfx "BubbleSpawn" 0.0)
                            (Niagara|SetNiagaraVariable(Int32) _inwaterfx "BigBubbleSpawn" 0))))
                      (:then_1
                        ;;[1 1 1] Deactivate Water bubble FX while sinking to certain point and no driver in vehicle
                        (Utilities|FlowControl|Delay 12.0)
                        (Variables|Default|GetDriverPawn
                          (:then)
                          (:else
                            (Components|Activation|Deactivate _inwaterfx)
                            (Utilities|Time|ClearandInvalidateTimerbyHandle (Variables|WaterInteraction|GetTimer%20Snap%20Foam%20to%20Water%20Surface))
                            (Development|PrintString "DeactiveWaterFXwhenDriverIsNotValid")))))))))
            (:else)))
        (:then_2
          (if (not (CallFunction|IsOverlappingWaterBody))
            (Variables|Default|SetWater%20Body 0)))))))

(event Custom|PedalForceFeedback
  (bind _drivingplayercontroller (Variables|Default|GetDrivingPlayerController))
  ;[1 1 1] Force Feedback for "Pedaling" (Called from AnimBP)
  (Utilities|IsValid _drivingplayercontroller
    (:Is%20Valid
      (Game|Feedback|ClientPlayForceFeedback _drivingplayercontroller "/Game/Weapons/ForceFeedback/FF_ShoppingCartPedal.FF_ShoppingCartPedal" "None" false false false)
      ;[1 1 1] Camera shake on push
      (CameraShakes|StartCameraShake :self (Variables|/Script/Engine.PlayerController|GetPlayerCameraManager (Variables|Default|GetDrivingPlayerController)) :ShakeClass "/Game/Athena/DrivableVehicles/CameraShake_Pedal.CameraShake_Pedal_C" :Scale 1.0 :PlaySpace "CameraLocal"))))

(event EventOnPawnExitVehicle
  (params (PlayerPawn :type Fort%20Player%20Pawn%20Object%20Reference) (ExitSocketName :type Name :default "None"))
  (bind _passengerplayercontroller (Variables|Default|GetPassengerPlayerController))
  ;[0 0.1875 0.1503] EVENT -> Pawn Exit Vehicle
  (if (Utilities|Equal(Object) PlayerPawn (Variables|Default|GetDriverPawn))
    ;; ManageStateDriverExit
    (Variables|Default|SetDriverPawn 0)
    (Variables|Default|SetLastDriver PlayerPawn)
    (Variables|Default|SetIgnoreHitLastDriver true)
    (Utilities|Time|SetTimerbyEvent (EventDispatchers|CreateEvent :!event CanHitLastDriver) 3.0)
    (Utilities|FlowControl|Sequence
      (:then_0
        (CallFunction|StopDriverRumble)
        (CallFunction|StopDriverCameraShake))
      (:then_1
        (Variables|Default|SetDrivingPlayerController 0)))
    (else
      ;; ManageStatePassengerExit
      (Utilities|FlowControl|Sequence
        (:then_0
          (Utilities|IsValid (Variables|Default|GetPassengerCameraShake)
            (:Is%20Valid
              (Utilities|IsValid _passengerplayercontroller
                (:Is%20Valid
                  (CameraShakes|StopCameraShake (Variables|/Script/Engine.PlayerController|GetPlayerCameraManager _passengerplayercontroller) (Variables|Default|GetPassengerCameraShake) false)
                  (Variables|Default|SetPassengerCameraShake 0))))))
        (:then_1
          (if (Utilities|Equal(Object) (Variables|Default|GetPassengerPawn) PlayerPawn)
            (Class|PassengerImpulseWeaponComponent|SetPassengerPawn (Variables|Default|GetImpulseComponent) 0 :!class /Game/Athena/DrivableVehicles/PassengerImpulseWeaponComponent.PassengerImpulseWeaponComponent_C)))
        (:then_2
          (Variables|Default|SetPassengerPlayerController 0)
          (Variables|Default|SetPassengerPawn 0))))))

(event EventTick
  (params (DeltaSeconds :type Float%20%28single-precision%29))
  ;[1 1 1] EVENT TICK -> Wait for Rocket fire, update Audio, Haptics and prop destruction trace
  (Utilities|FlowControl|Sequence
    (:then_0
      (CallFunction|UpdateMovementCosmeticParameters))
    (:then_1
      ;; HapticsOnTick
      (if (CallFunction|IsLocalPlayerRiding)
        (Utilities|FlowControl|Sequence
          (:then_0
            (Utilities|IsValid (Variables|Default|GetDrivingPlayerController)
              (:Is%20Valid
                (Game|Feedback|PlayDynamicForceFeedback (Variables|Default|GetDrivingPlayerController) (* (Math|Float|float+float (Variables|Vehicle|GetSkiddingAmountParam) (* (Variables|Vehicle|GetScrapingAmountParam) (Math|Float|MapRangeClamped (Vehicle|GetForwardSpeedKmh) 0.0 3e+01 0.005 0.02)) (* (Math|Float|Max(Float) (Variables|Vehicle|GetSparksLeftParam) (Variables|Vehicle|GetSparksRightParam)) (Variables|Default|GetSparksRumbleMulti))) (Variables|Haptics|GetRumbleMulti)) -1 false true false true))))
          (:then_1
            (CallFunction|UpdateDynamicCameraShake)))))))

(event EventOnPawnEnterVehicle
  (params (PlayerPawn :type Fort%20Player%20Pawn%20Object%20Reference) (SeatIdx :type Integer))
  ;[0 0.1875 0.1503] On Player enter, check for passenger.
  (switch int SeatIdx
    (:0
      ;; ManageStateDriverEnter
      (bind _output_get (Variables|Default|SetDriverPawn PlayerPawn))
      (bind _asplayer_controller (Utilities|Casting|CastToPlayerController (Pawn|GetController _output_get)))
      (Variables|Default|SetDrivingPlayerController _asplayer_controller)
      (CallFunction|CanHitLastDriver)
      (Variables|Default|SetPassengerIsInstigator false)
      (CallFunction|StartDriverRumble))
    (:1
      ;; ManageStatePassengerEnter
      (Variables|Default|SetPassengerPawn PlayerPawn)
      (Utilities|FlowControl|Sequence
        (:then_0
          (Class|PassengerImpulseWeaponComponent|SetPassengerPawn (Variables|Default|GetImpulseComponent) (Variables|Default|GetPassengerPawn) :!class /Game/Athena/DrivableVehicles/PassengerImpulseWeaponComponent.PassengerImpulseWeaponComponent_C))
        (:then_1
          (bind _passengerpawn (Variables|Default|GetPassengerPawn)
            (:then
              (bind _asplayer_controller_2 (Utilities|Casting|CastToPlayerController (Pawn|GetController _passengerpawn)))
              (bind _output_get_2 (Variables|Default|SetPassengerPlayerController _asplayer_controller_2))
              (bind _returnvalue (CameraShakes|StartCameraShake :self (Variables|/Script/Engine.PlayerController|GetPlayerCameraManager _output_get_2) :ShakeClass "/Game/Athena/DrivableVehicles/Speed_CameraShake.Speed_CameraShake_C" :Scale 1.0 :PlaySpace "CameraLocal"))
              (Variables|Default|SetPassengerCameraShake (CameraShakes|ConvLegacyCameraShake _returnvalue)))
            (:else)))
        (:then_2
          _))))
  ;[1 1 1] Check if in Creative Mode
  (bind _asfort_game_state_athena (Utilities|Casting|CastToFortGameStateAthena (Game|GetGameState)))
  (if (not (GameplayTags|HasTag (Class|FortGameStateAthena|GetAthenaPlaylistContextTags _asfort_game_state_athena) "(TagName=\"Athena.Creative\")"))
    ;[1 1 1] Check for Quest to use Cart
    (bind _asathena_player_controller (Utilities|Casting|CastToAthena_PlayerController (Pawn|GetController PlayerPawn)))
    (bind (_questisvalid _questcompleted _objectivecompleted) (Default|AthenaCheckQuestandObjectiveCompletion _asathena_player_controller (Variables|QuestStuff|GetQuestItem) (Variables|QuestStuff|GetObjBackendName)))
    (if (and _questisvalid (not _objectivecompleted))
      (Quest|GiveQuestUpdatetoPlayers :Quest%20Reference (Variables|QuestStuff|GetQuestItem) :Quest%20Objective%20Backend%20Name (Variables|QuestStuff|GetObjBackendName) :PlayerControllersForUpdate (Utilities|Array|MakeArray 0)))))

(event EventOnCollisionHitEffects
  (params (HitLocation :type Vector%20%28by%20ref%29 :default "0, 0, 0") (HitNormalImpulse :type Vector%20%28by%20ref%29 :default "0, 0, 0") (HitFrictionImpulse :type Vector%20%28by%20ref%29 :default "0, 0, 0") (HitNormal :type Vector%20%28by%20ref%29 :default "0, 0, 0") (HitActor :type Actor%20Object%20Reference) (HitSurfaceType :type EPhysicalSurface%20Enum :default "SurfaceType_Default"))
  ;[0 0 0] Audio/FX Collision
  (bind (_outresult _outmagnitude) (Audio|OnCollision (Variables|Default|GetFortCollisionAudio) HitLocation HitNormalImpulse))
  (if _outresult
    (CallFunction|PlayHapticsonHit _outmagnitude)
    (if (> _outmagnitude 1.5e+04)
      (Utilities|FXSpawnImpactEffect HitActor HitLocation HitNormal _outmagnitude))))

(event Custom|UpdateMovementCosmeticParameters
  ;[0 0 0] Audio Modulation (Movement, Coast, Skid, In-Air, Scrape) + FX
  (if (< (Variables|Vehicle|GetMovementAmountParam) 0.2)
    (Audio|Components|Audio|FadeOut (Variables|Default|GetFortLayeredAudio) 0.25 0.0 "Linear" :!class /Script/Engine.AudioComponent)
    (else
      (Utilities|FlowControl|Sequence
        (:then_0
          (Utilities|FXUpdateLoopingEffects))
        (:then_1
          (Utilities|AudioUpdateLoopingEffects))))))

(event EventPawnHitByVehicle
  (params (Pawn :type Fort%20Pawn%20Object%20Reference) (ImpactPoint :type Vector%20%28by%20ref%29 :default "0, 0, 0") (ImpactNormal :type Vector%20%28by%20ref%29 :default "0, 0, 0") (LaunchVector :type Vector :default "0, 0, 0") (bTeleportPawn :type Boolean) (TeleportPawnDistance :type Float%20%28single-precision%29))
  ;[1 1 1] EVENT HIT -> Launch Player hit by Cart
  (if (> (ScalableFloat|GetValueatLevel (Variables|Default|GetCanLaunchPlayer)) 0.0)
    (bind _asfort_player_pawn_athena (Utilities|Casting|CastToFortPlayerPawnAthena Pawn)
      (:then
        (if (Math|Boolean|ORBoolean (Utilities|Equal(Object) _asfort_player_pawn_athena (Vehicle|GetDriver)) (Utilities|Equal(Object) _asfort_player_pawn_athena (Variables|Default|GetPassengerPawn)) (Utilities|Equal(Object) _asfort_player_pawn_athena (Variables|Default|GetLastDriver)) (Utilities|Equal(Object) _asfort_player_pawn_athena (Variables|Default|GetLastHitPlayer)) (Parachute|IsSkydiving _asfort_player_pawn_athena))
          (if (and (Utilities|Equal(Object) (Variables|Default|GetLastDriver) _asfort_player_pawn_athena) (Variables|Default|GetIgnoreHitLastDriver))
            _
            (else
              _)))
        (Character|LaunchCharacter _asfort_player_pawn_athena LaunchVector true true) ; Launch player based on cart velocity
        (for _array_element (Vehicle|GetAllPassengers)
          (GameplayCue|ExecuteGameplayCue%28Burst%29Local _array_element (Variables|Default|GetHitPlayerCue) (Variables|Default|GetGCParamsEmpty))
          (Variables|Default|SetLastHitPlayer _asfort_player_pawn_athena)
          (if (Vehicle|HasDriver)
            (Class|FortPlayerPawnAthena|SetFallInstigator :self _asfort_player_pawn_athena :NewFallInstigator (Game|GetInstigatorController (Variables|Default|GetDriverPawn)))
            (elif (Variables|Default|GetPassengerIsInstigator)
              (Class|FortPlayerPawnAthena|SetFallInstigator :self _asfort_player_pawn_athena :NewFallInstigator (Game|GetInstigatorController (Variables|Default|GetPassengerPawn)))
              (else
                (Class|FortPlayerPawnAthena|SetFallInstigator :self _asfort_player_pawn_athena :NewFallInstigator (Game|GetInstigatorController (Variables|Default|GetLastDriver))))))
          (Utilities|FlowControl|Delay 0.2)
          (Variables|Default|SetLastHitPlayer 0)))
      (:CastFailed
        (bind _asfort_aipawn (Utilities|Casting|CastToFortAIPawn Pawn))
        (Components|Movement|StopMovementImmediately (Variables|/Script/Engine.Character|GetCharacterMovement _asfort_aipawn))
        (Character|LaunchCharacter _asfort_aipawn LaunchVector true true) ; Launch player based on cart velocity
        (Networking|SwitchHasAuthority
          (:Remote
            (for _array_element_2 (Vehicle|GetAllPassengers)
              (GameplayCue|ExecuteGameplayCue%28Burst%29Local _array_element_2 (Variables|Default|GetHitFiendCue) (Variables|Default|GetGCParamsEmpty)))))))))

(event EventOnDamaged
  (params (Damage :type Float%20%28single-precision%29) (DamageTags :type Gameplay%20Tag%20Container%20Structure%20%28by%20ref%29) (EffectContext :type Gameplay%20Effect%20Context%20Handle%20Structure%20%28by%20ref%29) (EventInstigator :type Controller%20Object%20Reference) (DamageCauser :type Actor%20Object%20Reference))
  ;[1 1 1] Update cart material based on remaining health %
  (CallFunction|UpdateDamageState (Math|Float|Lerp 1.0 0.0 (/ (Class|FortVehicleInterface|GetHealth) (Class|FortVehicleInterface|GetMaxHealth)))))

(event Custom|Update%20Damage%20State
  (params (Damage :type Float%20%28double-precision%29))
  ;[1 1 1] Update cart material based on remaining health %
  (Rendering|Material|SetScalarParameterValueonMaterials (Variables|Vehicle|GetSkeletalMesh) "Damage" Damage))

(event EventOnDeathPlayEffects
  (params (Damage :type Float%20%28single-precision%29) (DamageTags :type Gameplay%20Tag%20Container%20Structure%20%28by%20ref%29) (Momentum :type Vector :default "0, 0, 0") (HitInfo :type Hit%20Result%20Structure%20%28by%20ref%29) (InstigatedBy :type Fort%20Pawn%20Object%20Reference) (DamageCauser :type Actor%20Object%20Reference) (EffectContext :type Gameplay%20Effect%20Context%20Handle%20Structure))
  ;[1 1 1] Play Death Effects
  (Utilities|FlowControl|Sequence
    (:then_0
      (Utilities|FXSpawnDeathEffect))
    (:then_1
      (Audio|PlaySoundatLocation (Variables|Default|GetOnDeathSound) (Transformation|GetActorLocation))
      (Development|SetHiddeninGame (Variables|Vehicle|GetSkeletalMesh) true false))))

(event EventEndPlay
  (params (EndPlayReason :type EEndPlayReason%20Enum :default "Destroyed"))
  ;[1 1 1] EVENT END PLAY
  (Utilities|IsValid (Variables|Default|GetDriverPawn)))

(event EventOnHoldExitStarted
  (params (ExitingPawn :type Fort%20Pawn%20Object%20Reference) (ExitDuration :type Float%20%28single-precision%29))
  ;[1 1 1] EVENT -> OnHoldExitStarted/Stopped
  (Ability|SendGameplayEventtoActor ExitingPawn "(TagName=\"Athena.Vehicle.Biplane.OnHoldExitStarted\")" (Utilities|Struct|MakeGameplayEventData :ContextHandle "()" :InstigatorTags "(GameplayTags=)" :TargetTags "(GameplayTags=)" :EventMagnitude ExitDuration :TargetData "()")))

(event EventOnSpringCompression
  (bind _drivingplayercontroller (Variables|Default|GetDrivingPlayerController))
  ;[1 1 1] OnSpringCompression - play force feedback
  ;; PlayHapticsOnSpringCompression
  (Utilities|IsValid _drivingplayercontroller
    (:Is%20Valid
      (Game|Feedback|ClientPlayForceFeedback _drivingplayercontroller "/Game/Weapons/ForceFeedback/FF_ShoppingCartBump.FF_ShoppingCartBump" "None" false false false))))

(event EventOnHoldExitStopped
  (params (ExitingPawn :type Fort%20Pawn%20Object%20Reference))
  ;[1 1 1] EVENT -> OnHoldExitStarted/Stopped
  (Ability|SendGameplayEventtoActor ExitingPawn "(TagName=\"Athena.Vehicle.Biplane.OnHoldExitStopped\")"))

(event EventBeginPlay
  ;[1 1 1] EVENT BEGIN PLAY
  (Variables|Default|SetIgnoreHitLastDriver true)
  (Variables|Default|SetPassengerIsInstigator true)
  (Default|AssignOnPassengerFiredImpulseWeapon (Variables|Default|GetImpulseComponent) (delegate Custom|OnPassengerFiredImpulseWeapon))
  (Utilities|FXBeginPlay))

(event Custom|OnPassengerFiredImpulseWeapon
  (Variables|Default|SetPassengerIsInstigator true))

(fn AttachWeaponToCart
  (params (Selection :type Integer))
  (bind (_returnvalue_roll _returnvalue_pitch _returnvalue_yaw) (Math|Rotator|BreakRotator (Transformation|GetActorRotation (Variables|Default|GetFortPickup))))
  (bind _skeletalmesh (Variables|Vehicle|GetSkeletalMesh))
  (bind _fortpickup (Variables|Default|GetFortPickup))
  (switch int Selection
    (:Default
      _)
    (:0
      _)
    (:1
      (Transformation|AttachActorToComponent _fortpickup _skeletalmesh "Weapon2" "SnapToTarget" "SnapToTarget" "SnapToTarget" true)
      (Transformation|SetActorRotation (Variables|Default|GetFortPickup) (Math|Rotator|MakeRotator (Math|Random|RandomFloatinRange (+ _returnvalue_roll -1e+02) (+ _returnvalue_roll -8e+01)) (Math|Random|RandomFloatinRange (+ _returnvalue_pitch -5.0) (+ _returnvalue_pitch 5.0)) (Math|Random|RandomFloatinRange (+ _returnvalue_yaw -1e+02) (+ _returnvalue_yaw -8e+01))) false))
    (:2
      (Transformation|AttachActorToComponent _fortpickup _skeletalmesh "Weapon3" "SnapToTarget" "SnapToTarget" "SnapToTarget" true))
    (:3
      (Transformation|AttachActorToComponent _fortpickup _skeletalmesh "Weapon4" "SnapToTarget" "SnapToTarget" "SnapToTarget" true))
    (:4
      (Transformation|AttachActorToComponent _fortpickup _skeletalmesh "Weapon5" "SnapToTarget" "SnapToTarget" "SnapToTarget" true)))
  (Transformation|AttachActorToComponent _fortpickup _skeletalmesh "Weapon1" "SnapToTarget" "SnapToTarget" "SnapToTarget" true))

(fn IsLocalPlayerRiding
  (bind _passengerpawn (Variables|Default|GetPassengerPawn))
  (bind _driverpawn (Variables|Default|GetDriverPawn))
  (Utilities|FlowControl|Sequence
    (:then_0
      (Utilities|IsValid _driverpawn
        (:Is%20Valid
          (if (Pawn|IsLocallyControlled _driverpawn)
            _))))
    (:then_1
      (Utilities|IsValid _passengerpawn
        (:Is%20Valid
          (if (Pawn|IsLocallyControlled _passengerpawn)
            (return true)
            (else
              _)))
        (:Is%20Not%20Valid
          _))
      (return false))))

(fn OnRep_CorrectiveImpulse)

(fn PlayHapticsOnHit
  (params (HitMagnitude :type Float%20%28double-precision%29))
  (bind _passengerplayercontroller (Variables|Default|GetPassengerPlayerController))
  (bind _drivingplayercontroller (Variables|Default|GetDrivingPlayerController))
  (bind _validplayercontrollers (LocalVariables|GetValidPlayerControllers))
  (Utilities|FlowControl|Sequence
    (:then_0
      (Utilities|IsValid _drivingplayercontroller
        (:Is%20Valid
          (Utilities|Array|Add (LocalVariables|GetValidPlayerControllers) _drivingplayercontroller))))
    (:then_1
      (Utilities|IsValid _passengerplayercontroller
        (:Is%20Valid
          (Utilities|Array|Add (LocalVariables|GetValidPlayerControllers) _passengerplayercontroller))))
    (:then_2
      (if (> (Utilities|Array|Length (LocalVariables|GetValidPlayerControllers)) 0)
        (if (> (Variables|GetHitMagnitude) 1e+04)
          (Utilities|FlowControl|Sequence
            (:then_0
              (LocalVariables|SetCameraShakeMagnitude (Math|Float|MapRangeClamped (Variables|GetHitMagnitude) 1.5e+04 5e+04 0.1 1.0))
              (Utilities|IsValid (Variables|Default|GetDrivingPlayerController)
                (:Is%20Valid
                  (Game|Feedback|ClientStartCameraShake :self (LocalVariables|GetValidPlayerControllers) :Shake "/Game/Athena/DrivableVehicles/CameraShake_HitImpact_ShoppingCart.CameraShake_HitImpact_ShoppingCart_C" :Scale (LocalVariables|GetCameraShakeMagnitude) :PlaySpace "CameraLocal")
                  (if (> (Variables|GetHitMagnitude) 2e+04)
                    (Game|Feedback|ClientPlayForceFeedback _validplayercontrollers "/Game/Weapons/ForceFeedback/FF_Generic_Impact_Near.FF_Generic_Impact_Near" "None" false false false)
                    (else
                      (Game|Feedback|ClientPlayForceFeedback _validplayercontrollers "/Game/Weapons/ForceFeedback/FF_Generic_Build.FF_Generic_Build" "None" false false false))))))
            (:then_1
              (if (< (Math|Float|Absolute(Float) (Math|Vector|DotProduct (Math|Vector|Normalize (Vehicle|GetPreviousLinearVelocity)) "0, 0,1.000000")) 0.5)
                (bind _output_get (LocalVariables|SetHitDelta (- (Variables|Vehicle|GetCachedSpeed) (Math|Vector|VectorLength (Transformation|GetComponentVelocity (Variables|Vehicle|GetSkeletalMesh))))))
                (if (> _output_get (LocalVariables|GetCameraPushVelocityDelta))
                  (Game|Feedback|ClientStartCameraShake (LocalVariables|GetValidPlayerControllers) "/Game/Athena/DrivableVehicles/CameraShake_BigHit.CameraShake_BigHit_C" (Math|Float|MapRangeClamped (LocalVariables|GetHitDelta) 3e+02 9e+02 0.5 1.0) "UserDefined" (Math|Rotator|MakeRotfromX (Vehicle|GetPreviousLinearVelocity))))))))))))

(fn UpdateDynamicCameraShake
  (Utilities|FlowControl|Sequence
    (:then_0
      (Variables|CameraShake|SetInAirSmoothed (Math|Interpolation|FInterpTo (Variables|CameraShake|GetInAirSmoothed) (- 1.0 (Math|Conversions|ToFloat(Boolean) (Vehicle|VehicleIsinAir))) (Utilities|Time|GetWorldDeltaSeconds) 2e+01))
      (Variables|Default|SetSmoothedSpringCompression (Math|Float|Clamp(Float) (Math|Interpolation|FInterpTo (Variables|Default|GetSmoothedSpringCompression) (* (Variables|Vehicle|GetCurrentMaxSpringCompression) (Variables|GetSpringFudgeValue)) (Utilities|Time|GetWorldDeltaSeconds) 3.0) 0.33))
      (LocalVariables|SetShakeAmount (Math|Float|float*float (Variables|CameraShake|GetInAirSmoothed) (Math|Float|Power (Math|Float|Clamp(Float) (/ (Math|Float|Absolute(Float) (Vehicle|GetForwardSpeedKmh)) (Variables|CameraShake|GetNormalizedSpeed))) (Variables|CameraShake|GetCameraShakeSpeedCurvePow)) (ScalableFloat|GetValueatLevel (Variables|Haptics|GetScreenShakeMult))))
      (LocalVariables|SetCamerShakeAmplitude (Math|Float|Lerp (Variables|CameraShake|GetAmplitudeMin) (Variables|CameraShake|GetAmplitudeMax) (* (LocalVariables|GetShakeAmount) (Variables|Default|GetSmoothedSpringCompression))))
      (LocalVariables|SetShakeFrequency (Math|Float|Lerp (Variables|CameraShake|GetScreenShakeFrequencyMin) (Variables|CameraShake|GetScreenShakeFrequencyMax) (LocalVariables|GetShakeAmount))))
    (:then_2
      (Utilities|FlowControl|Sequence
        (:then_0
          ;[1 1 1] Driver
          (Utilities|IsValid (Variables|Default|GetCameraShake)
            (:Is%20Valid
              (if (Variables|Default|GetbCoastingCameraShake)
                (Variables|/Script/EngineCameras.LegacyCameraShake|SetRotOscillation (Utilities|Struct|MakeROscillator (Utilities|Struct|MakeFOscillator (LocalVariables|GetCamerShakeAmplitude) (LocalVariables|GetShakeFrequency) "PerlinNoise") (Utilities|Struct|MakeFOscillator (LocalVariables|GetCamerShakeAmplitude) (* (LocalVariables|GetShakeFrequency) (Variables|CameraShake|GetScreenShakeYawFrequencyMultipier)) "PerlinNoise")) (Variables|Default|GetCameraShake))
                (else
                  (Variables|/Script/EngineCameras.LegacyCameraShake|SetShakeScale (/ (Vehicle|GetForwardSpeedKmh) 2e+01) (Variables|Default|GetCameraShake)))))))
        (:then_1
          ;[1 1 1] Passenger
          (Utilities|IsValid (Variables|Default|GetPassengerCameraShake)
            (:Is%20Valid
              (Variables|/Script/EngineCameras.LegacyCameraShake|SetRotOscillation (Utilities|Struct|MakeROscillator (Utilities|Struct|MakeFOscillator (* (LocalVariables|GetCamerShakeAmplitude) (Variables|Default|GetPassengerCameraShakeMultiplier)) (LocalVariables|GetShakeFrequency) "PerlinNoise") (Utilities|Struct|MakeFOscillator (* (LocalVariables|GetCamerShakeAmplitude) (Variables|Default|GetPassengerCameraShakeMultiplier)) (* (LocalVariables|GetShakeFrequency) (Variables|CameraShake|GetScreenShakeYawFrequencyMultipier)) "PerlinNoise")) (Variables|Default|GetPassengerCameraShake)))))))))

(fn ConstructionScript
  (Variables|FortAthenaVehicle|SetOverlapComponent (Variables|Default|GetOverlapTrigger)))
