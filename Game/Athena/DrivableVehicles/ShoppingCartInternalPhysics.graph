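# Physics asset for the internal physics of the drivable shopping cart, covering its frame and seat base.
# Each of the two bones, Frame and seat_base, has a single rotated box body.
# One constraint joins seat_base to Frame with twist and swing 1 locked, leaving swing 2 as the only unlocked angular axis, so the seat can hinge against the frame.
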PhysicsAsset /Game/Athena/DrivableVehicles/ShoppingCartInternalPhysics
bodies {
 Frame [{shapeType=Box, extentX=20.1421, extentY=6.6849, extentZ=9.9561, center=<-15.1604 -0.2331 44.5297>, rotation=<rot 0 -90.1885 -75.5076>}]
 seat_base [{shapeType=Box, extentX=1.9331, extentY=19.3833, extentZ=50.7595, center=<18.2736 -0.5697 36.1815>, rotation=<rot -26.7914 1.918 -89.1739>}]
}
constraints [{
  bone1Name=seat_base
  bone2Name=Frame
  swing1LimitDegrees=45
  swing1Motion=Locked
  swing2LimitDegrees=24
  twistLimitDegrees=45
  twistMotion=Locked
 }]
