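# Animation Blueprint for the drivable shopping cart, posing its suspension, wheels and folding seat.
# Each update copies axle, wheel spin and steering values from the push vehicle to drive bone modifiers.
# A RigidBody node takes force from velocity change, while seat cover pitch is remapped from seat base deflection.
# Seat base pitch gets speed-scaled random jitter; similar front-wheel wobble and a seat look-at calculation are left unconnected.
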AnimBlueprint /Game/Athena/DrivableVehicles/ShoppingCartVehicleSK_AnimBP
target_skeleton=&/SK_ShoppingCart_Skeleton
graphs [
 AnimGraph {nodes [
   n1 = Get TruckTransformF() -> (TruckTransformF:Transform)
   n2 = "Transform (Modify) Bone - Bone: suspension_front"(
    Alpha=1
    Translation=TransformLocation!(T:"Transform (by ref)"=@n1, Location:Vector="0, 0, 0") -> (ReturnValue:Vector)
    Rotation=TransformRotation!(T:"Transform (by ref)"=@n1, Rotation:Rotator="0, 0, 0") -> (ReturnValue:Rotator)
    Scale="1,1,1"
   ) {
    rotationMode=BMM_Replace
    rotationSpace=BCS_BoneSpace
    translationMode=BMM_Replace
    translationSpace=BCS_BoneSpace
    alphaBoolBlend {blendOption=Linear}
    boneToModify {boneName=suspension_front}
   }
   n3 = Get TruckTransformB() -> (TruckTransformB:Transform)
   n4 = "Transform (Modify) Bone - Bone: suspension_back"(
    ComponentPose=@n2
    Alpha=1
    Translation=TransformLocation(T=@n3, Location="0, 0, 0")
    Rotation=TransformRotation(T=@n3, Rotation="0, 0, 0")
    Scale="1,1,1"
   ) {
    rotationMode=BMM_Replace
    rotationSpace=BCS_BoneSpace
    translationMode=BMM_Replace
    translationSpace=BCS_BoneSpace
    alphaBoolBlend {blendOption=Linear}
    boneToModify {boneName=suspension_back}
   }
   n5 = "Transform (Modify) Bone - Bone: dyn_wheel_FR"(ComponentPose=@n4, Alpha=1, Translation="0,0,0", Rotation=MakeRotator(Yaw=Get WheelRotationFR() -> (WheelRotationFR:double)), Scale="1,1,1") {
    rotationMode=BMM_Replace
    rotationSpace=BCS_BoneSpace
    translationSpace=BCS_BoneSpace
    alphaBoolBlend {blendOption=Linear}
    boneToModify {boneName=dyn_wheel_FR}
   }
   n6 = "Transform (Modify) Bone - Bone: dyn_wheel_FL"(ComponentPose=@n5, Alpha=1, Translation="0,0,0", Rotation=MakeRotator(Yaw=Get WheelRotationFL() -> (WheelRotationFL:double)), Scale="1,1,1") {
    rotationMode=BMM_Replace
    rotationSpace=BCS_BoneSpace
    translationSpace=BCS_BoneSpace
    alphaBoolBlend {blendOption=Linear}
    boneToModify {boneName=dyn_wheel_FL}
   }
   n7 = "Transform (Modify) Bone - Bone: wheel_FR_spin"(ComponentPose=@n6, Alpha=1, Translation="0,0,0", Rotation=MakeRotator(Pitch=Get WheelSpinFR() -> (WheelSpinFR:double)), Scale="1,1,1") {rotationMode=BMM_Replace, rotationSpace=BCS_BoneSpace, alphaBoolBlend {blendOption=Linear}, boneToModify {boneName=wheel_FR_spin}}
   n8 = "Transform (Modify) Bone - Bone: wheel_FL_spin"(ComponentPose=@n7, Alpha=1, Translation="0,0,0", Rotation=MakeRotator(Pitch=Get WheelSpinFL() -> (WheelSpinFL:double)), Scale="1,1,1") {rotationMode=BMM_Replace, rotationSpace=BCS_BoneSpace, alphaBoolBlend {blendOption=Linear}, boneToModify {boneName=wheel_FL_spin}}
   n9 = "Transform (Modify) Bone - Bone: wheel_BR"(ComponentPose=@n8, Alpha=1, Translation="0,0,0", Rotation=MakeRotator(Pitch=Get WheelSpinBR() -> (WheelSpinBR:double)), Scale="1,1,1") {rotationMode=BMM_Replace, rotationSpace=BCS_BoneSpace, alphaBoolBlend {blendOption=Linear}, boneToModify {boneName=wheel_BR}}
   n10 = "Transform (Modify) Bone - Bone: wheel_BL"(ComponentPose=@n9, Alpha=1, Translation="0,0,0", Rotation=MakeRotator(Pitch=Get WheelSpinBL() -> (WheelSpinBL:double)), Scale="1,1,1") {rotationMode=BMM_Replace, rotationSpace=BCS_BoneSpace, alphaBoolBlend {blendOption=Linear}, boneToModify {boneName=wheel_BL}}
   n11 = "Transform (Modify) Bone - Bone: seat_cover"(
    ComponentPose=RigidBody(ComponentPose=@n10, ExternalForce=Get SeatCoverForce() -> (SeatCoverForce:Vector)) {overridePhysicsAsset=&/ShoppingCartInternalPhysics, alphaBoolBlend {blendOption=Linear}}
    Alpha=1
    Translation="0,0,0"
    Rotation=Get SeatCoverRotation() -> (SeatCoverRotation:Rotator)
    Scale="1,1,1"
   ) {rotationMode=BMM_Replace, rotationSpace=BCS_BoneSpace, alphaBoolBlend {blendOption=Linear}, boneToModify {boneName=seat_cover}}
   n12 = "Transform (Modify) Bone - Bone: seat_base"(
    ComponentPose=@n11
    Alpha=1
    Translation=Get SeatBasePosition() -> (SeatBasePosition:Vector)
    Rotation=Get SeatBaseRotation() -> (SeatBaseRotation:Rotator)
    Scale="1,1,1"
   ) {
    rotationMode=BMM_Replace
    rotationSpace=BCS_BoneSpace
    translationMode=BMM_Replace
    translationSpace=BCS_BoneSpace
    alphaBoolBlend {blendOption=Linear}
    boneToModify {boneName=seat_base}
   }
   Output Pose(Result=Component To Local(ComponentPose=@n12)) {layerGroup=None}
  ]}
 EventGraph {nodes [
   Event BlueprintUpdateAnimation()
   > n1 = Cast To FortAthenaSKPushVehicle!(execute, Object=GetOwningActor()) -> (then, CastFailed, AsFort Athena SKPush Vehicle:Fort Athena SKPush Vehicle Object Reference) {is=title K2Node_DynamicCast}
   n2 = Get AxleRollF!(self:Fort Athena SKVehicle Object Reference=@n1) -> (AxleRollF:float)
   n3 = Get AxleCenterF!(self:Fort Athena SKVehicle Object Reference=@n1) -> (AxleCenterF:float)
   > Set TruckTransformF(TruckTransformF:Transform=MakeTransform(Location=MakeVector(Y=0, Z=@n3), Rotation=MakeRotator(Roll=@n2))) -> (Output_Get:Transform)
   n4 = Get AxleCenterB!(self:Fort Athena SKVehicle Object Reference=@n1) -> (AxleCenterB:float)
   n5 = Get AxleRollB!(self:Fort Athena SKVehicle Object Reference=@n1) -> (AxleRollB:float)
   > Set TruckTransformB(TruckTransformB:Transform=MakeTransform(Location=MakeVector(Y=0, Z=@n4), Rotation=MakeRotator(Roll=@n5))) -> (Output_Get:Transform)
   > Set WheelSpinFR(WheelSpinFR:double=Get WheelSpinFR!(self:Fort Athena SKVehicle Object Reference=@n1) -> (WheelSpinFR:float)) -> (Output_Get:double)
   > Set WheelSpinFL(WheelSpinFL:double=Get WheelSpinFL!(self:Fort Athena SKVehicle Object Reference=@n1) -> (WheelSpinFL:float)) -> (Output_Get:double)
   > Set WheelSpinBR(WheelSpinBR:double=Get WheelSpinBR!(self:Fort Athena SKVehicle Object Reference=@n1) -> (WheelSpinBR:float)) -> (Output_Get:double)
   > Set WheelSpinBL(WheelSpinBL:double=Get WheelSpinBL!(self:Fort Athena SKVehicle Object Reference=@n1) -> (WheelSpinBL:float)) -> (Output_Get:double)
   > Set WheelRotationFR(WheelRotationFR:double=Get WheelRotationFR!(self:Fort Athena SKVehicle Object Reference=@n1) -> (WheelRotationFR:float)) -> (Output_Get:double)
   > Set WheelRotationFL(WheelRotationFL:double=Get WheelRotationFL!(self:Fort Athena SKVehicle Object Reference=@n1) -> (WheelRotationFL:float)) -> (Output_Get:double)
   n6 = Get SkeletalMesh!(self:Fort Athena Vehicle Object Reference=@n1) -> (SkeletalMesh:Skeletal Mesh Component Budgeted Object Reference)
   n7 = BreakRotator(InRot=BreakTransform(InTransform=GetDeltaTransformFromRefPose(self=@n6, BoneName=seat_base, BaseName=Frame)).Rotation)
   > Set SeatCoverRotation(SeatCoverRotation:Rotator=MakeRotator(Pitch=MapRangeClamped(Value=@n7.Pitch, InRangeA=0, InRangeB=50, OutRangeA=-9, OutRangeB=80))) -> (Output_Get:Rotator)
   n8 = GetVelocity!(self:Actor Object Reference=@n1) -> (ReturnValue:Vector)
   n9 = GetPreviousLinearVelocity!(self:Fort Athena Vehicle Object Reference=@n1) -> (ReturnValue:Vector)
   > Set SeatCoverForce(SeatCoverForce:Vector=vector * float(A=vector - vector!(A:Vector=@n9, B:Vector=@n8) -> (ReturnValue:Vector) {is=title K2Node_CallFunction}, B=49)) -> (Output_Get:Vector)
   n10 = "Clamp (Float)"(Value=float * float(A=Vector Length(A=@n9), B=0.0005), Min=0, Max=1)
   n11 = RandomFloatInRange(Min=0, Max=1)
   > Set SeatBaseRotation(SeatBaseRotation:Rotator=MakeRotator(Pitch=float * float(A=@n11, B=@n10))) -> (Output_Get:Rotator)
   n12 = RandomFloatInRange(Min=-5, Max=5)
   Set WheelRotationFR(WheelRotationFR=float + float(A=float * float(A=@n12, B=@n10), B=Get WheelRotationFR!() -> (WheelRotationFR:double)) {is=title K2Node_CommutativeAssociativeBinaryOperator})
   n13 = RandomFloatInRange(Min=-5, Max=5)
   > Set WheelRotationFL(WheelRotationFL=float + float(A=float * float(A=@n13, B=@n10), B=Get WheelRotationFL!() -> (WheelRotationFL:double)))
   n14 = Get SkeletalMesh(self=@n1)
   n15 = GetSocketLocation!(self:Scene Component Object Reference=@n14, InSocketName:Name=SeatCoverPivot) -> (ReturnValue:Vector)
   n16 = GetSocketLocation(self=@n14, InSocketName=SeatBaseSlideJoint)
   BreakRotator(InRot=TransformToBoneSpace!(
     self:Skinned Mesh Component Object Reference=@n14
     BoneName:Name=root
     InPosition:Vector="0, 0, 0"
     InRotation:Rotator=FindLookAtRotation!(Start:"Vector (by ref)"=@n15, Target:"Vector (by ref)"=@n16) -> (ReturnValue:Rotator)
    ) -> (OutPosition:Vector, OutRotation:Rotator).OutRotation)
  ]}
]
variables [
 SeatBasePosition {type=Vector}
 SeatBaseRotation {type=Rotator}
 SeatCoverForce {type=Vector}
 SeatCoverRotation {type=Rotator}
 TruckTransformB {type=Transform, default {location=<0 0 0>, rotation=<rot 0 0 0>, scale=<1 1 1>}}
 TruckTransformF {type=Transform, default {location=<0 0 0>, rotation=<rot 0 0 0>, scale=<1 1 1>}}
 WheelRotationFL {type=number}
 WheelRotationFR {type=number}
 WheelSpinBL {type=number}
 WheelSpinBR {type=number}
 WheelSpinFL {type=number}
 WheelSpinFR {type=number}
]
