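; Surfboard vehicle Blueprint for Athena that adapts an inherited hoverboard-style vehicle for use on water.
; Begin play registers the water overlap volume, and cosmetics setup disables hoverboard glow and ripple emitters.
; Water body and volume enter/exit events stop the current movement audio and re-cache pointers for surf or hover loops.
; Overrides check in-water state to choose mount and dismount sounds and adjust jump and landing audio.
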(event EventBeginPlay
  ;[0.2225 1 0.1422] Set the water overlap component
  (Variables|FortAthenaVehicle|SetWaterOverlapComponent (Variables|Default|GetWaterOverlapVolume))
  (|CallParentFunction))

(event EventSetupCosmetics
  (bind _loopingfx (Variables|Default|GetLoopingFX))
  ;[0.9474 0.236 1] Overriding to hide hoverboard lights and play mount sound depending on if we are in water or not
  (Components|Activation|Activate _loopingfx false)
  (Effects|Components|ParticleSystem|SetEmitterEnable _loopingfx (Variables|Default|GetHoverboardGlowEmitterName) false)
  (Effects|Components|ParticleSystem|SetEmitterEnable _loopingfx (Variables|Default|GetHoverboardRippleFrontEmitterName) false)
  (Effects|Components|ParticleSystem|SetEmitterEnable _loopingfx (Variables|Default|GetHoverboardRippleRearEmitterName) false)
  (Audio|SpawnSoundAttached (select (Class|FortVehicleInterface|IsinWater) (Variables|Sounds|GetWaterMountSound) (Variables|Sounds|GetMountSound)) (Variables|Vehicle|GetSkeletalMesh)))

(event EventOnEnteredWaterBody
  (params (WaterBody :type Fort%20Water%20Body%20Actor%20Object%20Reference))
  ;[0 1 0.937] Switch between hoverboard and surfboard movement audio when entering and exiting water
  (Utilities|FlowControl|Sequence
    (:then_0
      (Class|FortVehicleAudioVoice|StopAllAudio (Variables|Default|GetAudioMovement))
      (CallFunction|CacheAudioPointers (Variables|Default|GetWaterAudioMovement) (Variables|Default|GetAudioBoost) (Variables|Default|GetAudioWind)))
    (:then_1
      (|CallParentFunction WaterBody))))

(event Vehicle|Water|EventOnEnteredWaterVolume
  (params (WaterSurfacePoint :type Vector%20%28by%20ref%29 :default "0, 0, 0"))
  ;[0 1 0.937] Switch between hoverboard and surfboard movement audio when entering and exiting water
  (Utilities|FlowControl|Sequence
    (:then_0
      (Class|FortVehicleAudioVoice|StopAllAudio (Variables|Default|GetAudioMovement))
      (CallFunction|CacheAudioPointers (Variables|Default|GetWaterAudioMovement) (Variables|Default|GetAudioBoost) (Variables|Default|GetAudioWind)))
    (:then_1
      (|CallParentFunction WaterSurfacePoint))))

(event EventOnExitedWaterBody
  (params (WaterBody :type Fort%20Water%20Body%20Actor%20Object%20Reference))
  ;[0 1 0.937] Switch between hoverboard and surfboard movement audio when entering and exiting water
  (Utilities|FlowControl|Sequence
    (:then_0
      (Class|FortVehicleAudioVoice|StopAllAudio (Variables|Default|GetWaterAudioMovement))
      (CallFunction|CacheAudioPointers (Variables|Default|GetAudioMovement) (Variables|Default|GetAudioBoost) (Variables|Default|GetAudioWind)))
    (:then_1
      (|CallParentFunction WaterBody))))

(event Vehicle|Water|EventOnExitedWaterVolume
  (params (WaterSurfacePoint :type Vector%20%28by%20ref%29 :default "0, 0, 0"))
  ;[0 1 0.937] Switch between hoverboard and surfboard movement audio when entering and exiting water
  (Utilities|FlowControl|Sequence
    (:then_0
      (Class|FortVehicleAudioVoice|StopAllAudio (Variables|Default|GetWaterAudioMovement))
      (CallFunction|CacheAudioPointers (Variables|Default|GetAudioMovement) (Variables|Default|GetAudioBoost) (Variables|Default|GetAudioWind)))
    (:then_1
      (|CallParentFunction WaterSurfacePoint))))

(event EventOnJumped
  (params (Charge :type Float%20%28single-precision%29))
  ;[0.5113 0.2222 1] Override Jump Audio if in water
  (|CallParentFunction Charge)
  (if (Class|FortVehicleInterface|IsinWater)
    (Components|Activation|Deactivate (Variables|Default|GetAudioJump))
    (Components|Activation|Activate (Variables|Default|GetWaterAudioJump) true)))

(event EventOnCollisionHitEffects
  (params (HitLocation :type Vector%20%28by%20ref%29 :default "0, 0, 0") (HitNormalImpulse :type Vector%20%28by%20ref%29 :default "0, 0, 0") (HitFrictionImpulse :type Vector%20%28by%20ref%29 :default "0, 0, 0") (HitNormal :type Vector%20%28by%20ref%29 :default "0, 0, 0") (HitActor :type Actor%20Object%20Reference) (HitSurfaceType :type EPhysicalSurface%20Enum :default "SurfaceType_Default"))
  ;[0.8997 0.1041 1] Override Land Audio if in water
  (|CallParentFunction HitLocation HitNormalImpulse HitFrictionImpulse HitNormal HitActor HitSurfaceType)
  (if (Class|FortVehicleInterface|IsinWater)
    (Components|Activation|Deactivate (Variables|Default|GetAudioLand))
    (Components|Activation|Activate (Variables|Default|GetWaterAudioLand) true)))

(event EventOnPawnExitVehicle
  (params (PlayerPawn :type Fort%20Player%20Pawn%20Object%20Reference) (ExitSocketName :type Name :default "None"))
  ;[1 0.2299 0.6154] Override Dismount Audio if in water
  (Variables|Sounds|SetDismountSound (select (Class|FortVehicleInterface|IsinWater) (Variables|Sounds|GetWaterDismountSound) (Variables|Sounds|GetLandDismountSound)))
  (|CallParentFunction PlayerPawn ExitSocketName))

(fn ConstructionScript
  (|CallParentFunction))
